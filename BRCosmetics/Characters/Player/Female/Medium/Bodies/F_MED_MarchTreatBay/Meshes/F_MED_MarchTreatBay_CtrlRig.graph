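# Control Rig for the F_MED_MarchTreatBay female character body, adding corrective deformation to body and clothing bones.
# A sequence runs the shared DeformRig function, then weighted parent constraints blend the laces, pelvis extra and jacket master bones between body bones.
# Spherical pose readers on upper arms, thighs and neck then drive shoulder, collar and front-jacket offsets, with curves remapping the thigh weights.
# A right-foot reader, scaled by dyn_crouch, adjusts right ankle bones only.
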ControlRigBlueprint /BRCosmetics/Characters/Player/Female/Medium/Bodies/F_MED_MarchTreatBay/Meshes/F_MED_MarchTreatBay_CtrlRig
graphs [
 RigVMFunctionLibrary {links [], nodes []}
 RigVMFunction_Sequence_1_ContainedGraph {nodes [
   n1 = RigVMFunctionEntryNode()
   RigVMFunction_Sequence = &RigVMFunction_Sequence(ExecuteContext=@n1.ExecuteContext)
   n2 = &RigVMFunction_Sequence(ExecuteContext=@RigVMFunction_Sequence.B)
   n3 = &RigVMFunction_Sequence(ExecuteContext=@n2.B)
   n4 = &RigVMFunction_Sequence(ExecuteContext=@n3.B)
   n5 = &RigVMFunction_Sequence(ExecuteContext=@n4.B)
   n6 = &RigVMFunction_Sequence(ExecuteContext=@n5.B)
   RigVMFunctionReturnNode(A=@RigVMFunction_Sequence.A, B=@n2.A, C=@n3.A, D=@n4.A, E=@n5.A, F=@n6.A, G=@n6.B)
  ]}
 RigVMModel {nodes [
   RigUnit_BeginExecution = &RigUnit_BeginExecution()
   RigVMFunction_Sequence = RigVMAggregateNode(ExecuteContext=@RigUnit_BeginExecution.ExecutePin)
   FN_DeformRig_FNC_1 = RigVMFunctionReferenceNode(
    DeformRig_ Curve="(deformRig_l_upperarm_up_alpha_72_#1=False,deformRig_r_upperarm_up_alpha_73_#2=False,deformRig_l_upperarm_fr_alpha_74_#3=False,deformRig_r_upperarm_fr_alpha_75_CB51B8 ...(truncated)"
    DeformRig_Bool="(upperarm_1_#4=True,elbow_3_#5=True,wrist_7_#6=True,groin_8_#7=True,glute_17_#8=True,knee_9_40C408324CA261 ...(truncated)"
    DeformRig_Debug=(upperarm_debug_18_#4=False,wrist_debug_20_#6=False,lat_debug_26_#9=False,groinGlute_debug_27_#7=False)
    DeformRig_Value="(upperarm_up_18_#10=(Rotation=(X=0,Y=0,Z=0,W=1),Translation=(X=0,Y=0,Z=8),Scale3D=(X=0.7,Y=1,Z=1)),upperarm_fwd_fr_19_#11=(Rotation=(X=0,Y=0,Z=0,W=1),Translation=(X=0,Y=10,Z=0),Scale3D=(X=0.5,Y=1,Z= ...(truncated)"
    ExecuteContext=@RigVMFunction_Sequence.A
   )
   &RigUnit_ParentConstraint(
    AdvancedSettings=(InterpolationType=Average,RotationOrderForFilter=XZY)
    Child=(Type=Bone,Name="extra_laces_MSTR")
    Parents=((Item=(Type=Bone,Name="pelvis"),Weight=0.778),(Item=(Type=Bone,Name="spine_01"),Weight=0.184),(Item=(Type=Bone,Name="spine_02"),Weight=0.038))
    Weight=1
    ExecutePin=@RigVMFunction_Sequence.B
   )
   > &RigUnit_ParentConstraint(
    AdvancedSettings=(InterpolationType=Average,RotationOrderForFilter=XZY)
    Child=(Type=Bone,Name="extrta_pelvis")
    Parents="((Item=(Type=Bone,Name=\"deform_groin_l\"),Weight=0.065),(Item=(Type=Bone,Name=\"deform_groin_r\"),Weight=0.065),(Item=(Type=Bone,Name=\"pelvis\"),Weight=0.212),(Item=(Type=Bone,Name=\"thigh_l\"),Weight=0.095),(Item=(Type=Bone,Name=\"thigh_r\"),Weight=0.107),(Item=( ...(truncated)"
    Weight=1
   )
   > &RigUnit_ParentConstraint(
    AdvancedSettings=(InterpolationType=Average,RotationOrderForFilter=XZY)
    Child=(Type=Bone,Name="dyn_jacket_l_fr_MSTR")
    Parents="((Item=(Type=Bone,Name=\"deform_pec_l\"),Weight=0.02),(Item=(Type=Bone,Name=\"pelvis\"),Weight=0.236),(Item=(Type=Bone,Name=\"spine_01\"),Weight=0.047),(Item=(Type=Bone,Name=\"spine_02\"),Weight=0.145),(Item=(Type=Bone,Name=\"spine_03\"),Weight=0.105),(Item=(Type=Bo ...(truncated)"
    Weight=1
   )
   > &RigUnit_ParentConstraint(
    AdvancedSettings=(InterpolationType=Average,RotationOrderForFilter=XZY)
    Child=(Type=Bone,Name="dyn_jacket_l_MSTR")
    Parents="((Item=(Type=Bone,Name=\"clavicle_l\"),Weight=0.018),(Item=(Type=Bone,Name=\"pelvis\"),Weight=0.272),(Item=(Type=Bone,Name=\"spine_01\"),Weight=0.059),(Item=(Type=Bone,Name=\"spine_02\"),Weight=0.169),(Item=(Type=Bone,Name=\"spine_03\"),Weight=0.097),(Item=(Type=Bon ...(truncated)"
    Weight=1
   )
   > &RigUnit_ParentConstraint(
    AdvancedSettings=(InterpolationType=Average,RotationOrderForFilter=XZY)
    Child=(Type=Bone,Name="dyn_jacket_l_bk_MSTR")
    Parents="((Item=(Type=Bone,Name=\"clavicle_l\"),Weight=0.021),(Item=(Type=Bone,Name=\"pelvis\"),Weight=0.261),(Item=(Type=Bone,Name=\"spine_01\"),Weight=0.071),(Item=(Type=Bone,Name=\"spine_02\"),Weight=0.186),(Item=(Type=Bone,Name=\"spine_03\"),Weight=0.12),(Item=(Type=Bone ...(truncated)"
    Weight=1
   )
   > &RigUnit_ParentConstraint(
    AdvancedSettings=(InterpolationType=Average,RotationOrderForFilter=XZY)
    Child=(Type=Bone,Name="dyn_jacket_r_fr_MSTR")
    Parents="((Item=(Type=Bone,Name=\"deform_pec_r\"),Weight=0.02),(Item=(Type=Bone,Name=\"pelvis\"),Weight=0.236),(Item=(Type=Bone,Name=\"spine_01\"),Weight=0.047),(Item=(Type=Bone,Name=\"spine_02\"),Weight=0.145),(Item=(Type=Bone,Name=\"spine_03\"),Weight=0.105),(Item=(Type=Bo ...(truncated)"
    Weight=1
   )
   > &RigUnit_ParentConstraint(
    AdvancedSettings=(InterpolationType=Average,RotationOrderForFilter=XZY)
    Child=(Type=Bone,Name="dyn_jacket_r_MSTR")
    Parents="((Item=(Type=Bone,Name=\"clavicle_r\"),Weight=0.018),(Item=(Type=Bone,Name=\"pelvis\"),Weight=0.272),(Item=(Type=Bone,Name=\"spine_01\"),Weight=0.059),(Item=(Type=Bone,Name=\"spine_02\"),Weight=0.169),(Item=(Type=Bone,Name=\"spine_03\"),Weight=0.097),(Item=(Type=Bon ...(truncated)"
    Weight=1
   )
   > &RigUnit_ParentConstraint(
    AdvancedSettings=(InterpolationType=Average,RotationOrderForFilter=XZY)
    Child=(Type=Bone,Name="dyn_jacket_r_bk_MSTR")
    Parents="((Item=(Type=Bone,Name=\"clavicle_r\"),Weight=0.021),(Item=(Type=Bone,Name=\"pelvis\"),Weight=0.261),(Item=(Type=Bone,Name=\"spine_01\"),Weight=0.071),(Item=(Type=Bone,Name=\"spine_02\"),Weight=0.186),(Item=(Type=Bone,Name=\"spine_03\"),Weight=0.12),(Item=(Type=Bone ...(truncated)"
    Weight=1
   )
   > &RigUnit_ParentConstraint(
    AdvancedSettings=(InterpolationType=Average,RotationOrderForFilter=XZY)
    Child=(Type=Bone,Name="dyn_jacket_bk_MSTR")
    Parents="((Item=(Type=Bone,Name=\"clavicle_l\"),Weight=0.013),(Item=(Type=Bone,Name=\"pelvis\"),Weight=0.261),(Item=(Type=Bone,Name=\"spine_01\"),Weight=0.082),(Item=(Type=Bone,Name=\"spine_02\"),Weight=0.19),(Item=(Type=Bone,Name=\"spine_03\"),Weight=0.147),(Item=(Type=Bone ...(truncated)"
    Weight=1
   )
   n1 = &RigVMFunction_Sequence(ExecuteContext=@RigVMFunction_Sequence.C)
   n2 = &RigUnit_SphericalPoseReader(
    ActiveRegionScaleFactors=(PositiveWidth=0.5,NegativeWidth=1,PositiveHeight=1,NegativeHeight=1)
    ActiveRegionSize=0.2
    Debug=(bDrawDebug=false,bDraw2D=False,bDrawLocalAxes=False,DebugScale=25,DebugSegments=20,DebugThickness=0.25)
    DriverAxis=(X=1,Y=0,Z=0)
    DriverItem=(Type=Bone,Name="upperarm_l")
    FalloffRegionScaleFactors=(PositiveWidth=0.5,NegativeWidth=1,PositiveHeight=1,NegativeHeight=1)
    FalloffSize=0.35
    FlipHeightScaling=False
    FlipWidthScaling=False
    OptionalParentItem=(Type=Bone,Name="spine_05")
    RotationOffset=(X=0,Y=0,Z=0)
    ExecutePin=@n1.A
   )
   > /Script/ControlRig.RigUnit_SetRelativeTransformForItem(
    Child=(Type=Bone,Name="deform_upperarm_up_l")
    Parent=(Type=Bone,Name="deform_upperarm_up_l")
    Value=(Rotation=(X=0,Y=0.2588,Z=0,W=0.9659),Translation=(X=4,Y=0,Z=-2),Scale3D=(X=1,Y=1,Z=1))
    bParentInitial=false
    bPropagateToChildren=true
    Weight=@n2.OutputParam
   ) {
    is=struct RigVMUnitNode
    template="Set Relative Transform::Execute(in Child,in Parent,in bParentInitial,in Value,in Weight,in bPropagateToChildren)"
   }
   > &RigUnit_SetRelativeTransformForItem(
    Child=(Type=Bone,Name="dyn_collar_l")
    Parent=(Type=Bone,Name="dyn_collar_l")
    Value=(Rotation=(X=0,Y=0,Z=0.1736,W=0.9848),Translation=(X=0,Y=0,Z=0),Scale3D=(X=1,Y=1,Z=1))
    bParentInitial=false
    bPropagateToChildren=true
    Weight=@n2.OutputParam
   ) {template="Set Relative Transform::Execute(in Child,in Parent,in bParentInitial,in Value,in Weight,in bPropagateToChildren)"}
   n3 = &RigUnit_SphericalPoseReader(
    ActiveRegionScaleFactors=(PositiveWidth=0.5,NegativeWidth=1,PositiveHeight=1,NegativeHeight=1)
    ActiveRegionSize=0.2
    Debug=(bDrawDebug=false,bDraw2D=False,bDrawLocalAxes=False,DebugScale=25,DebugSegments=20,DebugThickness=0.25)
    DriverAxis=(X=1,Y=0,Z=0)
    DriverItem=(Type=Bone,Name="upperarm_r")
    FalloffRegionScaleFactors=(PositiveWidth=0.5,NegativeWidth=1,PositiveHeight=1,NegativeHeight=1)
    FalloffSize=0.35
    FlipHeightScaling=False
    FlipWidthScaling=False
    OptionalParentItem=(Type=Bone,Name="spine_05")
    RotationOffset=(X=0,Y=0,Z=0)
    ExecutePin=@n1.B
   )
   > &RigUnit_SetRelativeTransformForItem(
    Child=(Type=Bone,Name="deform_upperarm_up_r")
    Parent=(Type=Bone,Name="deform_upperarm_up_r")
    Value=(Rotation=(X=0,Y=0.2588,Z=0,W=0.9659),Translation=(X=-4,Y=0,Z=2),Scale3D=(X=1,Y=1,Z=1))
    bParentInitial=false
    bPropagateToChildren=true
    Weight=@n3.OutputParam
   ) {template="Set Relative Transform::Execute(in Child,in Parent,in bParentInitial,in Value,in Weight,in bPropagateToChildren)"}
   > &RigUnit_SetRelativeTransformForItem(
    Child=(Type=Bone,Name="dyn_collar_r")
    Parent=(Type=Bone,Name="dyn_collar_r")
    Value=(Rotation=(X=0,Y=0,Z=0.1736,W=0.9848),Translation=(X=0,Y=0,Z=0),Scale3D=(X=1,Y=1,Z=1))
    bParentInitial=false
    bPropagateToChildren=true
    Weight=@n3.OutputParam
   ) {template="Set Relative Transform::Execute(in Child,in Parent,in bParentInitial,in Value,in Weight,in bPropagateToChildren)"}
   n4 = &RigVMFunction_Sequence(ExecuteContext=@RigVMFunction_Sequence.D)
   n5 = &RigUnit_SphericalPoseReader(
    ActiveRegionScaleFactors=(PositiveWidth=1,NegativeWidth=1,PositiveHeight=1,NegativeHeight=0.7)
    ActiveRegionSize=0.2
    Debug=(bDrawDebug=false,bDraw2D=false,bDrawLocalAxes=False,DebugScale=25,DebugSegments=20,DebugThickness=0.25)
    DriverAxis=(X=-1,Y=0,Z=0)
    DriverItem=(Type=Bone,Name="thigh_l")
    FalloffRegionScaleFactors=(PositiveWidth=0.5,NegativeWidth=0.9,PositiveHeight=0.6,NegativeHeight=0.8)
    FalloffSize=0.6
    FlipHeightScaling=False
    FlipWidthScaling=False
    OptionalParentItem=(Type=Bone,Name="pelvis")
    RotationOffset=(X=120,Y=0,Z=-40)
    ExecutePin=@n4.A
   )
   n6 = &RigVMFunction_AnimEvalRichCurve(
    Curve="(EditorCurveData=(Keys=((),(InterpMode=RCIM_Cubic,Time=0.2615,Value=0.1266,ArriveTangent=0.7812,LeaveTangent=0.7812),(InterpMode=RCIM_Cubic,Time=0.7297,Value=0.57,ArriveTangent=1.1827,LeaveTangent=1.1827),(InterpMode=RCIM_Cubic,TangentMode=RCTM_User,Time=1 ...(truncated)"
    SourceMinimum=0
    TargetMinimum=0
    Value=@n5.OutputParam
   )
   > &RigUnit_SetRelativeTransformForItem(
    Child=(Type=Bone,Name="dyn_jacket_l_fr_MSTR")
    Parent=(Type=Bone,Name="dyn_jacket_l_fr_MSTR")
    Value=(Rotation=(X=0,Y=0,Z=0,W=1),Translation=(X=-3,Y=0,Z=0),Scale3D=(X=1,Y=1,Z=1))
    bParentInitial=false
    bPropagateToChildren=true
    Weight=@n6.Result
   ) {template="Set Relative Transform::Execute(in Child,in Parent,in bParentInitial,in Value,in Weight,in bPropagateToChildren)"}
   n7 = &RigUnit_SphericalPoseReader(
    ActiveRegionScaleFactors=(PositiveWidth=1,NegativeWidth=1,PositiveHeight=1,NegativeHeight=0.7)
    ActiveRegionSize=0.2
    Debug=(bDrawDebug=false,bDraw2D=false,bDrawLocalAxes=False,DebugScale=25,DebugSegments=20,DebugThickness=0.25)
    DriverAxis=(X=-1,Y=0,Z=0)
    DriverItem=(Type=Bone,Name="thigh_r")
    FalloffRegionScaleFactors=(PositiveWidth=0.5,NegativeWidth=0.9,PositiveHeight=0.6,NegativeHeight=0.8)
    FalloffSize=0.6
    FlipHeightScaling=False
    FlipWidthScaling=False
    OptionalParentItem=(Type=Bone,Name="pelvis")
    RotationOffset=(X=120,Y=0,Z=-40)
    ExecutePin=@n4.B
   )
   n8 = &RigVMFunction_AnimEvalRichCurve(
    Curve="(EditorCurveData=(Keys=((),(InterpMode=RCIM_Cubic,Time=0.2615,Value=0.1266,ArriveTangent=0.7812,LeaveTangent=0.7812),(InterpMode=RCIM_Cubic,Time=0.7297,Value=0.57,ArriveTangent=1.1827,LeaveTangent=1.1827),(InterpMode=RCIM_Cubic,TangentMode=RCTM_User,Time=1 ...(truncated)"
    SourceMinimum=0
    TargetMinimum=0
    Value=@n7.OutputParam
   )
   > &RigUnit_SetRelativeTransformForItem(
    Child=(Type=Bone,Name="dyn_jacket_r_fr_MSTR")
    Parent=(Type=Bone,Name="dyn_jacket_r_fr_MSTR")
    Value=(Rotation=(X=0,Y=0,Z=0,W=1),Translation=(X=-3,Y=0,Z=0),Scale3D=(X=1,Y=1,Z=1))
    bParentInitial=false
    bPropagateToChildren=true
    Weight=@n8.Result
   ) {template="Set Relative Transform::Execute(in Child,in Parent,in bParentInitial,in Value,in Weight,in bPropagateToChildren)"}
   n9 = &RigUnit_SphericalPoseReader(
    ActiveRegionScaleFactors=(PositiveWidth=0.8,NegativeWidth=2,PositiveHeight=3,NegativeHeight=3)
    ActiveRegionSize=0.1
    Debug=(bDrawDebug=false,bDraw2D=False,bDrawLocalAxes=False,DebugScale=25,DebugSegments=20,DebugThickness=0.25)
    DriverAxis=(X=1,Y=0,Z=0)
    DriverItem=(Type=Bone,Name="neck_01")
    FalloffRegionScaleFactors=(PositiveWidth=0.35,NegativeWidth=0.5,PositiveHeight=0.5,NegativeHeight=0.5)
    FalloffSize=0.4
    FlipHeightScaling=False
    FlipWidthScaling=False
    OptionalParentItem=(Type=Bone,Name="spine_05")
    RotationOffset=(X=-90,Y=0,Z=45)
    ExecutePin=@RigVMFunction_Sequence.E
   )
   > &RigUnit_SetRelativeTransformForItem(
    Child=(Type=Bone,Name="extra_collar_back")
    Parent=(Type=Bone,Name="extra_collar_back")
    Value=(Rotation=(X=0,Y=0,Z=-0.2588,W=0.9659),Translation=(X=0,Y=0,Z=0),Scale3D=(X=1,Y=1,Z=1))
    bParentInitial=false
    bPropagateToChildren=true
    Weight=@n9.OutputParam
   ) {template="Set Relative Transform::Execute(in Child,in Parent,in bParentInitial,in Value,in Weight,in bPropagateToChildren)"}
   n10 = &RigVMFunction_Sequence(ExecuteContext=@RigVMFunction_Sequence.F)
   n11 = &RigUnit_SphericalPoseReader(
    ActiveRegionScaleFactors=(PositiveWidth=0.5,NegativeWidth=1,PositiveHeight=1.5,NegativeHeight=1.5)
    ActiveRegionSize=0.1
    Debug=(bDrawDebug=false,bDraw2D=False,bDrawLocalAxes=False,DebugScale=25,DebugSegments=20,DebugThickness=0.25)
    DriverAxis=(X=1,Y=0,Z=0)
    DriverItem=(Type=Bone,Name="upperarm_l")
    FalloffRegionScaleFactors=(PositiveWidth=0.5,NegativeWidth=1,PositiveHeight=1.5,NegativeHeight=1.5)
    FalloffSize=0.25
    FlipHeightScaling=False
    FlipWidthScaling=False
    OptionalParentItem=(Type=Bone,Name="spine_05")
    RotationOffset=(X=65,Y=60,Z=23)
    ExecutePin=@n10.A
   )
   > &RigUnit_SetRelativeTransformForItem(
    Child=(Type=Bone,Name="dyn_collar_l")
    Parent=(Type=Bone,Name="dyn_collar_l")
    Value=(Rotation=(X=0,Y=0,Z=0.342,W=0.9397),Translation=(X=0,Y=0,Z=0),Scale3D=(X=1,Y=1,Z=1))
    bParentInitial=false
    bPropagateToChildren=true
    Weight=@n11.OutputParam
   ) {template="Set Relative Transform::Execute(in Child,in Parent,in bParentInitial,in Value,in Weight,in bPropagateToChildren)"}
   n12 = &RigUnit_SphericalPoseReader(
    ActiveRegionScaleFactors=(PositiveWidth=0.5,NegativeWidth=1,PositiveHeight=1.5,NegativeHeight=1.5)
    ActiveRegionSize=0.1
    Debug=(bDrawDebug=false,bDraw2D=False,bDrawLocalAxes=False,DebugScale=25,DebugSegments=20,DebugThickness=0.25)
    DriverAxis=(X=1,Y=0,Z=0)
    DriverItem=(Type=Bone,Name="upperarm_r")
    FalloffRegionScaleFactors=(PositiveWidth=1,NegativeWidth=0.5,PositiveHeight=1.5,NegativeHeight=1.5)
    FalloffSize=0.25
    FlipHeightScaling=False
    FlipWidthScaling=False
    OptionalParentItem=(Type=Bone,Name="spine_05")
    RotationOffset=(X=65,Y=60,Z=23)
    ExecutePin=@n10.B
   )
   > &RigUnit_SetRelativeTransformForItem(
    Child=(Type=Bone,Name="dyn_collar_r")
    Parent=(Type=Bone,Name="dyn_collar_r")
    Value=(Rotation=(X=0,Y=0,Z=0.342,W=0.9397),Translation=(X=0,Y=0,Z=0),Scale3D=(X=1,Y=1,Z=1))
    bParentInitial=false
    bPropagateToChildren=true
    Weight=@n12.OutputParam
   ) {template="Set Relative Transform::Execute(in Child,in Parent,in bParentInitial,in Value,in Weight,in bPropagateToChildren)"}
   n13 = &RigVMFunction_Sequence(ExecuteContext=@RigVMFunction_Sequence.G)
   n14 = &RigUnit_SphericalPoseReader(
    ActiveRegionScaleFactors=(PositiveWidth=1,NegativeWidth=1,PositiveHeight=2,NegativeHeight=1)
    ActiveRegionSize=0.1
    Debug=(bDrawDebug=true,bDraw2D=False,bDrawLocalAxes=False,DebugScale=25,DebugSegments=20,DebugThickness=0.25)
    DriverAxis=(X=0,Y=-1,Z=0)
    DriverItem=(Type=Bone,Name="foot_r")
    FalloffRegionScaleFactors=(PositiveWidth=1,NegativeWidth=1,PositiveHeight=3,NegativeHeight=1)
    FalloffSize=0.25
    FlipHeightScaling=False
    FlipWidthScaling=False
    OptionalParentItem=(Type=Bone,Name="spine_05")
    RotationOffset=(X=-90,Y=-30,Z=-35)
    ExecutePin=@n13.A
   )
   Multiply = &RigVMFunction_MathDoubleMul(A=@n14.OutputParam, B=RigVMVariableNode() {variable=dyn_crouch}.Value)
   > &RigUnit_ModifyTransforms(
    ItemToModify=((Item=(Type=Bone,Name="dyn_ankle_front_MSTR_r"),Transform=(Rotation=(X=0.1285,Y=0.0227,Z=0.1722,W=0.9764),Translation=(X=0,Y=5,Z=0),Scale3D=(X=1,Y=1,Z=1))))
    WeightMinimum=0
    Weight=@Multiply.Result
   )
   > &RigUnit_ModifyTransforms(
    ItemToModify=((Item=(Type=Bone,Name="dyn_ankle_out_MSTR_r"),Transform=(Rotation=(X=0,Y=0,Z=0,W=1),Translation=(X=-2,Y=0,Z=2),Scale3D=(X=1,Y=1,Z=1))))
    WeightMinimum=0
    Weight=@Multiply.Result
   )
   > &RigUnit_ModifyTransforms(
    ItemToModify=((Item=(Type=Bone,Name="dyn_ankle_in_MSTR_r"),Transform=(Rotation=(X=0,Y=0,Z=0,W=1),Translation=(X=2,Y=0,Z=-2),Scale3D=(X=1,Y=1,Z=1))))
    WeightMinimum=0
    Weight=@Multiply.Result
   )
   AlphaInterp = &RigVMFunction_AlphaInterp(InterpSpeedDecreasing=100, InterpSpeedIncreasing=10, Scale=1, bClampResult=False, bInterpResult=False, Value=@n14.OutputParam)
  ]}
]
hierarchy [
 root {loc=<0 0 0>, rot=<0 0 0 1>}
 pelvis {parent=root, loc=<0 -0.5301 98.0571>, rot=<0.0127 0.707 -0.0127 -0.707>}
 spine_01 {parent=pelvis, loc=<6.1208 -0.9328 0>, rot=<0 0 0.034 -0.9994>}
 spine_02 {parent=spine_01, loc=<6.5997 -0.2095 0>, rot=<0 0 0.0238 -0.9997>}
 spine_03 {parent=spine_02, loc=<6.4914 -0.1237 0>, rot=<0 0 0.0149 -0.9999>}
 spine_04 {parent=spine_03, loc=<6.9598 0.0794 0>, rot=<0 0 -0.0132 -0.9999>}
 spine_05 {parent=spine_04, loc=<6.9987 0 0>, rot=<0 0 -0.0015 -1>}
 neck_01 {parent=spine_05, loc=<11.3138 0.1529 0>, rot=<0 0 -0.2003 -0.9797>}
 neck_02 {parent=neck_01, loc=<4.5627 0 0>, rot=<0 0 0.0504 -0.9987>}
 head {parent=neck_02, loc=<5.3914 0 0>, rot=<0 0 0.115 -0.9934>}
 clavicle_l {parent=spine_05, loc=<5.4831 -0.1053 -3.2946>, rot=<0.0465 -0.7289 -0.009 -0.6829>}
 upperarm_l {parent=clavicle_l, loc=<9.8658 -1.2136 -0.4105>, rot=<0.0164 -0.3245 -0.0269 -0.9454>}
 lowerarm_l {parent=upperarm_l, loc=<26.465 -0.8345 0.259>, rot=<0.0036 0.0161 -0.2298 -0.9731>}
 hand_l {parent=lowerarm_l, loc=<22.7355 -0.3203 -0.6105>, rot=<-0.6784 -0.0223 0.0735 -0.7307>}
 thumb_01_l {parent=hand_l, loc=<1.3118 -0.3191 -2.459>, rot=<0.6546 -0.4042 0.0664 -0.6354>}
 thumb_02_l {parent=thumb_01_l, loc=<5.9336 -0.0448 0.0401>, rot=<-0.0037 0.0213 0.298 -0.9543>}
 thumb_03_l {parent=thumb_02_l, loc=<2.2223 0.3682 -0.0907>, rot=<0.0118 -0.013 0.0066 -0.9998>}
 index_metacarpal_l {parent=hand_l, loc=<1.6535 0.0183 -2.96>, rot=<0.0158 -0.0124 -0.0173 -0.9996>}
 index_01_l {parent=index_metacarpal_l, loc=<8.3008 0.0357 -0.0524>, rot=<-0.0044 0.0485 0.0924 -0.9945>}
 index_02_l {parent=index_01_l, loc=<3.3587 -0.3004 0.0452>, rot=<-0.002 0.0027 0.3225 -0.9466>}
 index_03_l {parent=index_02_l, loc=<2.5578 -0.123 -0.0252>, rot=<-0.002 0.0067 0.2558 -0.9667>}
 middle_metacarpal_l {parent=hand_l, loc=<2.1856 -0.1725 -0.5984>, rot=<0.006 0.0168 -0.0329 -0.9993>}
 middle_01_l {parent=middle_metacarpal_l, loc=<7.3275 -0.2447 0.0129>, rot=<-0.0208 0.0042 0.0858 -0.9961>}
 middle_02_l {parent=middle_01_l, loc=<4.0123 0.1401 0.1175>, rot=<-0.0218 0.0258 0.3512 -0.9357>}
 middle_03_l {parent=middle_02_l, loc=<2.6077 0.0534 0.1373>, rot=<0.0082 0.0188 0.1708 -0.9851>}
 ring_metacarpal_l {parent=hand_l, loc=<2.2074 -0.3625 1.3638>, rot=<-0.0284 0.0722 -0.0458 -0.9959>}
 ring_01_l {parent=ring_metacarpal_l, loc=<7.0295 -0.6197 -0.1446>, rot=<-0.0489 -0.0252 0.1892 -0.9804>}
 ring_02_l {parent=ring_01_l, loc=<3.6174 0.4647 -0.1027>, rot=<0 0 0.2806 -0.9598>}
 ring_03_l {parent=ring_02_l, loc=<2.4152 0.114 0.0199>, rot=<0 0 0.1189 -0.9929>}
 pinky_metacarpal_l {parent=hand_l, loc=<1.7616 -0.9753 3.3814>, rot=<-0.0397 0.0814 -0.069 -0.9935>}
 pinky_01_l {parent=pinky_metacarpal_l, loc=<7.1709 -0.4831 -0.1094>, rot=<-0.0654 -0.0035 0.2242 -0.9723>}
 pinky_02_l {parent=pinky_01_l, loc=<2.8274 0.0893 -0.1043>, rot=<0.0037 0.0011 0.2854 -0.9584>}
 pinky_03_l {parent=pinky_02_l, loc=<2.2046 0.0307 -0.0078>, rot=<-0.0002 0.0098 0.1445 -0.9895>}
 deform_wrist_dn_base_l {parent=hand_l, loc=<0 -1.5326 0>, rot=<0 0 0 -1>}
 deform_wrist_dn_l {parent=deform_wrist_dn_base_l, loc=<0 0 0>, rot=<0 0 0 1>}
 deform_wrist_up_base_l {parent=hand_l, loc=<0 2.3009 0>, rot=<0 0 0 -1>}
 deform_wrist_up_l {parent=deform_wrist_up_base_l, loc=<0 0 0>, rot=<0 0 0 1>}
 lowerarm_twist_02_l {parent=lowerarm_l, loc=<10.6707 0 0>, rot=<0 0 0 1>}
 lowerarm_twist_01_l {parent=lowerarm_l, loc=<18.316 0 0>, rot=<0 0 0 1>}
 dyn_lowerarm_in_MSTR_l {parent=lowerarm_twist_01_l, loc=<-11.6354 6.3304 -1.7724>, rot=<1 0 0 0>}
 dyn_lowerarm_in_01_l {parent=dyn_lowerarm_in_MSTR_l, loc=<0 0 0>, rot=<0 0 0 1>}
 dyn_lowerarm_bot_MSTR_l {parent=lowerarm_twist_01_l, loc=<-13.1791 -0.9504 -6.3146>, rot=<-0.6315 0 0 -0.7754>}
 dyn_lowerarm_bot_01_l {parent=dyn_lowerarm_bot_MSTR_l, loc=<0 0 0>, rot=<0 0 0 1>}
 dyn_lowerarm_out_MSTR_l1 {parent=lowerarm_twist_01_l, loc=<-14.0522 -5.483 -1.5643>, rot=<0 0 0 1>}
 dyn_lowerarm_out_01_l {parent=dyn_lowerarm_out_MSTR_l1, loc=<0 0 0>, rot=<0 0 0 1>}
 dyn_lowerarm_top_MSTR_l {parent=lowerarm_twist_01_l, loc=<-12.3345 0.3568 3.7998>, rot=<0.7071 0 0 -0.7071>}
 dyn_lowerarm_top_01_l {parent=dyn_lowerarm_top_MSTR_l, loc=<0 0 0>, rot=<0 0 0 1>}
 deform_elbow_in_base_l {parent=lowerarm_l, loc=<0.1834 3.6062 -0.2572>, rot=<0 0 0 -1>}
 deform_elbow_in_l {parent=deform_elbow_in_base_l, loc=<0 0 0>, rot=<0 0 0 1>}
 deform_elbow_out_base_l {parent=lowerarm_l, loc=<0 -4 0>, rot=<0 0 0 1>}
 deform_elbow_out_l {parent=deform_elbow_out_base_l, loc=<0 0 0>, rot=<0 0 0 1>}
 upperarm_twist_01_l {parent=upperarm_l, loc=<7.2135 -0.1134 0>, rot=<0 0 0 1>}
 deform_upperarm_bk_base_l {parent=upperarm_twist_01_l, loc=<-6.4053 -4.3746 0.314>, rot=<0 0 0 -1>}
 deform_upperarm_bk_l {parent=deform_upperarm_bk_base_l, loc=<0 0 0>, rot=<0 0 0 1>}
 deform_upperarm_fr_base_l {parent=upperarm_twist_01_l, loc=<-5.6312 5.092 0.3354>, rot=<0 0 0 -1>}
 deform_upperarm_fr_l {parent=deform_upperarm_fr_base_l, loc=<0 0 0>, rot=<0 0 0 1>}
 deform_upperarm_up_base_l {parent=upperarm_twist_01_l, loc=<-8.891 0.4825 3.9517>, rot=<0 0 0 -1>}
 deform_upperarm_up_l {parent=deform_upperarm_up_base_l, loc=<0 0 0>, rot=<0 0 0 1>}
 upperarm_twist_02_l {parent=upperarm_l, loc=<19.7998 -0.1134 0>, rot=<0 0 0 1>}
 clavicle_r {parent=spine_05, loc=<5.4831 -0.1053 3.2946>, rot=<0.7289 0.0465 -0.6829 0.009>}
 upperarm_r {parent=clavicle_r, loc=<-9.8658 1.2136 0.4105>, rot=<0.0164 -0.3245 -0.0269 -0.9454>}
 lowerarm_r {parent=upperarm_r, loc=<-26.465 0.8345 -0.259>, rot=<0.0036 0.0161 -0.2298 -0.9731>}
 hand_r {parent=lowerarm_r, loc=<-22.7355 0.3203 0.6106>, rot=<-0.6784 -0.0223 0.0735 -0.7307>}
 thumb_01_r {parent=hand_r, loc=<-1.3118 0.3191 2.459>, rot=<0.6546 -0.4042 0.0664 -0.6354>}
 thumb_02_r {parent=thumb_01_r, loc=<-5.9336 0.0448 -0.0401>, rot=<-0.0037 0.0213 0.298 -0.9543>}
 thumb_03_r {parent=thumb_02_r, loc=<-2.2223 -0.3682 0.0907>, rot=<0.0118 -0.013 0.0066 -0.9998>}
 index_metacarpal_r {parent=hand_r, loc=<-1.6535 -0.0183 2.96>, rot=<0.0158 -0.0124 -0.0173 -0.9996>}
 index_01_r {parent=index_metacarpal_r, loc=<-8.3008 -0.0357 0.0524>, rot=<-0.0044 0.0485 0.0924 -0.9945>}
 index_02_r {parent=index_01_r, loc=<-3.3587 0.3004 -0.0452>, rot=<-0.002 0.0027 0.3225 -0.9466>}
 index_03_r {parent=index_02_r, loc=<-2.5578 0.123 0.0252>, rot=<-0.002 0.0067 0.2558 -0.9667>}
 middle_metacarpal_r {parent=hand_r, loc=<-2.1856 0.1725 0.5984>, rot=<0.006 0.0168 -0.0329 -0.9993>}
 middle_01_r {parent=middle_metacarpal_r, loc=<-7.3275 0.2447 -0.0129>, rot=<-0.0208 0.0042 0.0858 -0.9961>}
 middle_02_r {parent=middle_01_r, loc=<-4.0123 -0.1401 -0.1175>, rot=<-0.0218 0.0258 0.3512 -0.9357>}
 middle_03_r {parent=middle_02_r, loc=<-2.6077 -0.0534 -0.1373>, rot=<0.0082 0.0188 0.1708 -0.9851>}
 ring_metacarpal_r {parent=hand_r, loc=<-2.2074 0.3625 -1.3638>, rot=<-0.0284 0.0722 -0.0458 -0.9959>}
 ring_01_r {parent=ring_metacarpal_r, loc=<-7.0295 0.6197 0.1446>, rot=<-0.0489 -0.0252 0.1892 -0.9804>}
 ring_02_r {parent=ring_01_r, loc=<-3.6174 -0.4647 0.1027>, rot=<0 0 0.2806 -0.9598>}
 ring_03_r {parent=ring_02_r, loc=<-2.4152 -0.114 -0.0199>, rot=<0 0 0.1189 -0.9929>}
 pinky_metacarpal_r {parent=hand_r, loc=<-1.7616 0.9753 -3.3814>, rot=<-0.0397 0.0814 -0.069 -0.9935>}
 pinky_01_r {parent=pinky_metacarpal_r, loc=<-7.1709 0.4831 0.1094>, rot=<-0.0654 -0.0035 0.2242 -0.9723>}
 pinky_02_r {parent=pinky_01_r, loc=<-2.8274 -0.0893 0.1043>, rot=<0.0037 0.0011 0.2854 -0.9584>}
 pinky_03_r {parent=pinky_02_r, loc=<-2.2046 -0.0307 0.0078>, rot=<-0.0002 0.0098 0.1445 -0.9895>}
 deform_wrist_dn_base_r {parent=hand_r, loc=<-0.0001 1.5328 0>, rot=<0 0 0 -1>}
 deform_wrist_dn_r {parent=deform_wrist_dn_base_r, loc=<0 0 0>, rot=<0 0 0 1>}
 deform_wrist_up_base_r {parent=hand_r, loc=<-0.0002 -2.3007 -0.0001>, rot=<0 0 0 -1>}
 deform_wrist_up_r {parent=deform_wrist_up_base_r, loc=<0 0 0>, rot=<0 0 0 1>}
 lowerarm_twist_02_r {parent=lowerarm_r, loc=<-10.6707 0 0>, rot=<0 0 0 1>}
 lowerarm_twist_01_r {parent=lowerarm_r, loc=<-18.316 0 0>, rot=<0 0 0 1>}
 dyn_lowerarm_in_MSTR_r {parent=lowerarm_twist_01_r, loc=<11.6352 -6.3303 1.7728>, rot=<0 -1 0 0>}
 dyn_lowerarm_in_01_r {parent=dyn_lowerarm_in_MSTR_r, loc=<0 0 0>, rot=<0 0 0 1>}
 dyn_lowerarm_bot_MSTR_r {parent=lowerarm_twist_01_r, loc=<13.1791 0.9504 6.3145>, rot=<0 -0.6315 0.7754 0>}
 dyn_lowerarm_bot_01_r {parent=dyn_lowerarm_bot_MSTR_r, loc=<0 0 0>, rot=<0 0 0 1>}
 dyn_lowerarm_top_MSTR_r {parent=lowerarm_twist_01_r, loc=<12.3347 -0.3569 -3.8001>, rot=<0 0.7071 0.7071 0>}
 dyn_lowerarm_top_01_r {parent=dyn_lowerarm_top_MSTR_r, loc=<0 0 0>, rot=<0 0 0 1>}
 dyn_lowerarm_out_MSTR_r1 {parent=lowerarm_twist_01_r, loc=<14.0522 5.483 1.5642>, rot=<0 0 1 0>}
 dyn_lowerarm_out_01_r {parent=dyn_lowerarm_out_MSTR_r1, loc=<0 0 0>, rot=<0 0 0 1>}
 deform_elbow_in_base_r {parent=lowerarm_r, loc=<-0.1834 -3.6062 0.2572>, rot=<0 0 0 -1>}
 deform_elbow_in_r {parent=deform_elbow_in_base_r, loc=<0 0 0>, rot=<0 0 0 1>}
 deform_elbow_out_base_r {parent=lowerarm_r, loc=<0 4 -0.0001>, rot=<0 0 0 -1>}
 deform_elbow_out_r {parent=deform_elbow_out_base_r, loc=<0 0 0>, rot=<0 0 0 1>}
 upperarm_twist_01_r {parent=upperarm_r, loc=<-7.2135 0.1134 0>, rot=<0 0 0 1>}
 deform_upperarm_bk_base_r {parent=upperarm_twist_01_r, loc=<6.4053 4.3746 -0.314>, rot=<0 0 0 -1>}
 deform_upperarm_bk_r {parent=deform_upperarm_bk_base_r, loc=<0 0 0>, rot=<0 0 0 1>}
 deform_upperarm_fr_base_r {parent=upperarm_twist_01_r, loc=<5.6312 -5.092 -0.3354>, rot=<0 0 0 -1>}
 deform_upperarm_fr_r {parent=deform_upperarm_fr_base_r, loc=<0 0 0>, rot=<0 0 0 1>}
 deform_upperarm_up_base_r {parent=upperarm_twist_01_r, loc=<8.891 -0.4825 -3.9517>, rot=<0 0 0 -1>}
 deform_upperarm_up_r {parent=deform_upperarm_up_base_r, loc=<0 0 0>, rot=<0 0 0 1>}
 upperarm_twist_02_r {parent=upperarm_r, loc=<-19.7998 0.1134 0>, rot=<0 0 0 1>}
 attach_backpack {parent=spine_05, loc=<-0.1227 -13.473 0>, rot=<-0.0093 -0.707 0.0093 -0.707>}
 attach_cape {parent=spine_05, loc=<9.0173 -11.714 0>, rot=<0.0616 -0.7044 -0.0616 -0.7044>}
 dyn_collar_l {parent=spine_05, loc=<8.3849 2.4328 -5.4076>, rot=<-0.6146 -0.2431 -0.0508 -0.7487>}
 extra_collar_back {parent=spine_05, loc=<11.504 -3.3994 0>, rot=<0 0 -0.1191 -0.9929>}
 dyn_collar_r {parent=spine_05, loc=<8.385 2.4328 5.4076>, rot=<0.6146 0.2431 -0.0508 -0.7487>}
 deform_pec_aim_base_l {parent=spine_04, loc=<3.2456 9.3309 -1.2237>, rot=<0.0603 -0.4033 0.3318 -0.8507>}
 deform_pec_aim_l {parent=deform_pec_aim_base_l, loc=<0 0 0>, rot=<0 0 0 1>}
 deform_pec_base_l {parent=deform_pec_aim_l, loc=<9.3665 0 0>, rot=<0 0 0 1>}
 deform_pec_l {parent=deform_pec_base_l, loc=<0 0 0>, rot=<0 0 0 1>}
 deform_pec_aim_base_r {parent=spine_04, loc=<3.2452 9.3309 1.2237>, rot=<-0.4033 -0.0603 0.8507 0.3318>}
 deform_pec_aim_r {parent=deform_pec_aim_base_r, loc=<0 0 0>, rot=<0 0 0 1>}
 deform_pec_base_r {parent=deform_pec_aim_r, loc=<-9.3667 -0.0003 -0.0004>, rot=<0 0 0 1>}
 deform_pec_r {parent=deform_pec_base_r, loc=<0 0 0>, rot=<0 0 0 1>}
 dyn_jacket_l_MSTR {parent=spine_03, loc=<0.3366 0.3437 -12.0904>, rot=<0.0547 -0.9985 0 0>}
 dyn_jacket_l_01 {parent=dyn_jacket_l_MSTR, loc=<0 0 0>, rot=<0 0 0 1>}
 dyn_jacket_l_02 {parent=dyn_jacket_l_01, loc=<9.3155 0 0>, rot=<0 0 0 1>}
 dyn_jacket_l_fr_MSTR {parent=spine_03, loc=<-0.6472 8.9897 -8.2282>, rot=<0.0508 -0.9264 -0.3725 0.0204>}
 dyn_jacket_l_fr_01 {parent=dyn_jacket_l_fr_MSTR, loc=<0 0 0>, rot=<0 0 0 1>}
 dyn_jacket_l_fr_02 {parent=dyn_jacket_l_fr_01, loc=<9.7704 0 0>, rot=<0 0 0 1>}
 dyn_jacket_l_bk_MSTR {parent=spine_03, loc=<1.1083 -6.8189 -9.1508>, rot=<0.0496 -0.9043 0.4234 -0.0232>}
 dyn_jacket_l_bk_01 {parent=dyn_jacket_l_bk_MSTR, loc=<0 0 0>, rot=<0 0 0 1>}
 dyn_jacket_l_bk_02 {parent=dyn_jacket_l_bk_01, loc=<9.2807 0 0>, rot=<0 0 0 1>}
 dyn_jacket_bk_MSTR {parent=spine_03, loc=<1.3821 -9.3893 0>, rot=<0.0387 -0.706 0.706 -0.0387>}
 dyn_jacket_bk_01 {parent=dyn_jacket_bk_MSTR, loc=<0 0 0>, rot=<0 0 0 1>}
 dyn_jacket_bk_02 {parent=dyn_jacket_bk_01, loc=<9.3423 0 0>, rot=<0 0 0 1>}
 dyn_jacket_r_fr_MSTR {parent=spine_03, loc=<-0.6473 8.9897 8.2282>, rot=<-0.0204 0.3725 0.9264 -0.0508>}
 dyn_jacket_r_fr_01 {parent=dyn_jacket_r_fr_MSTR, loc=<0 0 0>, rot=<0 0 0 -1>}
 dyn_jacket_r_fr_02 {parent=dyn_jacket_r_fr_01, loc=<9.77 0 0>, rot=<0 0 0 1>}
 dyn_jacket_r_MSTR {parent=spine_03, loc=<0.3367 0.3437 12.0904>, rot=<0 0 0.9985 -0.0547>}
 dyn_jacket_r_01 {parent=dyn_jacket_r_MSTR, loc=<0 0 0>, rot=<0 0 0 -1>}
 dyn_jacket_r_02 {parent=dyn_jacket_r_01, loc=<9.316 0 0>, rot=<0 0 0 -1>}
 dyn_jacket_r_bk_MSTR {parent=spine_03, loc=<1.1082 -6.8189 9.1508>, rot=<0.0232 -0.4234 0.9043 -0.0496>}
 dyn_jacket_r_bk_01 {parent=dyn_jacket_r_bk_MSTR, loc=<0 0 0>, rot=<0 0 0 -1>}
 dyn_jacket_r_bk_02 {parent=dyn_jacket_r_bk_01, loc=<9.281 0 0>, rot=<0 0 0 1>}
 extra_laces_MSTR {parent=spine_01, loc=<0.0801 9.5043 0>, rot=<0 0 -0.0161 -0.9999>}
 dyn_lace_bot_MSTR_l {parent=extra_laces_MSTR, loc=<-0.7717 0.0681 -0.5813>, rot=<-0.0242 0.001 0.9966 0.0792>}
 dyn_lace_bot_01_l {parent=dyn_lace_bot_MSTR_l, loc=<0 0 0>, rot=<0 0 0 1>}
 dyn_lace_bot_02_l {parent=dyn_lace_bot_01_l, loc=<5.1576 0.0485 -0.0034>, rot=<0.0036 0.0142 -0.0857 -0.9962>}
 dyn_lace_bot_03_l {parent=dyn_lace_bot_02_l, loc=<4.4052 0 0>, rot=<-0.0022 0.0037 -0.0122 -0.9999>}
 dyn_lace_bot_04_l {parent=dyn_lace_bot_03_l, loc=<4.9566 0 0>, rot=<0.0029 0.0046 -0.0143 -0.9999>}
 dyn_lace_MSTR_r {parent=extra_laces_MSTR, loc=<-0.7916 0.0816 0.6061>, rot=<0.0266 0.0034 0.9968 0.0751>}
 dyn_lace_bot_01_r {parent=dyn_lace_MSTR_r, loc=<0 0 0>, rot=<0 0 0 1>}
 dyn_lace_bot_02_r {parent=dyn_lace_bot_01_r, loc=<5.1692 0.1655 -0.0005>, rot=<0.0029 -0.0187 -0.0644 -0.9977>}
 dyn_lace_bot_03_r {parent=dyn_lace_bot_02_r, loc=<4.3794 0 0>, rot=<-0.0001 -0.0041 -0.0261 -0.9997>}
 dyn_lace_bot_04_r {parent=dyn_lace_bot_03_r, loc=<4.9476 0 0>, rot=<0.0005 -0.0001 -0.0239 -0.9997>}
 dyn_lace_bot_05_r {parent=dyn_lace_bot_04_r, loc=<4.6416 0 0>, rot=<0.0013 0.0018 -0.0144 -0.9999>}
 dyn_lace_top_MSTR_l {parent=extra_laces_MSTR, loc=<-0.3359 0.3857 -0.8026>, rot=<-0.1549 -0.025 0.9744 0.1609>}
 dyn_lace_top_01_l {parent=dyn_lace_top_MSTR_l, loc=<0 0 0>, rot=<0 0 0 1>}
 dyn_lace_top_02_l {parent=dyn_lace_top_01_l, loc=<3.1158 0 0>, rot=<0 0 0 1>}
 dyn_lace_top_MSTR_r {parent=extra_laces_MSTR, loc=<-0.3359 0.2941 0.9616>, rot=<0.1385 -0.0647 0.9734 0.1706>}
 dyn_lace_top_01_r {parent=dyn_lace_top_MSTR_r, loc=<0 0 0>, rot=<0 0 0 1>}
 dyn_lace_top_02_r {parent=dyn_lace_top_01_r, loc=<3.3573 0 0>, rot=<-0.0532 -0.1271 -0.0729 -0.9878>}
 thigh_l {parent=pelvis, loc=<-1.0855 -0.2109 -8.7053>, rot=<0.0009 0.0355 0.0155 -0.9993>}
 calf_l {parent=thigh_l, loc=<-42.5284 -0.0019 -0.7297>, rot=<0.0006 -0.0161 -0.0391 -0.9991>}
 foot_l {parent=calf_l, loc=<-46.7401 -2.5014 -0.592>, rot=<0.0427 -0.0238 0.0353 -0.9982>}
 ball_l {parent=foot_l, loc=<-3.5658 10.8066 -0.1544>, rot=<0.0001 -0.0001 -0.7063 -0.7079>}
 calf_twist_02_l {parent=calf_l, loc=<-19.1546 -1.7943 -0.2638>, rot=<0.0237 0.0095 -0.0035 -0.9997>}
 calf_twist_01_l {parent=calf_l, loc=<-35.8483 -1.9197 -0.5794>, rot=<0.0237 0.0095 -0.0035 -0.9997>}
 dyn_ankle_front_MSTR_l {parent=calf_twist_01_l, loc=<17.025 7.3861 0>, rot=<0 0 0.9972 0.0745>}
 dyn_ankle_front_01_l {parent=dyn_ankle_front_MSTR_l, loc=<0 0 0>, rot=<0 0 0 1>}
 dyn_ankle_in_MSTR_l {parent=calf_twist_01_l, loc=<16.5941 0 8.1117>, rot=<0.0274 -0.7066 0.7066 0.0274>}
 dyn_ankle_in_01_l {parent=dyn_ankle_in_MSTR_l, loc=<0 0 0>, rot=<0 0 0 1>}
 dyn_ankle_back_MSTR_l {parent=calf_twist_01_l, loc=<15.4189 -8.1423 0>, rot=<-0.0095 -1 0 0>}
 dyn_ankle_back_01_l {parent=dyn_ankle_back_MSTR_l, loc=<0 0 0>, rot=<0 0 0 1>}
 dyn_ankle_out_MSTR_l {parent=calf_twist_01_l, loc=<15.6894 0 -7.4161>, rot=<0 0.7071 0.7071 0>}
 dyn_ankle_out_01_l {parent=dyn_ankle_out_MSTR_l, loc=<0 0 0>, rot=<0 0 0 1>}
 deform_knee_in_base_l {parent=calf_l, loc=<0.8031 -5.0231 0.2456>, rot=<0 0 0 -1>}
 deform_knee_in_l {parent=deform_knee_in_base_l, loc=<0 0 0>, rot=<0 0 0 1>}
 deform_knee_out_base_l {parent=calf_l, loc=<0 4 0>, rot=<0 0 0 -1>}
 deform_knee_out_l {parent=deform_knee_out_base_l, loc=<0 0 0>, rot=<0 0 0 1>}
 thigh_twist_01_l {parent=thigh_l, loc=<-27.952 0 0>, rot=<0.0013 0 0 -1>}
 deform_groin_base_l {parent=thigh_l, loc=<0.4801 5.6776 -0.2486>, rot=<0 0 0 -1>}
 deform_groin_l {parent=deform_groin_base_l, loc=<0 0 0>, rot=<0 0 0 1>}
 thigh_r {parent=pelvis, loc=<-1.0855 -0.2109 8.7053>, rot=<0.0355 -0.0009 0.9993 0.0155>}
 calf_r {parent=thigh_r, loc=<42.5284 0.0019 0.7297>, rot=<0.0006 -0.0161 -0.0391 -0.9991>}
 foot_r {parent=calf_r, loc=<46.7401 2.5014 0.592>, rot=<0.0427 -0.0238 0.0353 -0.9982>}
 ball_r {parent=foot_r, loc=<3.5658 -10.8066 0.1544>, rot=<0.0001 -0.0001 -0.7063 -0.7079>}
 calf_twist_02_r {parent=calf_r, loc=<19.1546 1.7943 0.2638>, rot=<0.0237 0.0095 -0.0035 -0.9997>}
 calf_twist_01_r {parent=calf_r, loc=<35.8483 1.9197 0.5794>, rot=<0.0237 0.0095 -0.0035 -0.9997>}
 dyn_ankle_back_MSTR_r {parent=calf_twist_01_r, loc=<-15.4189 8.1424 0>, rot=<1 -0.0095 0 0>}
 dyn_ankle_back_01_r {parent=dyn_ankle_back_MSTR_r, loc=<0 0 0>, rot=<0 0 0 -1>}
 dyn_ankle_in_MSTR_r {parent=calf_twist_01_r, loc=<-16.5942 0 -8.1117>, rot=<-0.7066 -0.0274 0.0274 -0.7066>}
 dyn_ankle_in_01_r {parent=dyn_ankle_in_MSTR_r, loc=<0 0 0>, rot=<0 0 0 1>}
 dyn_ankle_front_MSTR_r {parent=calf_twist_01_r, loc=<-17.025 -7.386 0>, rot=<0 0 0.0745 -0.9972>}
 dyn_ankle_front_01_r {parent=dyn_ankle_front_MSTR_r, loc=<0 0 0>, rot=<0 0 0 1>}
 dyn_ankle_out_MSTR_r {parent=calf_twist_01_r, loc=<-15.6895 0 7.416>, rot=<0.7071 0 0 -0.7071>}
 dyn_ankle_out_01_r {parent=dyn_ankle_out_MSTR_r, loc=<0 0 0>, rot=<0 0 0 1>}
 deform_knee_in_base_r {parent=calf_r, loc=<-0.8031 5.0231 -0.2457>, rot=<0 0 0 -1>}
 deform_knee_in_r {parent=deform_knee_in_base_r, loc=<0 0 0>, rot=<0 0 0 1>}
 deform_knee_out_base_r {parent=calf_r, loc=<0 -4 0>, rot=<0 0 0 -1>}
 deform_knee_out_r {parent=deform_knee_out_base_r, loc=<0 0 0>, rot=<0 0 0 1>}
 thigh_twist_01_r {parent=thigh_r, loc=<27.952 0 0>, rot=<0.0013 0 0 -1>}
 deform_groin_base_r {parent=thigh_r, loc=<-0.4801 -5.6776 0.2486>, rot=<0 0 0 -1>}
 deform_groin_r {parent=deform_groin_base_r, loc=<0 0 0>, rot=<0 0 0 1>}
 deform_glute_base_l {parent=pelvis, loc=<-7.3466 -9.0213 -9.1245>, rot=<0.0172 0.031 -0.4477 -0.8935>}
 deform_glute_l {parent=deform_glute_base_l, loc=<0 0 0>, rot=<0 0 0 1>}
 deform_glute_base_r {parent=pelvis, loc=<-7.3464 -9.0213 9.1245>, rot=<0.031 -0.0172 0.8935 -0.4477>}
 deform_glute_r {parent=deform_glute_base_r, loc=<0 0 0>, rot=<0 0 0 1>}
 extrta_pelvis {parent=pelvis, loc=<-11.3978 4.8972 0>, rot=<0 0 0 1>}
]
variables [dyn_crouch {type=double, default=0}]
