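# Physics asset for the jacket of the MarchTreatBay female medium player body cosmetic, modeling the garment as seven bone chains around the back, sides and front.
# Each chain runs from a small master capsule, joined to spine_01 with free rotation, through two capsule segments.
# Segment joints have tight twist and swing1 limits and wider swing2, widest at the front and narrowest at the sides.
# Pelvis, spine and thigh capsules supply body shapes, linked by free-rotation constraints.
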PhysicsAsset /BRCosmetics/Characters/Player/Female/Medium/Bodies/F_MED_MarchTreatBay/Meshes/F_MED_MarchTreatBay_Physics_Jacket
bodies {
 dyn_jacket_bk_01 [{shapeType=Capsule, length=12.9378, radius=0.8525, shapeName=dyn_jacket_bk_01_capsule, center=<5 0 0>, rotation=<rot 0 0 90>}]
 dyn_jacket_bk_02 [{shapeType=Capsule, length=9.8553, radius=2.2933, shapeName=dyn_jacket_bk_02_capsule, center=<3.2705 0 0>, rotation=<rot 0 0 90>}]
 dyn_jacket_bk_MSTR [{shapeType=Capsule, length=0.505, radius=0.505, shapeName=dyn_jacket_bk_MSTR_capsule, rotation=<rot 0 -90 0>}]
 dyn_jacket_l_01 [{shapeType=Capsule, length=12.644, radius=1.1463, shapeName=dyn_jacket_l_01_capsule, center=<5 0 0>, rotation=<rot 0 0 90>}]
 dyn_jacket_l_02 [{shapeType=Capsule, length=11.391, radius=1.4658, shapeName=dyn_jacket_l_02_capsule, center=<3.2705 0 0>, rotation=<rot 0 0 90>}]
 dyn_jacket_l_MSTR [{shapeType=Capsule, length=0.505, radius=0.505, shapeName=dyn_jacket_l_MSTR_capsule, rotation=<rot 0 -90 0>}]
 dyn_jacket_l_bk_01 [{shapeType=Capsule, length=12.644, radius=1.1463, shapeName=dyn_jacket_l_bk_01_capsule, center=<5 0 0>, rotation=<rot 0 0 90>}]
 dyn_jacket_l_bk_02 [{shapeType=Capsule, length=9.874, radius=1.8194, shapeName=dyn_jacket_l_bk_02_capsule, center=<3.2705 0 0>, rotation=<rot 0 0 90>}]
 dyn_jacket_l_bk_MSTR [{shapeType=Capsule, length=0.505, radius=0.505, shapeName=dyn_jacket_l_bk_MSTR_capsule, rotation=<rot 0 -90 0>}]
 dyn_jacket_l_fr_01 [{shapeType=Capsule, length=12.644, radius=1.1463, shapeName=dyn_jacket_l_fr_01_capsule, center=<5 0 0>, rotation=<rot 0 0 90>}]
 dyn_jacket_l_fr_02 [{shapeType=Capsule, length=8.7094, radius=1.2846, shapeName=dyn_jacket_l_fr_02_capsule, center=<2.5302 0 0>, rotation=<rot 0 0 90>}]
 dyn_jacket_l_fr_MSTR [{shapeType=Capsule, length=0.505, radius=0.505, shapeName=dyn_jacket_l_fr_MSTR_capsule, rotation=<rot 0 -90 0>}]
 dyn_jacket_r_01 [{shapeType=Capsule, length=12.644, radius=1.1463, shapeName=dyn_jacket_r_01_capsule, center=<5 0 0>, rotation=<rot 0 0 90>}]
 dyn_jacket_r_02 [{shapeType=Capsule, length=10.8426, radius=1.8796, shapeName=dyn_jacket_r_02_capsule, center=<3.2705 0 0>, rotation=<rot 0 0 90>}]
 dyn_jacket_r_MSTR [{shapeType=Capsule, length=0.505, radius=0.505, shapeName=dyn_jacket_r_MSTR_capsule, rotation=<rot 0 -90 0>}]
 dyn_jacket_r_bk_01 [{shapeType=Capsule, length=12.644, radius=1.1463, shapeName=dyn_jacket_r_bk_01_capsule, center=<5 0 0>, rotation=<rot 0 0 90>}]
 dyn_jacket_r_bk_02 [{shapeType=Capsule, length=10.3251, radius=1.8234, shapeName=dyn_jacket_r_bk_02_capsule, center=<3.2705 0 0>, rotation=<rot 0 0 90>}]
 dyn_jacket_r_bk_MSTR [{shapeType=Capsule, length=0.505, radius=0.505, shapeName=dyn_jacket_r_bk_MSTR_capsule, rotation=<rot 0 -90 0>}]
 dyn_jacket_r_fr_01 [{shapeType=Capsule, length=12.644, radius=1.1463, shapeName=dyn_jacket_r_fr_01_capsule, center=<5 0 0>, rotation=<rot 0 0 90>}]
 dyn_jacket_r_fr_02 [{shapeType=Capsule, length=8.6206, radius=1.311, shapeName=dyn_jacket_r_fr_02_capsule, center=<2.5302 0 0>, rotation=<rot 0 0 90>}]
 dyn_jacket_r_fr_MSTR [{shapeType=Capsule, length=0.505, radius=0.505, shapeName=dyn_jacket_r_fr_MSTR_capsule, rotation=<rot 0 -90 0>}]
 pelvis [{shapeType=Capsule, length=3.5551, radius=12.1282, shapeName=pelvis_capsule, center=<-3.0362 -0.9413 0>, rotation=<rot 0 -90 0>}]
 spine_01 [{shapeType=Capsule, length=4.4587, radius=8.4866, shapeName=spine_01_capsule}]
 thigh_l [{shapeType=Capsule, length=20.2036, radius=6.6028, shapeName=thigh_l_capsule, center=<-19.3188 0 0>, rotation=<rot 90 0 0>}]
 thigh_r [{shapeType=Capsule, length=20.2036, radius=6.6028, shapeName=thigh_r_capsule, center=<19.3188 0 0>, rotation=<rot 90 180 0>}]
}
constraints [
 {
  bone1Name=dyn_jacket_bk_MSTR
  bone2Name=spine_01
  swing1LimitDegrees=45
  swing1Motion=Free
  swing2LimitDegrees=45
  swing2Motion=Free
  twistLimitDegrees=45
  twistMotion=Free
 }
 {bone1Name=dyn_jacket_bk_01, bone2Name=dyn_jacket_bk_MSTR, swing1LimitDegrees=3, swing2LimitDegrees=30, twistLimitDegrees=4}
 {bone1Name=dyn_jacket_bk_02, bone2Name=dyn_jacket_bk_01, swing1LimitDegrees=3, swing2LimitDegrees=25, twistLimitDegrees=4}
 {
  bone1Name=dyn_jacket_l_MSTR
  bone2Name=spine_01
  swing1LimitDegrees=45
  swing1Motion=Free
  swing2LimitDegrees=45
  swing2Motion=Free
  twistLimitDegrees=45
  twistMotion=Free
 }
 {bone1Name=dyn_jacket_l_01, bone2Name=dyn_jacket_l_MSTR, swing1LimitDegrees=3, swing2LimitDegrees=15, twistLimitDegrees=4}
 {bone1Name=dyn_jacket_l_02, bone2Name=dyn_jacket_l_01, swing1LimitDegrees=4, swing2LimitDegrees=15, twistLimitDegrees=4}
 {
  bone1Name=dyn_jacket_l_bk_MSTR
  bone2Name=spine_01
  swing1LimitDegrees=45
  swing1Motion=Free
  swing2LimitDegrees=45
  swing2Motion=Free
  twistLimitDegrees=45
  twistMotion=Free
 }
 {bone1Name=dyn_jacket_l_bk_01, bone2Name=dyn_jacket_l_bk_MSTR, swing1LimitDegrees=3, swing2LimitDegrees=30, twistLimitDegrees=4}
 {bone1Name=dyn_jacket_l_bk_02, bone2Name=dyn_jacket_l_bk_01, swing1LimitDegrees=3, swing2LimitDegrees=25, twistLimitDegrees=4}
 {
  bone1Name=dyn_jacket_l_fr_MSTR
  bone2Name=spine_01
  swing1LimitDegrees=45
  swing1Motion=Free
  swing2LimitDegrees=45
  swing2Motion=Free
  twistLimitDegrees=45
  twistMotion=Free
 }
 {bone1Name=dyn_jacket_l_fr_01, bone2Name=dyn_jacket_l_fr_MSTR, swing1LimitDegrees=5, swing2LimitDegrees=45, twistLimitDegrees=4}
 {bone1Name=dyn_jacket_l_fr_02, bone2Name=dyn_jacket_l_fr_01, swing1LimitDegrees=5, swing2LimitDegrees=55, twistLimitDegrees=4}
 {
  bone1Name=dyn_jacket_r_MSTR
  bone2Name=spine_01
  swing1LimitDegrees=45
  swing1Motion=Free
  swing2LimitDegrees=45
  swing2Motion=Free
  twistLimitDegrees=45
  twistMotion=Free
 }
 {bone1Name=dyn_jacket_r_01, bone2Name=dyn_jacket_r_MSTR, swing1LimitDegrees=3, swing2LimitDegrees=15, twistLimitDegrees=4}
 {bone1Name=dyn_jacket_r_02, bone2Name=dyn_jacket_r_01, swing1LimitDegrees=4, swing2LimitDegrees=15, twistLimitDegrees=4}
 {
  bone1Name=dyn_jacket_r_bk_MSTR
  bone2Name=spine_01
  swing1LimitDegrees=45
  swing1Motion=Free
  swing2LimitDegrees=45
  swing2Motion=Free
  twistLimitDegrees=45
  twistMotion=Free
 }
 {bone1Name=dyn_jacket_r_bk_01, bone2Name=dyn_jacket_r_bk_MSTR, swing1LimitDegrees=3, swing2LimitDegrees=30, twistLimitDegrees=4}
 {bone1Name=dyn_jacket_r_bk_02, bone2Name=dyn_jacket_r_bk_01, swing1LimitDegrees=3, swing2LimitDegrees=25, twistLimitDegrees=4}
 {
  bone1Name=dyn_jacket_r_fr_MSTR
  bone2Name=spine_01
  swing1LimitDegrees=45
  swing1Motion=Free
  swing2LimitDegrees=45
  swing2Motion=Free
  twistLimitDegrees=45
  twistMotion=Free
 }
 {bone1Name=dyn_jacket_r_fr_01, bone2Name=dyn_jacket_r_fr_MSTR, swing1LimitDegrees=5, swing2LimitDegrees=45, twistLimitDegrees=4}
 {bone1Name=dyn_jacket_r_fr_02, bone2Name=dyn_jacket_r_fr_01, swing1LimitDegrees=5, swing2LimitDegrees=55, twistLimitDegrees=4}
 {
  bone1Name=spine_01
  bone2Name=pelvis
  swing1LimitDegrees=45
  swing1Motion=Free
  swing2LimitDegrees=45
  swing2Motion=Free
  twistLimitDegrees=45
  twistMotion=Free
 }
 {
  bone1Name=thigh_l
  bone2Name=pelvis
  swing1LimitDegrees=45
  swing1Motion=Free
  swing2LimitDegrees=45
  swing2Motion=Free
  twistLimitDegrees=45
  twistMotion=Free
 }
 {
  bone1Name=thigh_r
  bone2Name=pelvis
  swing1LimitDegrees=45
  swing1Motion=Free
  swing2LimitDegrees=45
  swing2Motion=Free
  twistLimitDegrees=45
  twistMotion=Free
 }
 {
  bone1Name=thigh_r
  bone2Name=thigh_l
  swing1LimitDegrees=45
  swing1Motion=Free
  swing2LimitDegrees=45
  swing2Motion=Free
  twistLimitDegrees=45
  twistMotion=Free
 }
]
