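# Physics asset for the MarchTreatBay female medium body cosmetic, setting up physics for its dangling lace bones.
# Top and bottom lace chains on each side start at master bones that link through extra_laces_MSTR to the pelvis with free angular motion.
# Capsule segments have tight twist limits, and the top chains also narrow one swing axis.
# Larger capsules on both thighs, the pelvis and an extra pelvis bone give the body's volume.
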PhysicsAsset /BRCosmetics/Characters/Player/Female/Medium/Bodies/F_MED_MarchTreatBay/Meshes/F_MED_MarchTreatBay_Physics_Laces
bodies {
 dyn_lace_MSTR_r [{shapeType=Capsule, length=0.505, radius=0.505, shapeName=dyn_lace_MSTR_r_capsule, rotation=<rot 0 -90 0>}]
 dyn_lace_bot_01_l [{shapeType=Capsule, length=2.5, radius=1.1, shapeName=dyn_lace_bot_01_l_capsule, center=<2.7127 -0.0392 0.0345>, rotation=<rot 90 0 0>}]
 dyn_lace_bot_01_r [{shapeType=Capsule, length=2.5, radius=1.1, shapeName=dyn_lace_bot_01_r_capsule, center=<2.7122 -0.0452 -0.037>, rotation=<rot 90 0 0>}]
 dyn_lace_bot_02_l [{shapeType=Capsule, length=3.0889, radius=0.7111, shapeName=dyn_lace_bot_02_l_capsule, center=<2 0 0>, rotation=<rot 90 0 0>}]
 dyn_lace_bot_02_r [{shapeType=Capsule, length=3.1399, radius=0.6601, shapeName=dyn_lace_bot_02_r_capsule, center=<2 0 0>, rotation=<rot 90 0 0>}]
 dyn_lace_bot_03_l [{shapeType=Capsule, length=3.0889, radius=0.7111, shapeName=dyn_lace_bot_03_l_capsule, center=<2 0 0>, rotation=<rot 90 0 0>}]
 dyn_lace_bot_03_r [{shapeType=Capsule, length=3.0317, radius=0.7683, shapeName=dyn_lace_bot_03_r_capsule, center=<2 0 0>, rotation=<rot 90 0 0>}]
 dyn_lace_bot_04_l [{shapeType=Capsule, length=5, radius=0.8, shapeName=dyn_lace_bot_04_l_capsule, center=<3 0 0>, rotation=<rot 90 0 0>}]
 dyn_lace_bot_04_r [{shapeType=Capsule, length=3, radius=0.8, shapeName=dyn_lace_bot_04_r_capsule, center=<2 0 0>, rotation=<rot 90 0 0>}]
 dyn_lace_bot_05_r [{shapeType=Capsule, length=3, radius=0.8, shapeName=dyn_lace_bot_05_r_capsule, center=<2 0 0>, rotation=<rot 90 0 0>}]
 dyn_lace_bot_MSTR_l [{shapeType=Capsule, length=0.505, radius=0.505, shapeName=dyn_lace_bot_MSTR_l_capsule, rotation=<rot 0 -90 0>}]
 dyn_lace_top_01_l [{
   shapeType=Capsule
   length=7.7918
   radius=0.5678
   shapeName=dyn_lace_top_01_l_capsule
   center=<3.4895 0.1034 -0.0175>
   rotation=<rot -9.9962 -87.4824 -1.623>
  }]
 dyn_lace_top_01_r [{
   shapeType=Capsule
   length=8.1837
   radius=0.5
   shapeName=dyn_lace_top_01_r_capsule
   center=<3.2082 -0.1222 -0.2493>
   rotation=<rot -0.0001 89.4639 11.0654>
  }]
 dyn_lace_top_02_l [{
   shapeType=Capsule
   length=6.1351
   radius=0.7099
   shapeName=dyn_lace_top_02_l_capsule
   center=<2.3861 0.2718 0.1648>
   rotation=<rot -13.8045 -90 -16.6932>
  }]
 dyn_lace_top_02_r [{shapeType=Capsule, length=5.9353, radius=0.7099, shapeName=dyn_lace_top_02_r_capsule, center=<2.1048 0 0>, rotation=<rot 0 -90 0>}]
 dyn_lace_top_MSTR_l [{shapeType=Capsule, length=0.505, radius=0.505, shapeName=dyn_lace_top_MSTR_l_capsule, rotation=<rot 0 -90 0>}]
 dyn_lace_top_MSTR_r [{shapeType=Capsule, length=0.505, radius=0.505, shapeName=dyn_lace_top_MSTR_r_capsule, rotation=<rot 0 -90 0>}]
 extra_laces_MSTR [{shapeType=Capsule, length=1, radius=1}]
 extrta_pelvis [{shapeType=Capsule, length=12.785, radius=13.9872, shapeName=extrta_pelvis_capsule, center=<6.6406 -9.6649 0>, rotation=<rot 0 -90 0>}]
 pelvis [{shapeType=Capsule, length=9.3751, radius=8.8327, shapeName=pelvis_capsule, rotation=<rot 0 -90 0>}]
 thigh_l [{shapeType=Capsule, length=28.7591, radius=10.8655, center=<-21 0 0>, rotation=<rot 90 -0.0272 0>}]
 thigh_r [{shapeType=Capsule, length=28.7591, radius=10.8655, center=<21 0 0>, rotation=<rot 90 179.9728 0>}]
}
constraints [
 {
  bone1Name=dyn_lace_MSTR_r
  bone2Name=extra_laces_MSTR
  swing1LimitDegrees=45
  swing1Motion=Free
  swing2LimitDegrees=45
  swing2Motion=Free
  twistLimitDegrees=45
  twistMotion=Free
 }
 {bone1Name=dyn_lace_bot_01_r, bone2Name=dyn_lace_MSTR_r, swing1LimitDegrees=45, swing2LimitDegrees=45, twistLimitDegrees=5}
 {bone1Name=dyn_lace_bot_02_r, bone2Name=dyn_lace_bot_01_r, swing1LimitDegrees=35, swing2LimitDegrees=35, twistLimitDegrees=5}
 {bone1Name=dyn_lace_bot_03_r, bone2Name=dyn_lace_bot_02_r, swing1LimitDegrees=35, swing2LimitDegrees=35, twistLimitDegrees=5}
 {bone1Name=dyn_lace_bot_04_r, bone2Name=dyn_lace_bot_03_r, swing1LimitDegrees=35, swing2LimitDegrees=35, twistLimitDegrees=5}
 {bone1Name=dyn_lace_bot_05_r, bone2Name=dyn_lace_bot_04_r, swing1LimitDegrees=35, swing2LimitDegrees=35, twistLimitDegrees=5}
 {
  bone1Name=dyn_lace_bot_MSTR_l
  bone2Name=extra_laces_MSTR
  swing1LimitDegrees=45
  swing1Motion=Free
  swing2LimitDegrees=45
  swing2Motion=Free
  twistLimitDegrees=45
  twistMotion=Free
 }
 {bone1Name=dyn_lace_bot_01_l, bone2Name=dyn_lace_bot_MSTR_l, swing1LimitDegrees=45, swing2LimitDegrees=45, twistLimitDegrees=5}
 {bone1Name=dyn_lace_bot_02_l, bone2Name=dyn_lace_bot_01_l, swing1LimitDegrees=35, swing2LimitDegrees=35, twistLimitDegrees=5}
 {bone1Name=dyn_lace_bot_03_l, bone2Name=dyn_lace_bot_02_l, swing1LimitDegrees=35, swing2LimitDegrees=35, twistLimitDegrees=5}
 {bone1Name=dyn_lace_bot_04_l, bone2Name=dyn_lace_bot_03_l, swing1LimitDegrees=35, swing2LimitDegrees=35, twistLimitDegrees=5}
 {
  bone1Name=dyn_lace_top_MSTR_l
  bone2Name=extra_laces_MSTR
  swing1LimitDegrees=45
  swing1Motion=Free
  swing2LimitDegrees=45
  swing2Motion=Free
  twistLimitDegrees=45
  twistMotion=Free
 }
 {bone1Name=dyn_lace_top_01_l, bone2Name=dyn_lace_top_MSTR_l, swing1LimitDegrees=50, swing2LimitDegrees=5, twistLimitDegrees=5}
 {bone1Name=dyn_lace_top_02_l, bone2Name=dyn_lace_top_01_l, swing1LimitDegrees=45, swing2LimitDegrees=10, twistLimitDegrees=5}
 {
  bone1Name=dyn_lace_top_MSTR_r
  bone2Name=extra_laces_MSTR
  swing1LimitDegrees=45
  swing1Motion=Free
  swing2LimitDegrees=45
  swing2Motion=Free
  twistLimitDegrees=45
  twistMotion=Free
 }
 {bone1Name=dyn_lace_top_01_r, bone2Name=dyn_lace_top_MSTR_r, swing1LimitDegrees=50, swing2LimitDegrees=5, twistLimitDegrees=5}
 {bone1Name=dyn_lace_top_02_r, bone2Name=dyn_lace_top_01_r, swing1LimitDegrees=45, swing2LimitDegrees=10, twistLimitDegrees=5}
 {
  bone1Name=extra_laces_MSTR
  bone2Name=pelvis
  swing1LimitDegrees=45
  swing1Motion=Free
  swing2LimitDegrees=45
  swing2Motion=Free
  twistLimitDegrees=45
  twistMotion=Free
 }
 {
  bone1Name=thigh_l
  bone2Name=pelvis
  swing1LimitDegrees=45
  swing1Motion=Free
  swing2LimitDegrees=45
  swing2Motion=Free
  twistLimitDegrees=45
  twistMotion=Free
 }
 {
  bone1Name=thigh_r
  bone2Name=pelvis
  swing1LimitDegrees=45
  swing1Motion=Free
  swing2LimitDegrees=45
  swing2Motion=Free
  twistLimitDegrees=45
  twistMotion=Free
 }
 {
  bone1Name=extrta_pelvis
  bone2Name=pelvis
  swing1LimitDegrees=45
  swing1Motion=Free
  swing2LimitDegrees=45
  swing2Motion=Free
  twistLimitDegrees=45
  twistMotion=Free
 }
]
