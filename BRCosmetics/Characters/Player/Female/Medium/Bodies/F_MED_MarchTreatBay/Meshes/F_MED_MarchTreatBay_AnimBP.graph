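# Post-process style AnimBP for the F_MED_MarchTreatBay cosmetic body, adding secondary motion to jacket, cuffs, collar and laces.
# Copied parent pose runs through a crouch-aware Control Rig, jacket RigidBody, per-bone AnimDynamics, then laces RigidBody.
# Lace bone offsets apply only in the front end.
# EventGraph selects sim-space settings, computes gravity override and feeds material WPO cloth simulation from calf_r, adding wind while skydiving.
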AnimBlueprint /BRCosmetics/Characters/Player/Female/Medium/Bodies/F_MED_MarchTreatBay/Meshes/F_MED_MarchTreatBay_AnimBP
target_skeleton=&/F_MED_MarchTreatBay_Skeleton
graphs [
 AnimGraph {nodes [
   n1 = Control Rig!(Source=Copy Pose From Mesh() {bUseAttachedParent=true}, Alpha, dyn_crouch:double=Get Dyn_Crouch() -> (Dyn_Crouch:double)) -> (Pose) {controlRigAssetReference {blueprintRigClass=&/F_MED_MarchTreatBay_CtrlRig.F_MED_MarchTreatBay_CtrlRig_C}}
   n2 = RigidBody!(ComponentPose=Local To Component(LocalPose=@n1), Alpha, OverrideWorldGravity:Vector="0,0,0") -> (Pose) {
    bClampLinearTranslationLimitToRefPose=true
    bOverrideWorldGravity=true
    overridePhysicsAsset=&/F_MED_MarchTreatBay_Physics_Jacket
    simulationSpace=BaseBoneSpace
    alphaBoolBlend {blendOption=Linear}
    baseBoneRef {boneName=spine_03}
    componentAppliedLinearAccClamp=<3000 3000 3000>
    simSpaceSettings {
     maxAngularAcceleration=20
     maxAngularVelocity=5
     maxLinearAcceleration=1200
     maxLinearVelocity=300
     velocityScaleZ=0.6
     worldAlpha=1
     externalLinearDragV=<1 1 1>
    }
   }
   n3 = Anim Dynamics - Bone: dyn_ankle_back_01_l!(
    ComponentPose=@n2
    Alpha=1
    GravityOverride:Vector="0,0,0"
    SimSpaceSettings:Anim Phys Sim Space Settings Structure=(SimSpaceAngularAlpha=0,MaxAngularVelocity=10000,MaxAngularAcceleration=10000,ExternalAngularVelocity=(X=0,Y=0,Z=0))
    ExternalForce="0,0,0"
   ) -> (Pose) {
    angularBiasOverride=1
    angularDampingOverride=1
    angularSpringConstant=1000
    bAngularSpring=true
    bGravityOverrideInSimSpace=true
    bOverrideAngularBias=true
    bOverrideAngularDamping=true
    bOverrideLinearDamping=true
    bUseGravityOverride=true
    gravityScale=0.35
    linearDampingOverride=1
    simulationSpace=RootRelative
    alphaBoolBlend {blendOption=Linear}
    alphaScaleBiasClamp {scale=0.3}
    boundBone {boneName=dyn_ankle_back_01_l}
    physicsBodyDefinitions [{
      sphereCollisionRadius=0
      boundBone {boneName=dyn_ankle_back_01_l}
      boxExtents=<10 20 20>
      constraintSetup {angularLimitsMax=<0 12 5>, angularLimitsMin=<0 -12 -22>, angularTarget=<1 0 0>}
      localJointOffset=<8 0 0>
     }]
   }
   n4 = Anim Dynamics - Bone: dyn_ankle_in_01_l!(
    ComponentPose=@n3
    Alpha=1
    GravityOverride:Vector="0,0,0"
    SimSpaceSettings:Anim Phys Sim Space Settings Structure=(SimSpaceAngularAlpha=0,MaxAngularVelocity=10000,MaxAngularAcceleration=10000,ExternalAngularVelocity=(X=0,Y=0,Z=0))
    ExternalForce="0,0,0"
   ) -> (Pose) {
    angularBiasOverride=1
    angularDampingOverride=1
    angularSpringConstant=1000
    bAngularSpring=true
    bGravityOverrideInSimSpace=true
    bOverrideAngularBias=true
    bOverrideAngularDamping=true
    bOverrideLinearDamping=true
    bUseGravityOverride=true
    gravityScale=0.35
    linearDampingOverride=1
    simulationSpace=RootRelative
    alphaBoolBlend {blendOption=Linear}
    alphaScaleBiasClamp {scale=0.3}
    boundBone {boneName=dyn_ankle_in_01_l}
    physicsBodyDefinitions [{
      sphereCollisionRadius=0
      boundBone {boneName=dyn_ankle_in_01_l}
      boxExtents=<10 20 20>
      constraintSetup {angularLimitsMax=<0 12 5>, angularLimitsMin=<0 -12 -22>, angularTarget=<1 0 0>}
      localJointOffset=<8 0 0>
     }]
   }
   n5 = Anim Dynamics - Bone: dyn_ankle_out_01_l!(
    ComponentPose=@n4
    Alpha=1
    GravityOverride:Vector="0,0,0"
    SimSpaceSettings:Anim Phys Sim Space Settings Structure=(SimSpaceAngularAlpha=0,MaxAngularVelocity=10000,MaxAngularAcceleration=10000,ExternalAngularVelocity=(X=0,Y=0,Z=0))
    ExternalForce="0,0,0"
   ) -> (Pose) {
    angularBiasOverride=1
    angularDampingOverride=1
    angularSpringConstant=1000
    bAngularSpring=true
    bGravityOverrideInSimSpace=true
    bOverrideAngularBias=true
    bOverrideAngularDamping=true
    bOverrideLinearDamping=true
    bUseGravityOverride=true
    gravityScale=0.35
    linearDampingOverride=1
    simulationSpace=RootRelative
    alphaBoolBlend {blendOption=Linear}
    alphaScaleBiasClamp {scale=0.3}
    boundBone {boneName=dyn_ankle_out_01_l}
    physicsBodyDefinitions [{
      sphereCollisionRadius=0
      boundBone {boneName=dyn_ankle_out_01_l}
      boxExtents=<10 20 20>
      constraintSetup {angularLimitsMax=<0 12 5>, angularLimitsMin=<0 -12 -22>, angularTarget=<1 0 0>}
      localJointOffset=<8 0 0>
     }]
   }
   n6 = Anim Dynamics - Bone: dyn_ankle_front_01_l!(
    ComponentPose=@n5
    Alpha=1
    GravityOverride:Vector="0,0,0"
    SimSpaceSettings:Anim Phys Sim Space Settings Structure=(SimSpaceAngularAlpha=0,MaxAngularVelocity=10000,MaxAngularAcceleration=10000,ExternalAngularVelocity=(X=0,Y=0,Z=0))
    ExternalForce="0,0,0"
   ) -> (Pose) {
    angularBiasOverride=1
    angularDampingOverride=1
    angularSpringConstant=1000
    bAngularSpring=true
    bGravityOverrideInSimSpace=true
    bOverrideAngularBias=true
    bOverrideAngularDamping=true
    bOverrideLinearDamping=true
    bUseGravityOverride=true
    gravityScale=0.35
    linearDampingOverride=1
    simulationSpace=RootRelative
    alphaBoolBlend {blendOption=Linear}
    alphaScaleBiasClamp {scale=0.3}
    boundBone {boneName=dyn_ankle_front_01_l}
    physicsBodyDefinitions [{
      sphereCollisionRadius=0
      boundBone {boneName=dyn_ankle_front_01_l}
      boxExtents=<10 20 20>
      constraintSetup {angularLimitsMax=<0 12 5>, angularLimitsMin=<0 -12 -22>, angularTarget=<1 0 0>}
      localJointOffset=<8 0 0>
     }]
   }
   n7 = Anim Dynamics - Bone: dyn_ankle_back_01_r!(
    ComponentPose=@n6
    Alpha=1
    GravityOverride:Vector="0,0,0"
    SimSpaceSettings:Anim Phys Sim Space Settings Structure=(SimSpaceAngularAlpha=0,MaxAngularVelocity=10000,MaxAngularAcceleration=10000,ExternalAngularVelocity=(X=0,Y=0,Z=0))
    ExternalForce="0,0,0"
   ) -> (Pose) {
    angularBiasOverride=1
    angularDampingOverride=1
    angularSpringConstant=1000
    bAngularSpring=true
    bGravityOverrideInSimSpace=true
    bOverrideAngularBias=true
    bOverrideAngularDamping=true
    bOverrideLinearDamping=true
    bUseGravityOverride=true
    gravityScale=0.35
    linearDampingOverride=1
    simulationSpace=RootRelative
    alphaBoolBlend {blendOption=Linear}
    alphaScaleBiasClamp {scale=0.3}
    boundBone {boneName=dyn_ankle_back_01_r}
    physicsBodyDefinitions [{
      sphereCollisionRadius=0
      boundBone {boneName=dyn_ankle_back_01_r}
      boxExtents=<10 20 20>
      constraintSetup {angularLimitsMax=<0 12 5>, angularLimitsMin=<0 -12 -22>, angularTarget=<1 0 0>}
      localJointOffset=<8 0 0>
     }]
   }
   n8 = Anim Dynamics - Bone: dyn_ankle_in_01_r!(
    ComponentPose=@n7
    Alpha=1
    GravityOverride:Vector="0,0,0"
    SimSpaceSettings:Anim Phys Sim Space Settings Structure=(SimSpaceAngularAlpha=0,MaxAngularVelocity=10000,MaxAngularAcceleration=10000,ExternalAngularVelocity=(X=0,Y=0,Z=0))
    ExternalForce="0,0,0"
   ) -> (Pose) {
    angularBiasOverride=1
    angularDampingOverride=1
    angularSpringConstant=1000
    bAngularSpring=true
    bGravityOverrideInSimSpace=true
    bOverrideAngularBias=true
    bOverrideAngularDamping=true
    bOverrideLinearDamping=true
    bUseGravityOverride=true
    gravityScale=0.35
    linearDampingOverride=1
    simulationSpace=RootRelative
    alphaBoolBlend {blendOption=Linear}
    alphaScaleBiasClamp {scale=0.3}
    boundBone {boneName=dyn_ankle_in_01_r}
    physicsBodyDefinitions [{
      sphereCollisionRadius=0
      boundBone {boneName=dyn_ankle_in_01_r}
      boxExtents=<10 20 20>
      constraintSetup {angularLimitsMax=<0 12 5>, angularLimitsMin=<0 -12 -22>, angularTarget=<1 0 0>}
      localJointOffset=<8 0 0>
     }]
   }
   n9 = Anim Dynamics - Bone: dyn_ankle_out_01_r!(
    ComponentPose=@n8
    Alpha=1
    GravityOverride:Vector="0,0,0"
    SimSpaceSettings:Anim Phys Sim Space Settings Structure=(SimSpaceAngularAlpha=0,MaxAngularVelocity=10000,MaxAngularAcceleration=10000,ExternalAngularVelocity=(X=0,Y=0,Z=0))
    ExternalForce="0,0,0"
   ) -> (Pose) {
    angularBiasOverride=1
    angularDampingOverride=1
    angularSpringConstant=1000
    bAngularSpring=true
    bGravityOverrideInSimSpace=true
    bOverrideAngularBias=true
    bOverrideAngularDamping=true
    bOverrideLinearDamping=true
    bUseGravityOverride=true
    gravityScale=0.35
    linearDampingOverride=1
    simulationSpace=RootRelative
    alphaBoolBlend {blendOption=Linear}
    alphaScaleBiasClamp {scale=0.3}
    boundBone {boneName=dyn_ankle_out_01_r}
    physicsBodyDefinitions [{
      sphereCollisionRadius=0
      boundBone {boneName=dyn_ankle_out_01_r}
      boxExtents=<10 20 20>
      constraintSetup {angularLimitsMax=<0 12 5>, angularLimitsMin=<0 -12 -22>, angularTarget=<1 0 0>}
      localJointOffset=<8 0 0>
     }]
   }
   n10 = Anim Dynamics - Bone: dyn_ankle_front_01_r!(
    ComponentPose=@n9
    Alpha=1
    GravityOverride:Vector="0,0,0"
    SimSpaceSettings:Anim Phys Sim Space Settings Structure=(SimSpaceAngularAlpha=0,MaxAngularVelocity=10000,MaxAngularAcceleration=10000,ExternalAngularVelocity=(X=0,Y=0,Z=0))
    ExternalForce="0,0,0"
   ) -> (Pose) {
    angularBiasOverride=1
    angularDampingOverride=1
    angularSpringConstant=1000
    bAngularSpring=true
    bGravityOverrideInSimSpace=true
    bOverrideAngularBias=true
    bOverrideAngularDamping=true
    bOverrideLinearDamping=true
    bUseGravityOverride=true
    gravityScale=0.35
    linearDampingOverride=1
    simulationSpace=RootRelative
    alphaBoolBlend {blendOption=Linear}
    alphaScaleBiasClamp {scale=0.3}
    boundBone {boneName=dyn_ankle_front_01_r}
    physicsBodyDefinitions [{
      sphereCollisionRadius=0
      boundBone {boneName=dyn_ankle_front_01_r}
      boxExtents=<10 20 20>
      constraintSetup {angularLimitsMax=<0 12 5>, angularLimitsMin=<0 -12 -22>, angularTarget=<1 0 0>}
      localJointOffset=<8 0 0>
     }]
   }
   n11 = Anim Dynamics - Bone: dyn_lowerarm_bot_01_l!(
    ComponentPose=@n10
    Alpha=1
    GravityOverride:Vector="0,0,0"
    SimSpaceSettings:Anim Phys Sim Space Settings Structure=(SimSpaceAngularAlpha=0,MaxAngularVelocity=10000,MaxAngularAcceleration=10000,ExternalAngularVelocity=(X=0,Y=0,Z=0))
    ExternalForce="0,0,0"
   ) -> (Pose) {
    angularBiasOverride=1
    angularDampingOverride=1
    angularSpringConstant=500
    bAngularSpring=true
    bGravityOverrideInSimSpace=true
    bOverrideAngularBias=true
    bOverrideAngularDamping=true
    bOverrideLinearDamping=true
    bUseGravityOverride=true
    gravityScale=0.35
    linearDampingOverride=1
    simulationSpace=RootRelative
    alphaBoolBlend {blendOption=Linear}
    alphaScaleBiasClamp {scale=0.3}
    boundBone {boneName=dyn_lowerarm_bot_01_l}
    physicsBodyDefinitions [{
      sphereCollisionRadius=0
      boundBone {boneName=dyn_lowerarm_bot_01_l}
      boxExtents=<10 20 20>
      constraintSetup {angularLimitsMax=<0 12 6>, angularLimitsMin=<0 -12 -22>, angularTarget=<1 0 0>}
      localJointOffset=<8 0 0>
     }]
   }
   n12 = Anim Dynamics - Bone: dyn_lowerarm_in_01_l!(
    ComponentPose=@n11
    Alpha=1
    GravityOverride:Vector="0,0,0"
    SimSpaceSettings:Anim Phys Sim Space Settings Structure=(SimSpaceAngularAlpha=0,MaxAngularVelocity=10000,MaxAngularAcceleration=10000,ExternalAngularVelocity=(X=0,Y=0,Z=0))
    ExternalForce="0,0,0"
   ) -> (Pose) {
    angularBiasOverride=1
    angularDampingOverride=1
    angularSpringConstant=500
    bAngularSpring=true
    bGravityOverrideInSimSpace=true
    bOverrideAngularBias=true
    bOverrideAngularDamping=true
    bOverrideLinearDamping=true
    bUseGravityOverride=true
    gravityScale=0.35
    linearDampingOverride=1
    simulationSpace=RootRelative
    alphaBoolBlend {blendOption=Linear}
    alphaScaleBiasClamp {scale=0.3}
    boundBone {boneName=dyn_lowerarm_in_01_l}
    physicsBodyDefinitions [{
      sphereCollisionRadius=0
      boundBone {boneName=dyn_lowerarm_in_01_l}
      boxExtents=<10 20 20>
      constraintSetup {angularLimitsMax=<0 12 6>, angularLimitsMin=<0 -12 -22>, angularTarget=<1 0 0>}
      localJointOffset=<8 0 0>
     }]
   }
   n13 = Anim Dynamics - Bone: dyn_lowerarm_out_01_l!(
    ComponentPose=@n12
    Alpha=1
    GravityOverride:Vector="0,0,0"
    SimSpaceSettings:Anim Phys Sim Space Settings Structure=(SimSpaceAngularAlpha=0,MaxAngularVelocity=10000,MaxAngularAcceleration=10000,ExternalAngularVelocity=(X=0,Y=0,Z=0))
    ExternalForce="0,0,0"
   ) -> (Pose) {
    angularBiasOverride=1
    angularDampingOverride=1
    angularSpringConstant=500
    bAngularSpring=true
    bGravityOverrideInSimSpace=true
    bOverrideAngularBias=true
    bOverrideAngularDamping=true
    bOverrideLinearDamping=true
    bUseGravityOverride=true
    gravityScale=0.35
    linearDampingOverride=1
    simulationSpace=RootRelative
    alphaBoolBlend {blendOption=Linear}
    alphaScaleBiasClamp {scale=0.3}
    boundBone {boneName=dyn_lowerarm_out_01_l}
    physicsBodyDefinitions [{
      sphereCollisionRadius=0
      boundBone {boneName=dyn_lowerarm_out_01_l}
      boxExtents=<10 20 20>
      constraintSetup {angularLimitsMax=<0 12 6>, angularLimitsMin=<0 -12 -22>, angularTarget=<1 0 0>}
      localJointOffset=<8 0 0>
     }]
   }
   n14 = Anim Dynamics - Bone: dyn_lowerarm_top_01_l!(
    ComponentPose=@n13
    Alpha=1
    GravityOverride:Vector="0,0,0"
    SimSpaceSettings:Anim Phys Sim Space Settings Structure=(SimSpaceAngularAlpha=0,MaxAngularVelocity=10000,MaxAngularAcceleration=10000,ExternalAngularVelocity=(X=0,Y=0,Z=0))
    ExternalForce="0,0,0"
   ) -> (Pose) {
    angularBiasOverride=1
    angularDampingOverride=1
    angularSpringConstant=500
    bAngularSpring=true
    bGravityOverrideInSimSpace=true
    bOverrideAngularBias=true
    bOverrideAngularDamping=true
    bOverrideLinearDamping=true
    bUseGravityOverride=true
    gravityScale=0.35
    linearDampingOverride=1
    simulationSpace=RootRelative
    alphaBoolBlend {blendOption=Linear}
    alphaScaleBiasClamp {scale=0.3}
    boundBone {boneName=dyn_lowerarm_top_01_l}
    physicsBodyDefinitions [{
      sphereCollisionRadius=0
      boundBone {boneName=dyn_lowerarm_top_01_l}
      boxExtents=<10 20 20>
      constraintSetup {angularLimitsMax=<0 12 6>, angularLimitsMin=<0 -12 -22>, angularTarget=<1 0 0>}
      localJointOffset=<8 0 0>
     }]
   }
   n15 = Anim Dynamics - Bone: dyn_lowerarm_bot_01_r!(
    ComponentPose=@n14
    Alpha=1
    GravityOverride:Vector="0,0,0"
    SimSpaceSettings:Anim Phys Sim Space Settings Structure=(SimSpaceAngularAlpha=0,MaxAngularVelocity=10000,MaxAngularAcceleration=10000,ExternalAngularVelocity=(X=0,Y=0,Z=0))
    ExternalForce="0,0,0"
   ) -> (Pose) {
    angularBiasOverride=1
    angularDampingOverride=1
    angularSpringConstant=500
    bAngularSpring=true
    bGravityOverrideInSimSpace=true
    bOverrideAngularBias=true
    bOverrideAngularDamping=true
    bOverrideLinearDamping=true
    bUseGravityOverride=true
    gravityScale=0.35
    linearDampingOverride=1
    simulationSpace=RootRelative
    alphaBoolBlend {blendOption=Linear}
    alphaScaleBiasClamp {scale=0.3}
    boundBone {boneName=dyn_lowerarm_bot_01_r}
    physicsBodyDefinitions [{
      sphereCollisionRadius=0
      boundBone {boneName=dyn_lowerarm_bot_01_r}
      boxExtents=<10 20 20>
      constraintSetup {angularLimitsMax=<0 12 6>, angularLimitsMin=<0 -12 -22>, angularTarget=<1 0 0>}
      localJointOffset=<8 0 0>
     }]
   }
   n16 = Anim Dynamics - Bone: dyn_lowerarm_in_01_r!(
    ComponentPose=@n15
    Alpha=1
    GravityOverride:Vector="0,0,0"
    SimSpaceSettings:Anim Phys Sim Space Settings Structure=(SimSpaceAngularAlpha=0,MaxAngularVelocity=10000,MaxAngularAcceleration=10000,ExternalAngularVelocity=(X=0,Y=0,Z=0))
    ExternalForce="0,0,0"
   ) -> (Pose) {
    angularBiasOverride=1
    angularDampingOverride=1
    angularSpringConstant=500
    bAngularSpring=true
    bGravityOverrideInSimSpace=true
    bOverrideAngularBias=true
    bOverrideAngularDamping=true
    bOverrideLinearDamping=true
    bUseGravityOverride=true
    gravityScale=0.35
    linearDampingOverride=1
    simulationSpace=RootRelative
    alphaBoolBlend {blendOption=Linear}
    alphaScaleBiasClamp {scale=0.3}
    boundBone {boneName=dyn_lowerarm_in_01_r}
    physicsBodyDefinitions [{
      sphereCollisionRadius=0
      boundBone {boneName=dyn_lowerarm_in_01_r}
      boxExtents=<10 20 20>
      constraintSetup {angularLimitsMax=<0 12 6>, angularLimitsMin=<0 -12 -22>, angularTarget=<1 0 0>}
      localJointOffset=<8 0 0>
     }]
   }
   n17 = Anim Dynamics - Bone: dyn_lowerarm_out_01_r!(
    ComponentPose=@n16
    Alpha=1
    GravityOverride:Vector="0,0,0"
    SimSpaceSettings:Anim Phys Sim Space Settings Structure=(SimSpaceAngularAlpha=0,MaxAngularVelocity=10000,MaxAngularAcceleration=10000,ExternalAngularVelocity=(X=0,Y=0,Z=0))
    ExternalForce="0,0,0"
   ) -> (Pose) {
    angularBiasOverride=1
    angularDampingOverride=1
    angularSpringConstant=500
    bAngularSpring=true
    bGravityOverrideInSimSpace=true
    bOverrideAngularBias=true
    bOverrideAngularDamping=true
    bOverrideLinearDamping=true
    bUseGravityOverride=true
    gravityScale=0.35
    linearDampingOverride=1
    simulationSpace=RootRelative
    alphaBoolBlend {blendOption=Linear}
    alphaScaleBiasClamp {scale=0.3}
    boundBone {boneName=dyn_lowerarm_out_01_r}
    physicsBodyDefinitions [{
      sphereCollisionRadius=0
      boundBone {boneName=dyn_lowerarm_out_01_r}
      boxExtents=<10 20 20>
      constraintSetup {angularLimitsMax=<0 12 6>, angularLimitsMin=<0 -12 -22>, angularTarget=<1 0 0>}
      localJointOffset=<8 0 0>
     }]
   }
   n18 = Anim Dynamics - Bone: dyn_lowerarm_top_01_r!(
    ComponentPose=@n17
    Alpha=1
    GravityOverride:Vector="0,0,0"
    SimSpaceSettings:Anim Phys Sim Space Settings Structure=(SimSpaceAngularAlpha=0,MaxAngularVelocity=10000,MaxAngularAcceleration=10000,ExternalAngularVelocity=(X=0,Y=0,Z=0))
    ExternalForce="0,0,0"
   ) -> (Pose) {
    angularBiasOverride=1
    angularDampingOverride=1
    angularSpringConstant=500
    bAngularSpring=true
    bGravityOverrideInSimSpace=true
    bOverrideAngularBias=true
    bOverrideAngularDamping=true
    bOverrideLinearDamping=true
    bUseGravityOverride=true
    gravityScale=0.35
    linearDampingOverride=1
    simulationSpace=RootRelative
    alphaBoolBlend {blendOption=Linear}
    alphaScaleBiasClamp {scale=0.3}
    boundBone {boneName=dyn_lowerarm_top_01_r}
    physicsBodyDefinitions [{
      sphereCollisionRadius=0
      boundBone {boneName=dyn_lowerarm_top_01_r}
      boxExtents=<10 20 20>
      constraintSetup {angularLimitsMax=<0 12 6>, angularLimitsMin=<0 -12 -22>, angularTarget=<1 0 0>}
      localJointOffset=<8 0 0>
     }]
   }
   n19 = Anim Dynamics - Bone: dyn_collar_l!(
    ComponentPose=@n18
    Alpha=1
    GravityOverride:Vector="0,0,-980"
    SimSpaceSettings:Anim Phys Sim Space Settings Structure=(SimSpaceAngularAlpha=0,MaxAngularVelocity=10000,MaxAngularAcceleration=10000,ExternalAngularVelocity=(X=0,Y=0,Z=0))
    ExternalForce="0,0,0"
   ) -> (Pose) {
    angularDampingOverride=0.9999
    angularSpringConstant=2000
    bAngularSpring=true
    bOverrideAngularDamping=true
    bUseGravityOverride=true
    bUsePlanarLimit=false
    gravityScale=-1
    simulationSpace=RootRelative
    alphaBoolBlend {blendOption=Linear}
    alphaScaleBiasClamp {scale=0.75}
    boundBone {boneName=dyn_collar_l}
    componentAppliedLinearAccClamp=<5000 5000 5000>
    componentLinearAccScale=<0.2 0.2 0.2>
    componentLinearVelScale=<0.2 0.2 0.2>
    gravityOverride {z=-980}
    physicsBodyDefinitions [{
      collisionType=CustomSphere
      sphereCollisionRadius=1
      boundBone {boneName=dyn_collar_l}
      boxExtents=<7.5 7.5 15>
      constraintSetup {angularLimitsMax=<0 0 20>, angularLimitsMin=<0 0 -20>, angularTarget=<1 0 0>}
      localJointOffset=<4 0 0>
     }]
   }
   n20 = Anim Dynamics - Bone: dyn_collar_r!(
    ComponentPose=@n19
    Alpha=1
    GravityOverride:Vector="0,0,-980"
    SimSpaceSettings:Anim Phys Sim Space Settings Structure=(SimSpaceAngularAlpha=0,MaxAngularVelocity=10000,MaxAngularAcceleration=10000,ExternalAngularVelocity=(X=0,Y=0,Z=0))
    ExternalForce="0,0,0"
   ) -> (Pose) {
    angularDampingOverride=0.9999
    angularSpringConstant=2500
    bAngularSpring=true
    bOverrideAngularDamping=true
    bUseGravityOverride=true
    bUsePlanarLimit=false
    gravityScale=-1
    simulationSpace=RootRelative
    alphaBoolBlend {blendOption=Linear}
    alphaScaleBiasClamp {scale=0.75}
    boundBone {boneName=dyn_collar_r}
    componentAppliedLinearAccClamp=<5000 5000 5000>
    componentLinearAccScale=<0.2 0.2 0.2>
    componentLinearVelScale=<0.2 0.2 0.2>
    gravityOverride {z=-980}
    physicsBodyDefinitions [{
      collisionType=CustomSphere
      sphereCollisionRadius=1
      boundBone {boneName=dyn_collar_r}
      boxExtents=<10 10 20>
      constraintSetup {angularLimitsMax=<0 0 20>, angularLimitsMin=<0 0 -20>, angularTarget=<1 0 0>}
      localJointOffset=<5 0 0>
     }]
   }
   n21 = RigidBody(ComponentPose=@n20, OverrideWorldGravity="0,0,0") {
    bClampLinearTranslationLimitToRefPose=true
    bOverrideWorldGravity=true
    cachedBoundsScale=1.4
    overridePhysicsAsset=&/F_MED_MarchTreatBay_Physics_Laces
    simulationSpace=BaseBoneSpace
    alphaBoolBlend {blendOption=Linear}
    baseBoneRef {boneName=spine_01}
    componentAppliedLinearAccClamp=<3000 3000 3000>
    simSpaceSettings {
     maxAngularAcceleration=20
     maxAngularVelocity=5
     maxLinearAcceleration=1200
     maxLinearVelocity=300
     velocityScaleZ=0.6
     worldAlpha=1
     externalLinearDragV=<1 1 1>
    }
   }
   n22 = Get FE rotation() -> (FE rotation:float)
   n23 = "Transform (Modify) Bone - Bone: dyn_lace_bot_MSTR_l"(ComponentPose=@n21, Alpha=@n22, Translation="0,0,0", Rotation="0,-5,0", Scale="1,1,1") {
    rotationMode=BMM_Additive
    rotationSpace=BCS_BoneSpace
    translationMode=BMM_Additive
    translationSpace=BCS_BoneSpace
    alphaBoolBlend {blendOption=Linear}
    alphaScaleBiasClamp {bInterpResult=true}
    boneToModify {boneName=dyn_lace_bot_MSTR_l}
    rotation {yaw=-5}
   }
   n24 = "Transform (Modify) Bone - Bone: dyn_lace_MSTR_r"(ComponentPose=@n23, Alpha=@n22, Translation="0,0,0", Rotation="0,0,0", Scale="1,1,1") {
    rotationMode=BMM_Additive
    rotationSpace=BCS_BoneSpace
    translationMode=BMM_Additive
    translationSpace=BCS_BoneSpace
    alphaBoolBlend {blendOption=Linear}
    alphaScaleBiasClamp {bInterpResult=true}
    boneToModify {boneName=dyn_lace_MSTR_r}
   }
   Output Pose(Result=Component To Local(ComponentPose=@n24)) {layerGroup=None}
   Reroute Node(InputPin:Wildcard) -> (OutputPin:Wildcard)
  ]}
 EventGraph {nodes [
   Set WPO Globals Function(
    Material=Get dynamic_mat_instance() -> (dynamic_mat_instance:Material Instance Dynamic Object Reference)
    Overlay Material=Get dynamic_overlay_mat() -> (dynamic_overlay_mat:Material Instance Dynamic Object Reference)
    Skeletal Mesh=Get mesh() -> (mesh:Skeletal Mesh Component Budgeted Object Reference)
    "Local Noise Transform  Location"="0, 0, 0"
    Local Noise Offset="0, 0, 0"
   )
   > Set WPO Regional Function!(
    execute
    Material:Material Instance Dynamic Object Reference=Get dynamic_mat_instance()
    Overlay Material:Material Instance Dynamic Object Reference=Get dynamic_overlay_mat()
    Skeletal Mesh:Skeletal Mesh Component Object Reference=Get mesh()
    Velocity Transform:Name=calf_r
    Velocity Interp Speed:double=Get Velocity Interp Speed_reg1() -> (Velocity Interp Speed_reg1:double)
    Collision Transform:Name=calf_r
    Collision Transform Axis:EAxis Enum
    Invert Collision Transform Axis:bool=false
    Delta Time X:double=Get DeltaTimeX() -> (DeltaTimeX:double)
    Previous Velocity Transform Location:"Vector (by ref)"=Get Previous Velocity Transform Location_reg1() -> (Previous Velocity Transform Location_reg1:Vector)
    Previous Velocity Transform Velocity:"Vector (by ref)"=Get Previous Velocity Transform Velocity_reg1() -> (Previous Velocity Transform Velocity_reg1:Vector)
    Region:int=1
   ) -> (then)
   n1 = Get dynamic_mat_instance()
   n2 = Get dynamic_overlay_mat()
   > SetScalarParameterValue(self=[@n1, @n2], ParameterName=ClothSim_WindStrength, Value=Get wind_strength() -> (wind_strength:float))
   > SetScalarParameterValue(
    self=[@n2, @n1]
    ParameterName=ClothSim_TurbulenceNoiseTileDivisor
    Value=Get TurbulenceNoiseTileDivisor() -> (TurbulenceNoiseTileDivisor:float)
   )
   n3 = Event BlueprintUpdateAnimation()
   > Set DeltaTimeX(DeltaTimeX:double=@n3.DeltaTimeX) -> (Output_Get:double)
   > n4 = Sequence!(execute) -> (then_0, then_1, then_2:Exec, then_3:Exec, then_4:Exec)
   > n5 = "Gravity Override State Function (Parameter Asset) (Nativized)"(
    CharacterPartAnimInstance=Self-Reference()
    GameTimeInSeconds=GetGameTimeInSeconds()
    Parameters=&/F_MED_MarchTreatBay_Parameters
   )
   > Set GravityOverride(GravityOverride=@n5)
   n6 = Branch(execute=@n4.then_2, Condition=Get bIsCrouching())
   > Set Dyn_Crouch(Dyn_Crouch:double=1) -> (Output_Get:double)
   Set Dyn_Crouch(execute=@n6.else, Dyn_Crouch=0)
   n7 = Branch(execute=@n4.then_4, Condition=Get bIsInFrontEnd())
   > Set FE rotation(FE rotation:float=1) -> (Output_Get:float)
   Set FE rotation(execute=@n7.else, FE rotation=0)
   n8 = NOT Boolean(A=Get bIsPlayingEmote())
   n9 = NOT Boolean(A=Get bIsSkyDiving())
   n10 = Branch(execute=@n4.then_1, Condition=AND Boolean(A=Get bIsInFrontEnd(), B=AND Boolean(A=@n9, B=@n8)))
   > Set SimSpaceSettings(SimSpaceSettings:Anim Phys Sim Space Settings Structure=Get SimSpaceSettingsFE() -> (SimSpaceSettingsFE:Anim Phys Sim Space Settings Structure)) -> (Output_Get:Anim Phys Sim Space Settings Structure)
   Set SimSpaceSettings(execute=@n10.else, SimSpaceSettings=Get SimSpaceSettingsPIE() -> (SimSpaceSettingsPIE:Anim Phys Sim Space Settings Structure))
   Event OnCharacterCustomizationLoaded()
   n11 = CheckForBackbling()
   > Set IsWearingBackpackOrCape(IsWearingBackpackOrCape:bool=OR Boolean!(A=@n11, B=@n11.bIsBackblingACape, C:bool=@n11.bIsBackblingLong) -> (ReturnValue)) -> (Output_Get:bool)
   Reroute Node(InputPin:Exec) -> (OutputPin:Exec)
   > n12 = Branch(Condition=Get bIsSkyDiving())
   > Set Velocity Interp Speed_reg1(Velocity Interp Speed_reg1:double=75) -> (Output_Get:double)
   > Set wind_strength(wind_strength:float=75) -> (Output_Get:float)
   > Set TurbulenceNoiseTileDivisor(TurbulenceNoiseTileDivisor:float=100) -> (Output_Get:float)
   Set Velocity Interp Speed_reg1(execute=@n12.else, Velocity Interp Speed_reg1=0)
   > Set wind_strength(wind_strength=0)
   > Set TurbulenceNoiseTileDivisor(TurbulenceNoiseTileDivisor=0)
   n13 = Set mesh(mesh:Skeletal Mesh Component Budgeted Object Reference=Get Array Item with EFort Custom Part Type(Array=Reroute Node(InputPin:Array of Skeletal Mesh Component Budgeted Object References) -> (OutputPin:Array of Skeletal Mesh Component Budgeted Object References))) -> (Output_Get:Skeletal Mesh Component Budgeted Object Reference)
   > n14 = Get Dynamic Material Instance(SkeletalMeshComponent=@n13)
   > Set dynamic_mat_instance(dynamic_mat_instance:Material Instance Dynamic Object Reference=@n14) -> (Output_Get:Material Instance Dynamic Object Reference)
   > n15 = GetOrCreateDynamicOverlayMaterial(MeshComponent=Get mesh(), MaterialName=None)
   > Set dynamic_overlay_mat(dynamic_overlay_mat:Material Instance Dynamic Object Reference=@n15) -> (Output_Get:Material Instance Dynamic Object Reference)
   Reroute Node(InputPin:Wildcard) -> (OutputPin:Wildcard)
  ]}
]
variables [
 DeltaTimeX {type=number}
 Dyn_Crouch {type=number}
 dynamic_mat_instance {type=MaterialInstanceDynamic}
 dynamic_overlay_mat {type=MaterialInstanceDynamic}
 FE rotation {type=number, default=1}
 GravityOverride {type=Vector}
 IsWearingBackpackOrCape {type=boolean}
 mesh {type=SkeletalMeshComponentBudgeted}
 Previous Velocity Transform Location_reg1 {type=Vector}
 Previous Velocity Transform Velocity_reg1 {type=Vector}
 SimSpaceSettings {
  type=AnimPhysSimSpaceSettings
  default {maxAngularAcceleration=100000, maxAngularVelocity=100000, simSpaceAngularAlpha=1, externalAngularVelocity=<100000 100000 100000>}
 }
 SimSpaceSettingsFE {
  type=AnimPhysSimSpaceSettings
  default {maxAngularAcceleration=50, maxAngularVelocity=3, simSpaceAngularAlpha=1, externalAngularVelocity=<0 0 0>}
 }
 SimSpaceSettingsPIE {
  type=AnimPhysSimSpaceSettings
  default {maxAngularAcceleration=20, maxAngularVelocity=10, simSpaceAngularAlpha=1, externalAngularVelocity=<0 0 2>}
 }
 TurbulenceNoiseTileDivisor {type=number}
 Velocity Interp Speed_reg1 {type=number, default=1}
 wind_strength {type=number}
]
