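# Post-process Control Rig for the F_MED_MarchTreatBay face accessory that offsets back hair bones as the head and arms move.
# Spherical pose readers track each upper arm against spine_05 and the head against neck_01.
# Each side takes the maximum of its arm and the shared head reading as the weight for self-relative rotation and translation offsets on dyn_Hair_bk_03 and 04.
# The head reader has debug drawing enabled.
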ControlRigBlueprint /BRCosmetics/Characters/Player/Female/Medium/Bodies/F_MED_MarchTreatBay/Meshes/Parts/F_MED_MarchTreatBay_FaceAcc_CtrlRig_Post
graphs [
 RigVMFunctionLibrary {links [], nodes []}
 RigVMModel {nodes [
   RigUnit_BeginExecution = &RigUnit_BeginExecution()
   n1 = &RigUnit_SphericalPoseReader(
    ActiveRegionScaleFactors=(PositiveWidth=0.5,NegativeWidth=1,PositiveHeight=1,NegativeHeight=1)
    ActiveRegionSize=0.2
    Debug=(bDrawDebug=false,bDraw2D=False,bDrawLocalAxes=False,DebugScale=25,DebugSegments=20,DebugThickness=0.25)
    DriverAxis=(X=1,Y=0,Z=0)
    DriverItem=(Type=Bone,Name="upperarm_l")
    FalloffRegionScaleFactors=(PositiveWidth=0.5,NegativeWidth=1,PositiveHeight=1,NegativeHeight=1)
    FalloffSize=0.35
    FlipHeightScaling=False
    FlipWidthScaling=False
    OptionalParentItem=(Type=Bone,Name="spine_05")
    RotationOffset=(X=0,Y=0,Z=0)
    ExecutePin=@RigUnit_BeginExecution.ExecutePin
   )
   > n2 = &RigUnit_SphericalPoseReader(
    ActiveRegionScaleFactors=(PositiveWidth=0.5,NegativeWidth=1,PositiveHeight=1,NegativeHeight=1)
    ActiveRegionSize=0.2
    Debug=(bDrawDebug=false,bDraw2D=False,bDrawLocalAxes=False,DebugScale=25,DebugSegments=20,DebugThickness=0.25)
    DriverAxis=(X=1,Y=0,Z=0)
    DriverItem=(Type=Bone,Name="upperarm_r")
    FalloffRegionScaleFactors=(PositiveWidth=0.5,NegativeWidth=1,PositiveHeight=1,NegativeHeight=1)
    FalloffSize=0.35
    FlipHeightScaling=False
    FlipWidthScaling=False
    OptionalParentItem=(Type=Bone,Name="spine_05")
    RotationOffset=(X=0,Y=0,Z=0)
   )
   > n3 = &RigUnit_SphericalPoseReader(
    ActiveRegionScaleFactors=(PositiveWidth=1,NegativeWidth=2,PositiveHeight=2,NegativeHeight=2)
    ActiveRegionSize=0.12
    Debug=(bDrawDebug=true,DebugScale=25,DebugSegments=20,DebugThickness=0.25)
    DriverAxis=(X=0,Y=-1,Z=0)
    DriverItem=(Type=Bone,Name="head")
    FalloffRegionScaleFactors=(PositiveWidth=1,NegativeWidth=3,PositiveHeight=4,NegativeHeight=4)
    FalloffSize=0.3
    FlipWidthScaling=false
    OptionalParentItem=(Type=Bone,Name="neck_01")
    RotationOffset=(X=-90,Y=0,Z=-55)
   )
   Maximum = &RigVMFunction_MathDoubleMax(A=@n3.OutputParam, B=@n1.OutputParam)
   > /Script/ControlRig.RigUnit_SetRelativeTransformForItem(
    Child=(Type=Bone,Name="dyn_Hair_bk_03_l")
    Parent=(Type=Bone,Name="dyn_Hair_bk_03_l")
    Value=(Rotation=(X=0.2588,Y=0,Z=0,W=0.9659),Translation=(X=-1,Y=0,Z=1.5),Scale3D=(X=1,Y=1,Z=1))
    bParentInitial=false
    bPropagateToChildren=true
    Weight=@Maximum.Result
   ) {
    is=struct RigVMUnitNode
    template="Set Relative Transform::Execute(in Child,in Parent,in bParentInitial,in Value,in Weight,in bPropagateToChildren)"
   }
   > &RigUnit_SetRelativeTransformForItem(
    Child=(Type=Bone,Name="dyn_Hair_bk_04_l")
    Parent=(Type=Bone,Name="dyn_Hair_bk_04_l")
    Value=(Rotation=(X=0.1736,Y=0,Z=0,W=0.9848),Translation=(X=-1,Y=0,Z=1.5),Scale3D=(X=1,Y=1,Z=1))
    bParentInitial=false
    bPropagateToChildren=true
    Weight=@Maximum.Result
   ) {template="Set Relative Transform::Execute(in Child,in Parent,in bParentInitial,in Value,in Weight,in bPropagateToChildren)"}
   n4 = &RigVMFunction_MathDoubleMax(A=@n3.OutputParam, B=@n2.OutputParam)
   > &RigUnit_SetRelativeTransformForItem(
    Child=(Type=Bone,Name="dyn_Hair_bk_03_r")
    Parent=(Type=Bone,Name="dyn_Hair_bk_03_r")
    Value=(Rotation=(X=-0.2588,Y=0,Z=0,W=0.9659),Translation=(X=-1,Y=0,Z=1.5),Scale3D=(X=1,Y=1,Z=1))
    bParentInitial=false
    bPropagateToChildren=true
    Weight=@n4.Result
   ) {template="Set Relative Transform::Execute(in Child,in Parent,in bParentInitial,in Value,in Weight,in bPropagateToChildren)"}
   > &RigUnit_SetRelativeTransformForItem(
    Child=(Type=Bone,Name="dyn_Hair_bk_04_r")
    Parent=(Type=Bone,Name="dyn_Hair_bk_04_r")
    Value=(Rotation=(X=-0.1736,Y=0,Z=0,W=0.9848),Translation=(X=-1,Y=0,Z=1.5),Scale3D=(X=1,Y=1,Z=1))
    bParentInitial=false
    bPropagateToChildren=true
    Weight=@n4.Result
   ) {template="Set Relative Transform::Execute(in Child,in Parent,in bParentInitial,in Value,in Weight,in bPropagateToChildren)"}
   RigVMRerouteNode(Value=())
  ]}
]
hierarchy [
 root {loc=<0 0 0>, rot=<0 0 0 1>}
 pelvis {parent=root, loc=<0 -0.5301 98.0571>, rot=<0.0127 0.707 -0.0127 -0.707>}
 spine_01 {parent=pelvis, loc=<6.1208 -0.9328 0>, rot=<0 0 0.034 -0.9994>}
 spine_02 {parent=spine_01, loc=<6.5997 -0.2095 0>, rot=<0 0 0.0238 -0.9997>}
 spine_03 {parent=spine_02, loc=<6.4914 -0.1237 0>, rot=<0 0 0.0149 -0.9999>}
 spine_04 {parent=spine_03, loc=<6.9598 0.0794 0>, rot=<0 0 -0.0132 -0.9999>}
 spine_05 {parent=spine_04, loc=<6.9987 0 0>, rot=<0 0 -0.0015 -1>}
 neck_01 {parent=spine_05, loc=<11.3138 0.1529 0>, rot=<0 0 -0.2003 -0.9797>}
 neck_02 {parent=neck_01, loc=<4.5627 0 0>, rot=<0 0 0.0504 -0.9987>}
 head {parent=neck_02, loc=<5.3914 0 0>, rot=<0 0 0.115 -0.9934>}
 dyn_hair_ear_MSTR_r {parent=head, loc=<15.3706 4.0834 0.4673>, rot=<-0.0874 -0.9073 -0.2367 -0.3365>}
 dyn_hair_ear_01_r {parent=dyn_hair_ear_MSTR_r, loc=<0 0 0>, rot=<0 0 0 1>}
 dyn_hair_ear_MSTR_l {parent=head, loc=<15.3705 4.0834 -0.4673>, rot=<0.9073 -0.0874 -0.3365 0.2367>}
 dyn_hair_ear_01_l {parent=dyn_hair_ear_MSTR_l, loc=<0 0 0>, rot=<0 0 0 1>}
 dyn_Hair_bk_MSTR {parent=head, loc=<9.8456 -3.389 0>, rot=<0.5287 -0.4695 0.4695 -0.5287>}
 dyn_Hair_bk_01_ct {parent=dyn_Hair_bk_MSTR, loc=<0 0 0>, rot=<0 -0.5708 0 -0.8211>}
 dyn_Hair_bk_02_ct {parent=dyn_Hair_bk_01_ct, loc=<7.522 0 0>, rot=<0 -0.0688 0 -0.9976>}
 dyn_Hair_bk_03_ct {parent=dyn_Hair_bk_02_ct, loc=<7.6539 0 0>, rot=<0 0.0957 0 -0.9954>}
 dyn_Hair_bk_04_ct {parent=dyn_Hair_bk_03_ct, loc=<6.7403 0 0>, rot=<0 -0.0481 0 -0.9988>}
 dyn_Hair_bk_05_ct {parent=dyn_Hair_bk_04_ct, loc=<8.3629 0 0.0065>, rot=<0 -0.1004 0 -0.9949>}
 dyn_Hair_bk_06_ct {parent=dyn_Hair_bk_05_ct, loc=<7.522 0 0>, rot=<0 -0.1034 0 -0.9946>}
 dyn_Hair_bk_07_01_ct {parent=dyn_Hair_bk_06_ct, loc=<4.7083 3.6181 -0.0281>, rot=<0.0351 -0.1163 -0.0041 -0.9926>}
 dyn_Hair_bk_07_02_ct {parent=dyn_Hair_bk_06_ct, loc=<7.1834 -1.1654 0.4508>, rot=<-0.0099 -0.1589 -0.0547 -0.9857>}
 dyn_Hair_bk_06_r {parent=dyn_Hair_bk_06_ct, loc=<-0.0723 -6.406 -0.7431>, rot=<0 0.0053 0 -1>}
 dyn_Hair_bk_07_01_r {parent=dyn_Hair_bk_06_r, loc=<9.9174 0.7755 -0.5767>, rot=<-0.2436 -0.1814 -0.1563 -0.9399>}
 dyn_Hair_bk_07_02_r {parent=dyn_Hair_bk_06_r, loc=<10.5748 -1.7357 -1.9954>, rot=<-0.2625 -0.1948 -0.0965 -0.9401>}
 dyn_Hair_bk_07_03_r {parent=dyn_Hair_bk_06_r, loc=<12.2078 -4.0763 -4.4322>, rot=<-0.2483 -0.2111 -0.0127 -0.9453>}
 dyn_Hair_bk_07_04_r {parent=dyn_Hair_bk_06_r, loc=<12.2785 -5.4701 -5.2907>, rot=<-0.2284 -0.2507 0.0825 -0.9371>}
 dyn_Hair_bk_06_l {parent=dyn_Hair_bk_06_ct, loc=<-0.0723 6.4062 -0.7431>, rot=<0 0.0053 0 -1>}
 dyn_Hair_bk_07_01_l {parent=dyn_Hair_bk_06_l, loc=<5.7169 6.873 -4.431>, rot=<0.2967 -0.1498 0.0937 -0.9385>}
 dyn_Hair_bk_07_02_l {parent=dyn_Hair_bk_06_l, loc=<5.3987 3.0933 -1.9238>, rot=<0.2374 -0.1594 0.0651 -0.956>}
 dyn_Hair_bk_05_r {parent=dyn_Hair_bk_05_ct, loc=<-0.2201 -6.406 -0.4832>, rot=<0 -0.0152 0 -0.9999>}
 dyn_Hair_bk_05_l {parent=dyn_Hair_bk_05_ct, loc=<-0.2201 6.4062 -0.4832>, rot=<0 -0.0152 0 -0.9999>}
 dyn_Hair_bk_04_l {parent=dyn_Hair_bk_04_ct, loc=<-0.707 6.4062 -0.3537>, rot=<0 0.1189 0 -0.9929>}
 dyn_Hair_bk_04_r {parent=dyn_Hair_bk_04_ct, loc=<-0.707 -6.406 -0.3537>, rot=<0 0.1189 0 -0.9929>}
 dyn_Hair_bk_03_r {parent=dyn_Hair_bk_03_ct, loc=<-0.3941 -6.406 -0.5243>, rot=<0 0.0188 0 -0.9998>}
 dyn_Hair_bk_03_l {parent=dyn_Hair_bk_03_ct, loc=<-0.3941 6.4062 -0.5243>, rot=<0 0.0188 0 -0.9998>}
 dyn_R_Hairstrand_MSTR {parent=head, loc=<10.6145 3.0927 6.3964>, rot=<0.2753 0.0486 0.9589 -0.0488>}
 dyn_R_Hairstrand_01_jnt {parent=dyn_R_Hairstrand_MSTR, loc=<0 0 0>, rot=<0 -0.0333 0 -0.9994>}
 dyn_R_Hairstrand_02_jnt {parent=dyn_R_Hairstrand_01_jnt, loc=<5.824 -0.3852 -0.5827>, rot=<-0.1214 -0.2279 -0.0026 -0.9661>}
 dyn_R_Hairstrand_03_jnt {parent=dyn_R_Hairstrand_02_jnt, loc=<4.1567 -0.4878 0.5521>, rot=<-0.0463 0.0149 0.0807 -0.9956>}
 dyn_R_Hairstrand_03_Tip_04_jnt {parent=dyn_R_Hairstrand_03_jnt, loc=<4.6279 -3.1555 -0.3955>, rot=<-0.125 -0.123 0.3185 -0.9316>}
 dyn_R_Hairstrand_03_Tip_03_jnt {parent=dyn_R_Hairstrand_03_jnt, loc=<6.3188 -2.4609 -0.3669>, rot=<-0.0716 -0.1792 0.2849 -0.9389>}
 dyn_R_Hairstrand_03_Tip_02_jnt {parent=dyn_R_Hairstrand_03_jnt, loc=<8.6129 -1.5002 -0.3491>, rot=<0.0539 -0.1115 0.2562 -0.9587>}
 dyn_R_Hairstrand_01_Tip_01_jnt {parent=dyn_R_Hairstrand_03_jnt, loc=<9.4352 0.5929 -0.4397>, rot=<0.0511 -0.0445 0.2306 -0.9707>}
 dyn_R_Hairstrand_01_Tip_02_jnt {parent=dyn_R_Hairstrand_01_Tip_01_jnt, loc=<4.6707 -0.0002 0.0001>, rot=<-0.0279 -0.1123 0.0536 -0.9918>}
 dyn_R_Hairstrand_03_Tip_05_jnt {parent=dyn_R_Hairstrand_03_jnt, loc=<1.595 -2.9868 -0.7488>, rot=<0.0085 -0.1669 0.2978 -0.9399>}
 dyn_hair_l_01 {parent=head, loc=<8.432 9.0677 -4.72>, rot=<-0.0354 0.2814 0.9582 -0.037>}
 dyn_hair_l_02 {parent=head, loc=<5.9631 8.237 -6.2942>, rot=<-0.0282 0.2797 0.9476 -0.1516>}
 dyn_hair_f_02 {parent=head, loc=<5.9116 10.5319 2.9184>, rot=<0.2856 -0.1409 0.9304 -0.1817>}
 dyn_hair_f_01 {parent=head, loc=<7.0212 10.5366 -0.8523>, rot=<0.3215 0.0925 0.9376 -0.0945>}
 dyn_L_Hairstrand_MSTR {parent=head, loc=<10.6146 3.0927 -6.3964>, rot=<0.0488 -0.9589 -0.0486 -0.2753>}
 dyn_L_Hairstrand_01_jnt {parent=dyn_L_Hairstrand_MSTR, loc=<0 0 0>, rot=<0 -0.0333 0 -0.9994>}
 dyn_L_Hairstrand_02_jnt {parent=dyn_L_Hairstrand_01_jnt, loc=<5.0067 0.9263 -0.2077>, rot=<0.9684 0.0656 -0.2235 0.0888>}
 dyn_L_Hairstrand_03_jnt1 {parent=dyn_L_Hairstrand_02_jnt, loc=<3.888 0.0071 -0.0049>, rot=<0.2302 0.0013 0.0794 -0.9699>}
 dyn_L_Hairstrand_01_Tip_01_jnt {parent=dyn_L_Hairstrand_03_jnt1, loc=<6.4659 2.6344 0.6365>, rot=<-0.1536 0.0159 0.1195 -0.9807>}
 dyn_L_Hairstrand_01_Tip_02_jnt {parent=dyn_L_Hairstrand_01_Tip_01_jnt, loc=<6.2844 0 0>, rot=<0.016 0.0987 0.0992 -0.99>}
 dyn_L_Hairstrand_01_Tip_03_jnt {parent=dyn_L_Hairstrand_01_Tip_02_jnt, loc=<2.9907 -0.0952 -0.068>, rot=<-0.0233 0.06 0.0306 -0.9975>}
 dyn_L_Hairstrand_03_Tip_02_jnt {parent=dyn_L_Hairstrand_03_jnt1, loc=<6.3124 -0.0264 0.1443>, rot=<-0.1697 0.0527 0.2616 -0.9487>}
 dyn_L_Hairstrand_03_Tip_03_jnt {parent=dyn_L_Hairstrand_03_jnt1, loc=<2.2865 -1.8121 -0.7153>, rot=<-0.209 0.065 0.1956 -0.9559>}
 dyn_L_Hairstrand_03_Tip_04_jnt {parent=dyn_L_Hairstrand_02_jnt, loc=<4.0689 -3.2763 0.719>, rot=<0.12 0.2327 0.1986 -0.9445>}
 clavicle_l {parent=spine_05, loc=<5.4831 -0.1053 -3.2946>, rot=<0.0465 -0.7289 -0.009 -0.6829>}
 upperarm_l {parent=clavicle_l, loc=<9.8658 -1.2136 -0.4105>, rot=<0.0164 -0.3245 -0.0269 -0.9454>}
 clavicle_r {parent=spine_05, loc=<5.483 -0.1053 3.2946>, rot=<0.7289 0.0465 -0.6829 0.009>}
 upperarm_r {parent=clavicle_r, loc=<-9.8658 1.2136 0.4105>, rot=<0.0164 -0.3245 -0.0269 -0.9454>}
]
