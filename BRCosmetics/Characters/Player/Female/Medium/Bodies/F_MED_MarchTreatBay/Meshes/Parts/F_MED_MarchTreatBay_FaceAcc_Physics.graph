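# Physics asset for the face accessory mesh of the F_MED_MarchTreatBay female player cosmetic, providing physics for long back hair.
# A six-capsule center chain hangs from a zero-size master body, with angular limits generally widening down it, then fans into eight smaller tip capsules.
# Head, neck, spine, clavicle and upper-arm capsules form the body volume around the hair.
# Constraints among neck, upper spine, clavicles and arms leave swing and twist free.
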PhysicsAsset /BRCosmetics/Characters/Player/Female/Medium/Bodies/F_MED_MarchTreatBay/Meshes/Parts/F_MED_MarchTreatBay_FaceAcc_Physics
bodies {
 clavicle_l [{
   shapeType=Capsule
   length=13.2727
   radius=7.2273
   shapeName=clavicle_l_capsule
   center=<3.5279 -1.1467 -0.8132>
   rotation=<rot 79.9999 -0.0071 -0.0067>
  }]
 clavicle_r [{
   shapeType=Capsule
   length=13.2727
   radius=7.2273
   shapeName=clavicle_r_capsule
   center=<-3.5279 1.1467 0.8132>
   rotation=<rot 79.9999 -0.0082 4.9918>
  }]
 dyn_Hair_bk_01_ct [{shapeType=Capsule, length=14.8272, radius=4.7702, shapeName=dyn_Hair_bk_01_ct_capsule, center=<7 0 2.0966>, rotation=<rot 0 0 90>}]
 dyn_Hair_bk_02_ct [{shapeType=Capsule, length=20.0459, radius=6.6466, shapeName=dyn_Hair_bk_02_ct_capsule, center=<7.0533 0 1.3358>, rotation=<rot 0 0 90>}]
 dyn_Hair_bk_03_ct [{shapeType=Capsule, length=23.1305, radius=6.3954, shapeName=dyn_Hair_bk_03_ct_capsule, center=<5.1096 0 1.4668>, rotation=<rot -0.0004 0 90>}]
 dyn_Hair_bk_04_ct [{shapeType=Capsule, length=24.8752, radius=5.624, shapeName=dyn_Hair_bk_04_ct_capsule, center=<6.4852 0 0.3543>, rotation=<rot 0 0 90>}]
 dyn_Hair_bk_05_ct [{shapeType=Capsule, length=22.415, radius=7.5606, shapeName=dyn_Hair_bk_05_ct_capsule, center=<4.7478 -0.0414 1>, rotation=<rot 0 0 90>}]
 dyn_Hair_bk_06_ct [{shapeType=Capsule, length=21.0863, radius=8.015, shapeName=dyn_Hair_bk_06_ct_capsule, center=<7.4117 0 -1.4916>, rotation=<rot 0 0 90>}]
 dyn_Hair_bk_07_01_ct [{
   shapeType=Capsule
   length=9.383
   radius=2.6779
   shapeName=dyn_Hair_bk_07_01_ct_capsule
   center=<8.6378 0.4786 0.133>
   rotation=<rot 0 96.3436 -98.4709>
  }]
 dyn_Hair_bk_07_01_l [{
   shapeType=Capsule
   length=7.0744
   radius=3.7218
   shapeName=dyn_Hair_bk_07_01_l_capsule
   center=<8.4814 0.3232 0.3416>
   rotation=<rot 0.7326 92.7152 -87.2647>
  }]
 dyn_Hair_bk_07_01_r [{
   shapeType=Capsule
   length=6.8083
   radius=2.2865
   shapeName=dyn_Hair_bk_07_01_r_capsule
   center=<5.7717 0.2033 1.0523>
   rotation=<rot -1.7687 87.7664 -100.0868>
  }]
 dyn_Hair_bk_07_02_ct [{
   shapeType=Capsule
   length=10.7287
   radius=2.962
   shapeName=dyn_Hair_bk_07_02_ct_capsule
   center=<9.7206 0.2936 0.2823>
   rotation=<rot 0.0001 92.0863 -98.5977>
  }]
 dyn_Hair_bk_07_02_l [{
   shapeType=Capsule
   length=9.6933
   radius=4.0207
   shapeName=dyn_Hair_bk_07_02_l_capsule
   center=<9.6331 -0.3831 0.9593>
   rotation=<rot -0.0001 82.6576 -96.1255>
  }]
 dyn_Hair_bk_07_02_r [{
   shapeType=Capsule
   length=9.9701
   radius=2.1887
   shapeName=dyn_Hair_bk_07_02_r_capsule
   center=<6.154 -0.021 1.0411>
   rotation=<rot -1.674 88.52 -99.5417>
  }]
 dyn_Hair_bk_07_03_r [{
   shapeType=Capsule
   length=5.5892
   radius=1.9591
   shapeName=dyn_Hair_bk_07_03_r_capsule
   center=<4.9126 0.2114 0.1579>
   rotation=<rot 0 84.1934 -91.3239>
  }]
 dyn_Hair_bk_07_04_r [{
   shapeType=Capsule
   length=4.9969
   radius=1.5162
   shapeName=dyn_Hair_bk_07_04_r_capsule
   center=<4.2226 -0.2663 0.4983>
   rotation=<rot 0 79.4181 -85.2037>
  }]
 dyn_Hair_bk_MSTR [{shapeType=Capsule, length=0, radius=0, shapeName=dyn_Hair_bk_MSTR_capsule}]
 head [{shapeType=Capsule, length=4.25, radius=4.8836, shapeName=head_capsule, center=<0.2441 1.1813 1.145>, rotation=<rot 1.1693 95.42 1.1975>}]
 neck_01 [{shapeType=Capsule, length=6.6065, radius=4.6703, shapeName=neck_01_capsule, rotation=<rot 90 0 0>}]
 spine_01 [{shapeType=Capsule, length=15.8574, radius=11.9281, shapeName=spine_01_capsule, center=<2.7203 0.2404 0>, rotation=<rot 0 -175.0002 -180>}]
 spine_03 [{shapeType=Capsule, length=18.3475, radius=9.9419, shapeName=spine_04_capsule, center=<-0.8771 0 0>}]
 spine_04 [{shapeType=Capsule, length=19.25, radius=9.75, shapeName=spine_04_capsule, rotation=<rot 0 -90 0>}]
 spine_05 [{shapeType=Capsule, length=20.6889, radius=8.6847, shapeName=spine_05_capsule, center=<-0.1841 0.1134 0>}]
 upperarm_l [{shapeType=Capsule, length=19.2932, radius=6.5605, shapeName=upperarm_l_capsule, center=<14.4228 -1.4841 -0.0222>, rotation=<rot 90 0 0>}]
 upperarm_r [{shapeType=Capsule, length=19.2932, radius=6.5605, shapeName=upperarm_r_capsule, center=<-14.4228 1.4841 0.0222>, rotation=<rot 90 180 0>}]
}
constraints [
 {
  bone1Name=clavicle_l
  bone2Name=spine_05
  swing1LimitDegrees=45
  swing1Motion=Free
  swing2LimitDegrees=45
  swing2Motion=Free
  twistLimitDegrees=45
  twistMotion=Free
 }
 {
  bone1Name=clavicle_r
  bone2Name=spine_05
  swing1LimitDegrees=45
  swing1Motion=Free
  swing2LimitDegrees=45
  swing2Motion=Free
  twistLimitDegrees=45
  twistMotion=Free
 }
 {
  bone1Name=upperarm_l
  bone2Name=clavicle_l
  swing1LimitDegrees=45
  swing1Motion=Free
  swing2LimitDegrees=45
  swing2Motion=Free
  twistLimitDegrees=45
  twistMotion=Free
 }
 {
  bone1Name=upperarm_r
  bone2Name=clavicle_r
  swing1LimitDegrees=45
  swing1Motion=Free
  swing2LimitDegrees=45
  swing2Motion=Free
  twistLimitDegrees=45
  twistMotion=Free
 }
 {bone1Name=dyn_Hair_bk_01_ct, bone2Name=dyn_Hair_bk_MSTR, swing1LimitDegrees=10, swing2LimitDegrees=20, twistLimitDegrees=10}
 {bone1Name=dyn_Hair_bk_02_ct, bone2Name=dyn_Hair_bk_01_ct, swing1LimitDegrees=10, swing2LimitDegrees=25, twistLimitDegrees=10}
 {bone1Name=dyn_Hair_bk_03_ct, bone2Name=dyn_Hair_bk_02_ct, swing1LimitDegrees=10, swing2LimitDegrees=25, twistLimitDegrees=20}
 {bone1Name=dyn_Hair_bk_04_ct, bone2Name=dyn_Hair_bk_03_ct, swing1LimitDegrees=15, swing2LimitDegrees=25, twistLimitDegrees=20}
 {bone1Name=dyn_Hair_bk_05_ct, bone2Name=dyn_Hair_bk_04_ct, swing1LimitDegrees=20, swing2LimitDegrees=25, twistLimitDegrees=30}
 {bone1Name=dyn_Hair_bk_06_ct, bone2Name=dyn_Hair_bk_05_ct, swing1LimitDegrees=15, swing2LimitDegrees=30, twistLimitDegrees=30}
 {
  bone1Name=neck_01
  bone2Name=spine_05
  swing1LimitDegrees=45
  swing1Motion=Free
  swing2LimitDegrees=45
  swing2Motion=Free
  twistLimitDegrees=45
  twistMotion=Free
 }
 {bone1Name=dyn_Hair_bk_07_01_r, bone2Name=dyn_Hair_bk_06_ct, swing1LimitDegrees=20, swing2LimitDegrees=20, twistLimitDegrees=10}
 {bone1Name=dyn_Hair_bk_07_02_r, bone2Name=dyn_Hair_bk_06_ct, swing1LimitDegrees=20, swing2LimitDegrees=20, twistLimitDegrees=10}
 {bone1Name=dyn_Hair_bk_07_03_r, bone2Name=dyn_Hair_bk_06_ct, swing1LimitDegrees=20, swing2LimitDegrees=20, twistLimitDegrees=10}
 {bone1Name=dyn_Hair_bk_07_04_r, bone2Name=dyn_Hair_bk_06_ct, swing1LimitDegrees=20, swing2LimitDegrees=20, twistLimitDegrees=10}
 {bone1Name=dyn_Hair_bk_07_01_ct, bone2Name=dyn_Hair_bk_06_ct, swing1LimitDegrees=20, swing2LimitDegrees=20, twistLimitDegrees=10}
 {bone1Name=dyn_Hair_bk_07_02_ct, bone2Name=dyn_Hair_bk_06_ct, swing1LimitDegrees=20, swing2LimitDegrees=20, twistLimitDegrees=10}
 {bone1Name=dyn_Hair_bk_07_02_l, bone2Name=dyn_Hair_bk_06_ct, swing1LimitDegrees=30, swing2LimitDegrees=30, twistLimitDegrees=20}
 {bone1Name=dyn_Hair_bk_07_01_l, bone2Name=dyn_Hair_bk_06_ct, swing1LimitDegrees=30, swing2LimitDegrees=30, twistLimitDegrees=20}
]
