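# Physics asset for the front face-accessory part of the MarchTreatBay female cosmetic body, covering dynamic left and right hair strands.
# Capsules on the head, upper spine, clavicles and upper arms give body volumes around the strands.
# Each strand chains from the head through a free-moving root joint into angle-limited segments with tight twist, with one more tip segment on the left.
# Head and arm links are unrestricted in rotation, and spine_04 has no constraint.
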PhysicsAsset /BRCosmetics/Characters/Player/Female/Medium/Bodies/F_MED_MarchTreatBay/Meshes/Parts/F_MED_MarchTreatBay_FaceAcc_Physics_Front
bodies {
 clavicle_l [{
   shapeType=Capsule
   length=10.7742
   radius=6.9758
   shapeName=clavicle_l_capsule
   center=<3.5546 -1.6757 -0.8114>
   rotation=<rot 80 -0.0006 10>
  }]
 clavicle_r [{shapeType=Capsule, length=10.7742, radius=6.9758, shapeName=clavicle_r_capsule, center=<-3.5546 1.6757 0.8114>, rotation=<rot 80 -0.0003 10>}]
 dyn_L_Hairstrand_01_Tip_01_jnt [{shapeType=Capsule, length=3.7118, radius=1.8693, shapeName=dyn_L_Hairstrand_01_Tip_01_jnt_capsule, center=<3 0 0>, rotation=<rot 90 0 0>}]
 dyn_L_Hairstrand_01_Tip_02_jnt [{
   shapeType=Capsule
   length=2.6318
   radius=1.4179
   shapeName=dyn_L_Hairstrand_01_Tip_02_jnt_capsule
   center=<2.4237 0 0>
   rotation=<rot 90 0 0>
  }]
 dyn_L_Hairstrand_01_Tip_03_jnt [{
   shapeType=Capsule
   length=4.7984
   radius=1.2871
   shapeName=dyn_L_Hairstrand_01_Tip_03_jnt_capsule
   center=<3.7142 0 0>
   rotation=<rot 90 0 0>
  }]
 dyn_L_Hairstrand_02_jnt [{shapeType=Capsule, length=1, radius=1, shapeName=dyn_L_Hairstrand_02_jnt_capsule}]
 dyn_L_Hairstrand_03_jnt1 [{
   shapeType=Capsule
   length=5.3516
   radius=0.8285
   shapeName=dyn_L_Hairstrand_03_jnt1_capsule
   center=<4.297 0 0>
   rotation=<rot 70 -90 -179.9999>
  }]
 dyn_R_Hairstrand_01_Tip_01_jnt [{shapeType=Capsule, length=3.9148, radius=1.1658, shapeName=dyn_R_Hairstrand_01_Tip_01_jnt_capsule, center=<2 0 0>, rotation=<rot 90 0 0>}]
 dyn_R_Hairstrand_01_Tip_02_jnt [{
   shapeType=Capsule
   length=5.9876
   radius=0.9588
   shapeName=dyn_R_Hairstrand_01_Tip_02_jnt_capsule
   center=<4.2142 0 0>
   rotation=<rot 90 0 0>
  }]
 dyn_R_Hairstrand_02_jnt [{shapeType=Capsule, length=1, radius=1, shapeName=dyn_R_Hairstrand_02_jnt_capsule}]
 dyn_R_Hairstrand_03_jnt [{shapeType=Capsule, length=5.0983, radius=1.1385, shapeName=dyn_R_Hairstrand_03_jnt_capsule, center=<6.4843 -1 0>, rotation=<rot 0 0 90>}]
 head [{shapeType=Capsule, length=7.1966, radius=7.8858, shapeName=head_capsule, rotation=<rot 90 0 0>}]
 spine_04 [{shapeType=Capsule, length=12.755, radius=6.755, shapeName=spine_04_capsule, center=<2.805 4.3595 0>, rotation=<rot 0 -90 0>}]
 spine_05 [{shapeType=Capsule, length=20.3438, radius=9.0298, shapeName=spine_05_capsule, center=<-0.1841 0.1134 0>}]
 upperarm_l [{shapeType=Capsule, length=22.3342, radius=4.286, shapeName=upperarm_l_capsule, center=<9.1883 0.5842 1.8106>, rotation=<rot 90 180 0>}]
 upperarm_r [{shapeType=Capsule, length=22.3342, radius=4.286, shapeName=upperarm_r_capsule, center=<-9.1883 -0.5842 -1.8106>, rotation=<rot 90 0 0>}]
}
constraints [
 {
  bone1Name=clavicle_l
  bone2Name=spine_05
  swing1LimitDegrees=45
  swing1Motion=Free
  swing2LimitDegrees=45
  swing2Motion=Free
  twistLimitDegrees=45
  twistMotion=Free
 }
 {
  bone1Name=clavicle_r
  bone2Name=spine_05
  swing1LimitDegrees=45
  swing1Motion=Free
  swing2LimitDegrees=45
  swing2Motion=Free
  twistLimitDegrees=45
  twistMotion=Free
 }
 {
  bone1Name=upperarm_l
  bone2Name=clavicle_l
  swing1LimitDegrees=45
  swing1Motion=Free
  swing2LimitDegrees=45
  swing2Motion=Free
  twistLimitDegrees=45
  twistMotion=Free
 }
 {
  bone1Name=upperarm_r
  bone2Name=clavicle_r
  swing1LimitDegrees=45
  swing1Motion=Free
  swing2LimitDegrees=45
  swing2Motion=Free
  twistLimitDegrees=45
  twistMotion=Free
 }
 {
  bone1Name=dyn_R_Hairstrand_02_jnt
  bone2Name=head
  swing1LimitDegrees=45
  swing1Motion=Free
  swing2LimitDegrees=45
  swing2Motion=Free
  twistLimitDegrees=45
  twistMotion=Free
 }
 {bone1Name=dyn_R_Hairstrand_03_jnt, bone2Name=dyn_R_Hairstrand_02_jnt, swing1LimitDegrees=45, swing2LimitDegrees=45, twistLimitDegrees=2}
 {
  bone1Name=dyn_R_Hairstrand_01_Tip_01_jnt
  bone2Name=dyn_R_Hairstrand_03_jnt
  swing1LimitDegrees=35
  swing2LimitDegrees=35
  twistLimitDegrees=5
 }
 {
  bone1Name=dyn_R_Hairstrand_01_Tip_02_jnt
  bone2Name=dyn_R_Hairstrand_01_Tip_01_jnt
  swing1LimitDegrees=35
  swing2LimitDegrees=35
  twistLimitDegrees=5
 }
 {
  bone1Name=dyn_L_Hairstrand_02_jnt
  bone2Name=head
  swing1LimitDegrees=45
  swing1Motion=Free
  swing2LimitDegrees=45
  swing2Motion=Free
  twistLimitDegrees=45
  twistMotion=Free
 }
 {bone1Name=dyn_L_Hairstrand_03_jnt1, bone2Name=dyn_L_Hairstrand_02_jnt, swing1LimitDegrees=45, swing2LimitDegrees=45, twistLimitDegrees=2}
 {
  bone1Name=dyn_L_Hairstrand_01_Tip_01_jnt
  bone2Name=dyn_L_Hairstrand_03_jnt1
  swing1LimitDegrees=35
  swing2LimitDegrees=35
  twistLimitDegrees=5
 }
 {
  bone1Name=dyn_L_Hairstrand_01_Tip_02_jnt
  bone2Name=dyn_L_Hairstrand_01_Tip_01_jnt
  swing1LimitDegrees=35
  swing2LimitDegrees=35
  twistLimitDegrees=5
 }
 {
  bone1Name=dyn_L_Hairstrand_01_Tip_03_jnt
  bone2Name=dyn_L_Hairstrand_01_Tip_02_jnt
  swing1LimitDegrees=35
  swing2LimitDegrees=35
  twistLimitDegrees=5
 }
 {
  bone1Name=head
  bone2Name=spine_05
  swing1LimitDegrees=45
  swing1Motion=Free
  swing2LimitDegrees=45
  swing2Motion=Free
  twistLimitDegrees=45
  twistMotion=Free
 }
]
