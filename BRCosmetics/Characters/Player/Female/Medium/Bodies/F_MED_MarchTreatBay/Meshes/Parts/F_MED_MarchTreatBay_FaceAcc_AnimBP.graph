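# Post-style animation Blueprint for the MarchTreatBay female character's FaceAcc part, adding secondary motion to hair.
# It copies the parent pose, layering body-mesh clavicles and arms.
# From PreSimCache, the Hair bool selects bone constraints, offsets and a post Control Rig when true, or head-space RigidBody simulation with offsets when false.
# The cached result feeds a neck-space front RigidBody, then chained per-bone AnimDynamics springs for back, front, strand and ear hair.
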AnimBlueprint /BRCosmetics/Characters/Player/Female/Medium/Bodies/F_MED_MarchTreatBay/Meshes/Parts/F_MED_MarchTreatBay_FaceAcc_AnimBP
target_skeleton=&/F_MED_MarchTreatBay_FaceAcc_Skeleton
graphs [
 AnimGraph {nodes [
   RigidBody!(ComponentPose, Alpha, OverrideWorldGravity:Vector="0,0,-1680") -> (Pose) {
    bClampLinearTranslationLimitToRefPose=true
    bOverrideWorldGravity=true
    cachedBoundsScale=1.4
    overridePhysicsAsset=&/F_MED_MarchTreatBay_FaceAcc_Physics
    simulationSpace=BaseBoneSpace
    alphaBoolBlend {blendOption=Linear}
    alphaScaleBiasClamp {bInterpResult=true, interpSpeedIncreasing=2}
    baseBoneRef {boneName=spine_05}
    componentAppliedLinearAccClamp=<1000 1000 1000>
    simSpaceSettings {
     maxAngularAcceleration=32
     maxAngularVelocity=4
     maxLinearAcceleration=1150
     maxLinearVelocity=500
     worldAlpha=0.8
     externalLinearDragV=<1 1 1>
    }
   }
   n1 = RigidBody(ComponentPose=Local To Component(LocalPose=Use cached pose 'RigidBodyCache'()), OverrideWorldGravity="0,0,-1680") {
    bClampLinearTranslationLimitToRefPose=true
    bOverrideWorldGravity=true
    cachedBoundsScale=1.4
    overridePhysicsAsset=&/F_MED_MarchTreatBay_FaceAcc_Physics_Front
    simulationSpace=BaseBoneSpace
    alphaBoolBlend {blendOption=Linear}
    alphaScaleBiasClamp {bInterpResult=true, interpSpeedIncreasing=2}
    baseBoneRef {boneName=neck_02}
    componentAppliedLinearAccClamp=<3000 3000 3000>
    simSpaceSettings {
     maxAngularAcceleration=8
     maxAngularVelocity=2
     maxLinearAcceleration=800
     maxLinearVelocity=200
     velocityScaleZ=0.5
     worldAlpha=0.5
     externalLinearDragV=<0.2 0.2 0.2>
    }
   }
   n2 = Anim Dynamics - Bone: dyn_Hair_bk_07_01_ct!(
    ComponentPose=@n1
    Alpha=1
    GravityOverride:Vector="0,0,-980"
    SimSpaceSettings:Anim Phys Sim Space Settings Structure=(SimSpaceAngularAlpha=0,MaxAngularVelocity=10000,MaxAngularAcceleration=10000,ExternalAngularVelocity=(X=0,Y=0,Z=0))
    ExternalForce="0,0,0"
   ) -> (Pose) {
    angularDampingOverride=1
    angularSpringConstant=500
    bAngularSpring=true
    bOverrideAngularDamping=true
    bUseGravityOverride=true
    gravityScale=0.35
    simulationSpace=RootRelative
    boundBone {boneName=dyn_Hair_bk_07_01_ct}
    gravityOverride {z=-980}
    physicsBodyDefinitions [{
      boundBone {boneName=dyn_Hair_bk_07_01_ct}
      boxExtents=<13 13 13>
      constraintSetup {angularLimitsMax=<5 7 20>, angularLimitsMin=<-5 -30 -20>, angularTarget=<1 0 0>}
      localJointOffset=<1 0 0>
     }]
   }
   n3 = Anim Dynamics - Bone: dyn_Hair_bk_07_02_ct!(
    ComponentPose=@n2
    Alpha=1
    GravityOverride:Vector="0,0,-980"
    SimSpaceSettings:Anim Phys Sim Space Settings Structure=(SimSpaceAngularAlpha=0,MaxAngularVelocity=10000,MaxAngularAcceleration=10000,ExternalAngularVelocity=(X=0,Y=0,Z=0))
    ExternalForce="0,0,0"
   ) -> (Pose) {
    angularDampingOverride=1
    angularSpringConstant=500
    bAngularSpring=true
    bOverrideAngularDamping=true
    bUseGravityOverride=true
    gravityScale=0.35
    simulationSpace=RootRelative
    alphaScaleBiasClamp {scale=0.9}
    boundBone {boneName=dyn_Hair_bk_07_02_ct}
    gravityOverride {z=-980}
    physicsBodyDefinitions [{
      boundBone {boneName=dyn_Hair_bk_07_02_ct}
      boxExtents=<15 15 15>
      constraintSetup {angularLimitsMax=<5 7 20>, angularLimitsMin=<-5 -30 -20>, angularTarget=<1 0 0>}
      localJointOffset=<1.1 0 0>
     }]
   }
   n4 = Anim Dynamics - Bone: dyn_Hair_bk_07_01_l!(
    ComponentPose=@n3
    Alpha=1
    GravityOverride:Vector="0,0,-980"
    SimSpaceSettings:Anim Phys Sim Space Settings Structure=(SimSpaceAngularAlpha=0,MaxAngularVelocity=10000,MaxAngularAcceleration=10000,ExternalAngularVelocity=(X=0,Y=0,Z=0))
    ExternalForce="0,0,0"
   ) -> (Pose) {
    angularDampingOverride=1
    angularSpringConstant=500
    bAngularSpring=true
    bOverrideAngularDamping=true
    bUseGravityOverride=true
    gravityScale=0.35
    simulationSpace=RootRelative
    boundBone {boneName=dyn_Hair_bk_07_01_l}
    gravityOverride {z=-980}
    physicsBodyDefinitions [{
      boundBone {boneName=dyn_Hair_bk_07_01_l}
      boxExtents=<12 12 12>
      constraintSetup {angularLimitsMax=<5 7 20>, angularLimitsMin=<-5 -30 -20>, angularTarget=<1 0 0>}
      localJointOffset=<1 0 0>
     }]
   }
   n5 = Anim Dynamics - Bone: dyn_Hair_bk_07_02_l!(
    ComponentPose=@n4
    Alpha=1
    GravityOverride:Vector="0,0,-980"
    SimSpaceSettings:Anim Phys Sim Space Settings Structure=(SimSpaceAngularAlpha=0,MaxAngularVelocity=10000,MaxAngularAcceleration=10000,ExternalAngularVelocity=(X=0,Y=0,Z=0))
    ExternalForce="0,0,0"
   ) -> (Pose) {
    angularDampingOverride=1
    angularSpringConstant=500
    bAngularSpring=true
    bOverrideAngularDamping=true
    bUseGravityOverride=true
    gravityScale=0.35
    simulationSpace=RootRelative
    alphaScaleBiasClamp {scale=0.9}
    boundBone {boneName=dyn_Hair_bk_07_02_l}
    gravityOverride {z=-980}
    physicsBodyDefinitions [{
      boundBone {boneName=dyn_Hair_bk_07_02_l}
      boxExtents=<16 16 16>
      constraintSetup {angularLimitsMax=<5 7 20>, angularLimitsMin=<-5 -30 -20>, angularTarget=<1 0 0>}
      localJointOffset=<1.1 0 0>
     }]
   }
   n6 = Anim Dynamics - Bone: dyn_Hair_bk_07_01_r!(
    ComponentPose=@n5
    Alpha=1
    GravityOverride:Vector="0,0,-980"
    SimSpaceSettings:Anim Phys Sim Space Settings Structure=(SimSpaceAngularAlpha=0,MaxAngularVelocity=10000,MaxAngularAcceleration=10000,ExternalAngularVelocity=(X=0,Y=0,Z=0))
    ExternalForce="0,0,0"
   ) -> (Pose) {
    angularDampingOverride=1
    angularSpringConstant=500
    bAngularSpring=true
    bOverrideAngularDamping=true
    bUseGravityOverride=true
    gravityScale=0.35
    simulationSpace=RootRelative
    boundBone {boneName=dyn_Hair_bk_07_01_r}
    gravityOverride {z=-980}
    physicsBodyDefinitions [{
      boundBone {boneName=dyn_Hair_bk_07_01_r}
      boxExtents=<10 10 10>
      constraintSetup {angularLimitsMax=<5 7 20>, angularLimitsMin=<-5 -30 -20>, angularTarget=<1 0 0>}
      localJointOffset=<1 0 0>
     }]
   }
   n7 = Anim Dynamics - Bone: dyn_Hair_bk_07_02_r!(
    ComponentPose=@n6
    Alpha=1
    GravityOverride:Vector="0,0,-980"
    SimSpaceSettings:Anim Phys Sim Space Settings Structure=(SimSpaceAngularAlpha=0,MaxAngularVelocity=10000,MaxAngularAcceleration=10000,ExternalAngularVelocity=(X=0,Y=0,Z=0))
    ExternalForce="0,0,0"
   ) -> (Pose) {
    angularDampingOverride=1
    angularSpringConstant=500
    bAngularSpring=true
    bOverrideAngularDamping=true
    bUseGravityOverride=true
    gravityScale=0.35
    simulationSpace=RootRelative
    boundBone {boneName=dyn_Hair_bk_07_02_r}
    gravityOverride {z=-980}
    physicsBodyDefinitions [{
      boundBone {boneName=dyn_Hair_bk_07_02_r}
      boxExtents=<11 11 12>
      constraintSetup {angularLimitsMax=<5 7 20>, angularLimitsMin=<-5 -30 -20>, angularTarget=<1 0 0>}
      localJointOffset=<1 0 0>
     }]
   }
   n8 = Anim Dynamics - Bone: dyn_Hair_bk_07_03_r!(
    ComponentPose=@n7
    Alpha=1
    GravityOverride:Vector="0,0,-980"
    SimSpaceSettings:Anim Phys Sim Space Settings Structure=(SimSpaceAngularAlpha=0,MaxAngularVelocity=10000,MaxAngularAcceleration=10000,ExternalAngularVelocity=(X=0,Y=0,Z=0))
    ExternalForce="0,0,0"
   ) -> (Pose) {
    angularDampingOverride=1
    angularSpringConstant=500
    bAngularSpring=true
    bOverrideAngularDamping=true
    bUseGravityOverride=true
    gravityScale=0.35
    simulationSpace=RootRelative
    boundBone {boneName=dyn_Hair_bk_07_03_r}
    gravityOverride {z=-980}
    physicsBodyDefinitions [{
      boundBone {boneName=dyn_Hair_bk_07_03_r}
      boxExtents=<12 12 12>
      constraintSetup {angularLimitsMax=<5 7 20>, angularLimitsMin=<-5 -30 -20>, angularTarget=<1 0 0>}
      localJointOffset=<1 0 0>
     }]
   }
   n9 = Anim Dynamics - Bone: dyn_Hair_bk_07_04_r!(
    ComponentPose=@n8
    Alpha=1
    GravityOverride:Vector="0,0,-980"
    SimSpaceSettings:Anim Phys Sim Space Settings Structure=(SimSpaceAngularAlpha=0,MaxAngularVelocity=10000,MaxAngularAcceleration=10000,ExternalAngularVelocity=(X=0,Y=0,Z=0))
    ExternalForce="0,0,0"
   ) -> (Pose) {
    angularDampingOverride=1
    angularSpringConstant=500
    bAngularSpring=true
    bOverrideAngularDamping=true
    bUseGravityOverride=true
    gravityScale=0.35
    simulationSpace=RootRelative
    boundBone {boneName=dyn_Hair_bk_07_04_r}
    gravityOverride {z=-980}
    physicsBodyDefinitions [{
      boundBone {boneName=dyn_Hair_bk_07_04_r}
      boxExtents=<13 13 13>
      constraintSetup {angularLimitsMax=<5 7 20>, angularLimitsMin=<-5 -30 -20>, angularTarget=<1 0 0>}
      localJointOffset=<1 0 0>
     }]
   }
   n10 = Anim Dynamics - Bone: dyn_Hair_bk_03_l!(
    ComponentPose=@n9
    Alpha=1
    GravityOverride:Vector="0,0,-980"
    SimSpaceSettings:Anim Phys Sim Space Settings Structure=(SimSpaceAngularAlpha=0,MaxAngularVelocity=10000,MaxAngularAcceleration=10000,ExternalAngularVelocity=(X=0,Y=0,Z=0))
    ExternalForce="0,0,0"
   ) -> (Pose) {
    angularDampingOverride=1
    angularSpringConstant=500
    bAngularSpring=true
    bOverrideAngularDamping=true
    bUseGravityOverride=true
    gravityScale=0.35
    simulationSpace=RootRelative
    boundBone {boneName=dyn_Hair_bk_03_l}
    gravityOverride {z=-980}
    physicsBodyDefinitions [{
      boundBone {boneName=dyn_Hair_bk_03_l}
      boxExtents=<12 12 12>
      constraintSetup {angularLimitsMax=<2.5 5 5>, angularLimitsMin=<-30 -5 -5>, angularTarget=<0 1 0>}
      localJointOffset=<0 1 0>
     }]
   }
   n11 = Anim Dynamics - Bone: dyn_Hair_bk_03_r!(
    ComponentPose=@n10
    Alpha=1
    GravityOverride:Vector="0,0,-980"
    SimSpaceSettings:Anim Phys Sim Space Settings Structure=(SimSpaceAngularAlpha=0,MaxAngularVelocity=10000,MaxAngularAcceleration=10000,ExternalAngularVelocity=(X=0,Y=0,Z=0))
    ExternalForce="0,0,0"
   ) -> (Pose) {
    angularDampingOverride=1
    angularSpringConstant=500
    bAngularSpring=true
    bOverrideAngularDamping=true
    bUseGravityOverride=true
    gravityScale=0.35
    simulationSpace=RootRelative
    boundBone {boneName=dyn_Hair_bk_03_r}
    gravityOverride {z=-980}
    physicsBodyDefinitions [{
      boundBone {boneName=dyn_Hair_bk_03_r}
      boxExtents=<12 12 12>
      constraintSetup {angularLimitsMax=<30 5 5>, angularLimitsMin=<-2.5 -5 -5>, angularTarget=<0 1 0>}
      localJointOffset=<0 -1 0>
     }]
   }
   n12 = Anim Dynamics - Bone: dyn_Hair_bk_04_l!(
    ComponentPose=@n11
    Alpha=1
    GravityOverride:Vector="0,0,-980"
    SimSpaceSettings:Anim Phys Sim Space Settings Structure=(SimSpaceAngularAlpha=0,MaxAngularVelocity=10000,MaxAngularAcceleration=10000,ExternalAngularVelocity=(X=0,Y=0,Z=0))
    ExternalForce="0,0,0"
   ) -> (Pose) {
    angularDampingOverride=1
    angularSpringConstant=500
    bAngularSpring=true
    bOverrideAngularDamping=true
    bUseGravityOverride=true
    gravityScale=0.35
    simulationSpace=RootRelative
    boundBone {boneName=dyn_Hair_bk_04_l}
    gravityOverride {z=-980}
    physicsBodyDefinitions [{
      boundBone {boneName=dyn_Hair_bk_04_l}
      boxExtents=<12 12 12>
      constraintSetup {angularLimitsMax=<2.5 5 5>, angularLimitsMin=<-30 -5 -5>, angularTarget=<0 1 0>}
      localJointOffset=<0 1 0>
     }]
   }
   n13 = Anim Dynamics - Bone: dyn_Hair_bk_04_r!(
    ComponentPose=@n12
    Alpha=1
    GravityOverride:Vector="0,0,-980"
    SimSpaceSettings:Anim Phys Sim Space Settings Structure=(SimSpaceAngularAlpha=0,MaxAngularVelocity=10000,MaxAngularAcceleration=10000,ExternalAngularVelocity=(X=0,Y=0,Z=0))
    ExternalForce="0,0,0"
   ) -> (Pose) {
    angularDampingOverride=1
    angularSpringConstant=500
    bAngularSpring=true
    bOverrideAngularDamping=true
    bUseGravityOverride=true
    gravityScale=0.35
    simulationSpace=RootRelative
    boundBone {boneName=dyn_Hair_bk_04_r}
    gravityOverride {z=-980}
    physicsBodyDefinitions [{
      boundBone {boneName=dyn_Hair_bk_04_r}
      boxExtents=<12 12 12>
      constraintSetup {angularLimitsMax=<30 5 5>, angularLimitsMin=<-2.5 -5 -5>, angularTarget=<0 1 0>}
      localJointOffset=<0 -1 0>
     }]
   }
   n14 = Anim Dynamics - Bone: dyn_Hair_bk_05_l!(
    ComponentPose=@n13
    Alpha=1
    GravityOverride:Vector="0,0,-980"
    SimSpaceSettings:Anim Phys Sim Space Settings Structure=(SimSpaceAngularAlpha=0,MaxAngularVelocity=10000,MaxAngularAcceleration=10000,ExternalAngularVelocity=(X=0,Y=0,Z=0))
    ExternalForce="0,0,0"
   ) -> (Pose) {
    angularDampingOverride=1
    angularSpringConstant=500
    bAngularSpring=true
    bOverrideAngularDamping=true
    bUseGravityOverride=true
    gravityScale=0.35
    simulationSpace=RootRelative
    boundBone {boneName=dyn_Hair_bk_05_l}
    gravityOverride {z=-980}
    physicsBodyDefinitions [{
      boundBone {boneName=dyn_Hair_bk_05_l}
      boxExtents=<12 12 12>
      constraintSetup {angularLimitsMax=<2.5 5 5>, angularLimitsMin=<-30 -5 -5>, angularTarget=<0 1 0>}
      localJointOffset=<0 1 0>
     }]
   }
   n15 = Anim Dynamics - Bone: dyn_Hair_bk_05_r!(
    ComponentPose=@n14
    Alpha=1
    GravityOverride:Vector="0,0,-980"
    SimSpaceSettings:Anim Phys Sim Space Settings Structure=(SimSpaceAngularAlpha=0,MaxAngularVelocity=10000,MaxAngularAcceleration=10000,ExternalAngularVelocity=(X=0,Y=0,Z=0))
    ExternalForce="0,0,0"
   ) -> (Pose) {
    angularDampingOverride=1
    angularSpringConstant=500
    bAngularSpring=true
    bOverrideAngularDamping=true
    bUseGravityOverride=true
    gravityScale=0.35
    simulationSpace=RootRelative
    boundBone {boneName=dyn_Hair_bk_05_r}
    gravityOverride {z=-980}
    physicsBodyDefinitions [{
      boundBone {boneName=dyn_Hair_bk_05_r}
      boxExtents=<12 12 12>
      constraintSetup {angularLimitsMax=<30 5 5>, angularLimitsMin=<-2.5 -5 -5>, angularTarget=<0 1 0>}
      localJointOffset=<0 -1 0>
     }]
   }
   n16 = Anim Dynamics - Bone: dyn_Hair_bk_06_l!(
    ComponentPose=@n15
    Alpha=1
    GravityOverride:Vector="0,0,-980"
    SimSpaceSettings:Anim Phys Sim Space Settings Structure=(SimSpaceAngularAlpha=0,MaxAngularVelocity=10000,MaxAngularAcceleration=10000,ExternalAngularVelocity=(X=0,Y=0,Z=0))
    ExternalForce="0,0,0"
   ) -> (Pose) {
    angularDampingOverride=1
    angularSpringConstant=500
    bAngularSpring=true
    bOverrideAngularDamping=true
    bUseGravityOverride=true
    gravityScale=0.35
    simulationSpace=RootRelative
    boundBone {boneName=dyn_Hair_bk_06_l}
    gravityOverride {z=-980}
    physicsBodyDefinitions [{
      boundBone {boneName=dyn_Hair_bk_06_l}
      boxExtents=<12 12 12>
      constraintSetup {angularLimitsMax=<2.5 5 5>, angularLimitsMin=<-30 -5 -5>, angularTarget=<0 1 0>}
      localJointOffset=<0 1 0>
     }]
   }
   n17 = Anim Dynamics - Bone: dyn_Hair_bk_06_r!(
    ComponentPose=@n16
    Alpha=1
    GravityOverride:Vector="0,0,-980"
    SimSpaceSettings:Anim Phys Sim Space Settings Structure=(SimSpaceAngularAlpha=0,MaxAngularVelocity=10000,MaxAngularAcceleration=10000,ExternalAngularVelocity=(X=0,Y=0,Z=0))
    ExternalForce="0,0,0"
   ) -> (Pose) {
    angularDampingOverride=1
    angularSpringConstant=500
    bAngularSpring=true
    bOverrideAngularDamping=true
    bUseGravityOverride=true
    gravityScale=0.35
    simulationSpace=RootRelative
    boundBone {boneName=dyn_Hair_bk_06_r}
    gravityOverride {z=-980}
    physicsBodyDefinitions [{
      boundBone {boneName=dyn_Hair_bk_06_r}
      boxExtents=<12 12 12>
      constraintSetup {angularLimitsMax=<30 5 5>, angularLimitsMin=<-2.5 -5 -5>, angularTarget=<0 1 0>}
      localJointOffset=<0 -1 0>
     }]
   }
   n18 = Anim Dynamics - Bone: dyn_hair_f_01!(
    ComponentPose=@n17
    Alpha=1
    GravityOverride:Vector="0,0,0"
    SimSpaceSettings:Anim Phys Sim Space Settings Structure=(SimSpaceAngularAlpha=0,MaxAngularVelocity=10000,MaxAngularAcceleration=10000,ExternalAngularVelocity=(X=0,Y=0,Z=0))
    ExternalForce="0,0,0"
   ) -> (Pose) {
    angularBiasOverride=0.9999
    angularDampingOverride=0.9999
    bLinearSpring=true
    bOverrideAngularDamping=true
    bUseGravityOverride=true
    linearDampingOverride=0.9999
    linearSpringConstant=4000
    simulationSpace=RootRelative
    alphaBoolBlend {blendOption=Linear}
    alphaScaleBiasClamp {scale=0.5}
    boundBone {boneName=dyn_hair_f_01}
    componentAppliedLinearAccClamp=<1250 1250 1250>
    componentLinearAccScale=<0.4 0.5 0.5>
    componentLinearVelScale=<0.4 0.5 0.5>
    physicsBodyDefinitions [{
      boundBone {boneName=dyn_hair_f_01}
      boxExtents=<12 12 12>
      constraintSetup {angularTargetAxis=AxisZ, angularLimitsMax=<5 10 1>, angularLimitsMin=<-5 -20 -15>, angularTarget=<0 1 0>}
      localJointOffset=<1.5 0 0>
     }]
   }
   n19 = Anim Dynamics - Bone: dyn_hair_f_02!(
    ComponentPose=@n18
    Alpha=1
    GravityOverride:Vector="0,0,0"
    SimSpaceSettings:Anim Phys Sim Space Settings Structure=(SimSpaceAngularAlpha=0,MaxAngularVelocity=10000,MaxAngularAcceleration=10000,ExternalAngularVelocity=(X=0,Y=0,Z=0))
    ExternalForce="0,0,0"
   ) -> (Pose) {
    angularBiasOverride=0.9999
    angularDampingOverride=0.9999
    bLinearSpring=true
    bOverrideAngularDamping=true
    bUseGravityOverride=true
    linearDampingOverride=0.9999
    linearSpringConstant=4000
    simulationSpace=RootRelative
    alphaBoolBlend {blendOption=Linear}
    alphaScaleBiasClamp {scale=0.5}
    boundBone {boneName=dyn_hair_f_02}
    componentAppliedLinearAccClamp=<1250 1250 1250>
    componentLinearAccScale=<0.4 0.5 0.5>
    componentLinearVelScale=<0.4 0.5 0.5>
    physicsBodyDefinitions [{
      boundBone {boneName=dyn_hair_f_02}
      boxExtents=<12 12 12>
      constraintSetup {angularTargetAxis=AxisZ, angularLimitsMax=<5 10 1>, angularLimitsMin=<-5 -20 -15>, angularTarget=<0 1 0>}
      localJointOffset=<1.5 0 0>
     }]
   }
   n20 = Anim Dynamics - Bone: dyn_hair_l_01!(
    ComponentPose=@n19
    Alpha=1
    GravityOverride:Vector="0,0,0"
    SimSpaceSettings:Anim Phys Sim Space Settings Structure=(SimSpaceAngularAlpha=0,MaxAngularVelocity=10000,MaxAngularAcceleration=10000,ExternalAngularVelocity=(X=0,Y=0,Z=0))
    ExternalForce="0,0,0"
   ) -> (Pose) {
    angularBiasOverride=0.9999
    angularDampingOverride=0.9999
    bLinearSpring=true
    bOverrideAngularDamping=true
    bUseGravityOverride=true
    linearDampingOverride=0.9999
    linearSpringConstant=4000
    simulationSpace=RootRelative
    alphaBoolBlend {blendOption=Linear}
    alphaScaleBiasClamp {scale=0.5}
    boundBone {boneName=dyn_hair_l_01}
    componentAppliedLinearAccClamp=<1250 1250 1250>
    componentLinearAccScale=<0.4 0.5 0.5>
    componentLinearVelScale=<0.4 0.5 0.5>
    physicsBodyDefinitions [{
      boundBone {boneName=dyn_hair_l_01}
      boxExtents=<12 12 12>
      constraintSetup {angularTargetAxis=AxisZ, angularLimitsMax=<5 10 1>, angularLimitsMin=<-5 -20 -15>, angularTarget=<0 1 0>}
      localJointOffset=<1.5 0 0>
     }]
   }
   n21 = Anim Dynamics - Bone: dyn_hair_l_02!(
    ComponentPose=@n20
    Alpha=1
    GravityOverride:Vector="0,0,0"
    SimSpaceSettings:Anim Phys Sim Space Settings Structure=(SimSpaceAngularAlpha=0,MaxAngularVelocity=10000,MaxAngularAcceleration=10000,ExternalAngularVelocity=(X=0,Y=0,Z=0))
    ExternalForce="0,0,0"
   ) -> (Pose) {
    angularBiasOverride=0.9999
    angularDampingOverride=0.9999
    bLinearSpring=true
    bOverrideAngularDamping=true
    bUseGravityOverride=true
    linearDampingOverride=0.9999
    linearSpringConstant=4000
    simulationSpace=RootRelative
    alphaBoolBlend {blendOption=Linear}
    alphaScaleBiasClamp {scale=0.5}
    boundBone {boneName=dyn_hair_l_02}
    componentAppliedLinearAccClamp=<1250 1250 1250>
    componentLinearAccScale=<0.4 0.5 0.5>
    componentLinearVelScale=<0.4 0.5 0.5>
    physicsBodyDefinitions [{
      boundBone {boneName=dyn_hair_l_02}
      boxExtents=<12 12 12>
      constraintSetup {angularTargetAxis=AxisZ, angularLimitsMax=<5 10 1>, angularLimitsMin=<-5 -20 -15>, angularTarget=<0 1 0>}
      localJointOffset=<1.5 0 0>
     }]
   }
   n22 = Anim Dynamics - Bone: dyn_L_Hairstrand_03_Tip_03_jnt!(
    ComponentPose=@n21
    Alpha=1
    GravityOverride:Vector="0,0,0"
    SimSpaceSettings:Anim Phys Sim Space Settings Structure=(SimSpaceAngularAlpha=0,MaxAngularVelocity=10000,MaxAngularAcceleration=10000,ExternalAngularVelocity=(X=0,Y=0,Z=0))
    ExternalForce="0,0,0"
   ) -> (Pose) {
    angularBiasOverride=0.9999
    angularDampingOverride=0.9999
    bLinearSpring=true
    bOverrideAngularDamping=true
    bUseGravityOverride=true
    linearDampingOverride=0.9999
    linearSpringConstant=4000
    simulationSpace=RootRelative
    alphaBoolBlend {blendOption=Linear}
    alphaScaleBiasClamp {scale=0.5}
    boundBone {boneName=dyn_L_Hairstrand_03_Tip_03_jnt}
    componentAppliedLinearAccClamp=<1250 1250 1250>
    componentLinearAccScale=<0.4 0.5 0.5>
    componentLinearVelScale=<0.4 0.5 0.5>
    physicsBodyDefinitions [{
      boundBone {boneName=dyn_L_Hairstrand_03_Tip_03_jnt}
      boxExtents=<12 12 12>
      constraintSetup {angularTargetAxis=AxisZ, angularLimitsMax=<5 30 1>, angularLimitsMin=<-5 10 -10>, angularTarget=<0 1 0>}
      localJointOffset=<1.5 0 0>
     }]
   }
   n23 = Anim Dynamics - Bone: dyn_hair_f_02(
    ComponentPose=@n22
    Alpha=1
    GravityOverride="0,0,0"
    SimSpaceSettings=(SimSpaceAngularAlpha=0,MaxAngularVelocity=10000,MaxAngularAcceleration=10000,ExternalAngularVelocity=(X=0,Y=0,Z=0))
    ExternalForce="0,0,0"
   ) {
    angularBiasOverride=0.9999
    angularDampingOverride=0.9999
    bLinearSpring=true
    bOverrideAngularDamping=true
    bUseGravityOverride=true
    linearDampingOverride=0.9999
    linearSpringConstant=4000
    simulationSpace=RootRelative
    alphaBoolBlend {blendOption=Linear}
    alphaScaleBiasClamp {scale=0.5}
    boundBone {boneName=dyn_hair_f_02}
    componentAppliedLinearAccClamp=<1250 1250 1250>
    componentLinearAccScale=<0.4 0.5 0.5>
    componentLinearVelScale=<0.4 0.5 0.5>
    physicsBodyDefinitions [{
      boundBone {boneName=dyn_hair_f_02}
      boxExtents=<12 12 12>
      constraintSetup {angularTargetAxis=AxisZ, angularLimitsMax=<5 10 1>, angularLimitsMin=<-5 -20 -15>, angularTarget=<0 1 0>}
      localJointOffset=<1.5 0 0>
     }]
   }
   n24 = Anim Dynamics - Bone: dyn_L_Hairstrand_03_Tip_04_jnt!(
    ComponentPose=@n23
    Alpha=1
    GravityOverride:Vector="0,0,0"
    SimSpaceSettings:Anim Phys Sim Space Settings Structure=(SimSpaceAngularAlpha=0,MaxAngularVelocity=10000,MaxAngularAcceleration=10000,ExternalAngularVelocity=(X=0,Y=0,Z=0))
    ExternalForce="0,0,0"
   ) -> (Pose) {
    angularBiasOverride=0.9999
    angularDampingOverride=0.9999
    bLinearSpring=true
    bOverrideAngularDamping=true
    bUseGravityOverride=true
    linearDampingOverride=0.9999
    linearSpringConstant=4000
    simulationSpace=RootRelative
    alphaBoolBlend {blendOption=Linear}
    alphaScaleBiasClamp {scale=0.5}
    boundBone {boneName=dyn_L_Hairstrand_03_Tip_04_jnt}
    componentAppliedLinearAccClamp=<1250 1250 1250>
    componentLinearAccScale=<0.4 0.5 0.5>
    componentLinearVelScale=<0.4 0.5 0.5>
    physicsBodyDefinitions [{
      boundBone {boneName=dyn_L_Hairstrand_03_Tip_04_jnt}
      boxExtents=<12 12 12>
      constraintSetup {angularTargetAxis=AxisZ, angularLimitsMax=<5 20 0>, angularLimitsMin=<-5 0 -25>, angularTarget=<0 1 0>}
      localJointOffset=<1.5 0 0>
     }]
   }
   n25 = Anim Dynamics - Bone: dyn_R_Hairstrand_03_Tip_02_jnt!(
    ComponentPose=@n24
    Alpha=1
    GravityOverride:Vector="0,0,0"
    SimSpaceSettings:Anim Phys Sim Space Settings Structure=(SimSpaceAngularAlpha=0,MaxAngularVelocity=10000,MaxAngularAcceleration=10000,ExternalAngularVelocity=(X=0,Y=0,Z=0))
    ExternalForce="0,0,0"
   ) -> (Pose) {
    angularBiasOverride=0.9999
    angularDampingOverride=0.9999
    bLinearSpring=true
    bOverrideAngularDamping=true
    bUseGravityOverride=true
    linearDampingOverride=0.9999
    linearSpringConstant=4000
    simulationSpace=RootRelative
    alphaBoolBlend {blendOption=Linear}
    alphaScaleBiasClamp {scale=0.5}
    boundBone {boneName=dyn_R_Hairstrand_03_Tip_02_jnt}
    componentAppliedLinearAccClamp=<1250 1250 1250>
    componentLinearAccScale=<0.4 0.5 0.5>
    componentLinearVelScale=<0.4 0.5 0.5>
    physicsBodyDefinitions [{
      boundBone {boneName=dyn_R_Hairstrand_03_Tip_02_jnt}
      boxExtents=<12 12 12>
      constraintSetup {angularTargetAxis=AxisZ, angularLimitsMax=<5 10 1>, angularLimitsMin=<-5 -20 -15>, angularTarget=<0 1 0>}
      localJointOffset=<1.5 0 0>
     }]
   }
   n26 = Anim Dynamics - Bone: dyn_R_Hairstrand_03_Tip_03_jnt!(
    ComponentPose=@n25
    Alpha=1
    GravityOverride:Vector="0,0,0"
    SimSpaceSettings:Anim Phys Sim Space Settings Structure=(SimSpaceAngularAlpha=0,MaxAngularVelocity=10000,MaxAngularAcceleration=10000,ExternalAngularVelocity=(X=0,Y=0,Z=0))
    ExternalForce="0,0,0"
   ) -> (Pose) {
    angularBiasOverride=0.9999
    angularDampingOverride=0.9999
    bLinearSpring=true
    bOverrideAngularDamping=true
    bUseGravityOverride=true
    linearDampingOverride=0.9999
    linearSpringConstant=4000
    simulationSpace=RootRelative
    alphaBoolBlend {blendOption=Linear}
    alphaScaleBiasClamp {scale=0.5}
    boundBone {boneName=dyn_R_Hairstrand_03_Tip_03_jnt}
    componentAppliedLinearAccClamp=<1250 1250 1250>
    componentLinearAccScale=<0.4 0.5 0.5>
    componentLinearVelScale=<0.4 0.5 0.5>
    physicsBodyDefinitions [{
      boundBone {boneName=dyn_R_Hairstrand_03_Tip_03_jnt}
      boxExtents=<12 12 12>
      constraintSetup {angularTargetAxis=AxisZ, angularLimitsMax=<5 10 1>, angularLimitsMin=<-5 -20 -15>, angularTarget=<0 1 0>}
      localJointOffset=<1.5 0 0>
     }]
   }
   n27 = Anim Dynamics - Bone: dyn_R_Hairstrand_03_Tip_04_jnt!(
    ComponentPose=@n26
    Alpha=1
    GravityOverride:Vector="0,0,0"
    SimSpaceSettings:Anim Phys Sim Space Settings Structure=(SimSpaceAngularAlpha=0,MaxAngularVelocity=10000,MaxAngularAcceleration=10000,ExternalAngularVelocity=(X=0,Y=0,Z=0))
    ExternalForce="0,0,0"
   ) -> (Pose) {
    angularBiasOverride=0.9999
    angularDampingOverride=0.9999
    bLinearSpring=true
    bOverrideAngularDamping=true
    bUseGravityOverride=true
    linearDampingOverride=0.9999
    linearSpringConstant=4000
    simulationSpace=RootRelative
    alphaBoolBlend {blendOption=Linear}
    alphaScaleBiasClamp {scale=0.5}
    boundBone {boneName=dyn_R_Hairstrand_03_Tip_04_jnt}
    componentAppliedLinearAccClamp=<1250 1250 1250>
    componentLinearAccScale=<0.4 0.5 0.5>
    componentLinearVelScale=<0.4 0.5 0.5>
    physicsBodyDefinitions [{
      boundBone {boneName=dyn_R_Hairstrand_03_Tip_04_jnt}
      boxExtents=<12 12 12>
      constraintSetup {angularTargetAxis=AxisZ, angularLimitsMax=<5 10 1>, angularLimitsMin=<-5 -20 -15>, angularTarget=<0 1 0>}
      localJointOffset=<1.5 0 0>
     }]
   }
   n28 = Anim Dynamics - Bone: dyn_R_Hairstrand_03_Tip_05_jnt!(
    ComponentPose=@n27
    Alpha=1
    GravityOverride:Vector="0,0,0"
    SimSpaceSettings:Anim Phys Sim Space Settings Structure=(SimSpaceAngularAlpha=0,MaxAngularVelocity=10000,MaxAngularAcceleration=10000,ExternalAngularVelocity=(X=0,Y=0,Z=0))
    ExternalForce="0,0,0"
   ) -> (Pose) {
    angularBiasOverride=0.9999
    angularDampingOverride=0.9999
    bLinearSpring=true
    bOverrideAngularDamping=true
    bUseGravityOverride=true
    linearDampingOverride=0.9999
    linearSpringConstant=4000
    simulationSpace=RootRelative
    alphaBoolBlend {blendOption=Linear}
    alphaScaleBiasClamp {scale=0.5}
    boundBone {boneName=dyn_R_Hairstrand_03_Tip_05_jnt}
    componentAppliedLinearAccClamp=<1250 1250 1250>
    componentLinearAccScale=<0.4 0.5 0.5>
    componentLinearVelScale=<0.4 0.5 0.5>
    physicsBodyDefinitions [{
      boundBone {boneName=dyn_R_Hairstrand_03_Tip_05_jnt}
      boxExtents=<12 12 12>
      constraintSetup {angularTargetAxis=AxisZ, angularLimitsMax=<5 10 1>, angularLimitsMin=<-5 -20 -15>, angularTarget=<0 1 0>}
      localJointOffset=<1.5 0 0>
     }]
   }
   n29 = Anim Dynamics - Bone: dyn_hair_ear_01_l!(
    ComponentPose=@n28
    Alpha=1
    GravityOverride:Vector="0,0,-980"
    SimSpaceSettings:Anim Phys Sim Space Settings Structure=(SimSpaceAngularAlpha=0,MaxAngularVelocity=10000,MaxAngularAcceleration=10000,ExternalAngularVelocity=(X=0,Y=0,Z=0))
    ExternalForce="0,0,0"
   ) -> (Pose) {
    angularDampingOverride=1
    angularSpringConstant=5000
    bAngularSpring=true
    bOverrideAngularDamping=true
    bUseGravityOverride=true
    gravityScale=0.35
    simulationSpace=RootRelative
    boundBone {boneName=dyn_hair_ear_01_l}
    gravityOverride {z=-980}
    physicsBodyDefinitions [{
      boundBone {boneName=dyn_hair_ear_01_l}
      boxExtents=<13 13 13>
      constraintSetup {angularLimitsMax=<15 15 15>, angularLimitsMin=<-15 -15 -5>, angularTarget=<1 0 0>}
      localJointOffset=<1 0 0>
     }]
   }
   n30 = Anim Dynamics - Bone: dyn_hair_ear_01_r!(
    ComponentPose=@n29
    Alpha=1
    GravityOverride:Vector="0,0,-980"
    SimSpaceSettings:Anim Phys Sim Space Settings Structure=(SimSpaceAngularAlpha=0,MaxAngularVelocity=10000,MaxAngularAcceleration=10000,ExternalAngularVelocity=(X=0,Y=0,Z=0))
    ExternalForce="0,0,0"
   ) -> (Pose) {
    angularDampingOverride=1
    angularSpringConstant=5000
    bAngularSpring=true
    bOverrideAngularDamping=true
    bUseGravityOverride=true
    gravityScale=0.35
    simulationSpace=RootRelative
    boundBone {boneName=dyn_hair_ear_01_r}
    gravityOverride {z=-980}
    physicsBodyDefinitions [{
      boundBone {boneName=dyn_hair_ear_01_r}
      boxExtents=<14 14 14>
      constraintSetup {angularLimitsMax=<15 15 15>, angularLimitsMin=<-15 -15 -5>, angularTarget=<1 0 0>}
      localJointOffset=<1 0 0>
     }]
   }
   Output Pose(Result=Component To Local(ComponentPose=@n30)) {layerGroup=None}
   n31 = "Constraint - dyn_Hair_bk_01_ct (PP)"!(ComponentPose=Local To Component(LocalPose=Use cached pose 'PreSimCache'()), Alpha=1, ConstraintWeights_0=0.2, ConstraintWeights_1:float=0.8) -> (Pose) {
    alphaBoolBlend {blendOption=Linear}
    boneToModify {boneName=dyn_Hair_bk_01_ct}
    constraintSetup [{targetBone {boneName=neck_02}}, {targetBone {boneName=head}}]
    constraintWeights [1,1]
   }
   n32 = "Constraint - dyn_Hair_bk_02_ct (PP)"!(ComponentPose=@n31, Alpha=1, ConstraintWeights_0=0.5, ConstraintWeights_1:float=0.5) -> (Pose) {
    alphaBoolBlend {blendOption=Linear}
    boneToModify {boneName=dyn_Hair_bk_02_ct}
    constraintSetup [{targetBone {boneName=neck_02}}, {targetBone {boneName=head}}]
    constraintWeights [1,1]
   }
   n33 = "Constraint - dyn_Hair_bk_03_ct (PP)"!(ComponentPose=@n32, Alpha=1, ConstraintWeights_0=0.5, ConstraintWeights_1:float=0.5) -> (Pose) {
    alphaBoolBlend {blendOption=Linear}
    boneToModify {boneName=dyn_Hair_bk_03_ct}
    constraintSetup [{targetBone {boneName=neck_02}}, {targetBone {boneName=spine_05}}]
    constraintWeights [1,1]
   }
   n34 = "Constraint - dyn_Hair_bk_03_l (PP)"!(ComponentPose=@n33, Alpha=1, ConstraintWeights_0=0.7, ConstraintWeights_1:float=0.3) -> (Pose) {
    alphaBoolBlend {blendOption=Linear}
    boneToModify {boneName=dyn_Hair_bk_03_l}
    constraintSetup [{targetBone {boneName=clavicle_l}}, {targetBone {boneName=spine_05}}]
    constraintWeights [1,1]
   }
   n35 = "Constraint - dyn_Hair_bk_03_r (PP)"!(ComponentPose=@n34, Alpha=1, ConstraintWeights_0=0.7, ConstraintWeights_1:float=0.3) -> (Pose) {
    alphaBoolBlend {blendOption=Linear}
    boneToModify {boneName=dyn_Hair_bk_03_r}
    constraintSetup [{targetBone {boneName=clavicle_r}}, {targetBone {boneName=spine_05}}]
    constraintWeights [1,1]
   }
   n36 = "Constraint - dyn_Hair_bk_04_ct (PP)"!(ComponentPose=@n35, Alpha=1, ConstraintWeights_0=0.5, ConstraintWeights_1:float=0.5) -> (Pose) {
    alphaBoolBlend {blendOption=Linear}
    boneToModify {boneName=dyn_Hair_bk_04_ct}
    constraintSetup [{targetBone {boneName=spine_04}}, {targetBone {boneName=spine_05}}]
    constraintWeights [1,1]
   }
   n37 = "Constraint - dyn_Hair_bk_04_l (PP)"!(ComponentPose=@n36, Alpha=1, ConstraintWeights_0=0.7, ConstraintWeights_1:float=0.3) -> (Pose) {
    alphaBoolBlend {blendOption=Linear}
    boneToModify {boneName=dyn_Hair_bk_04_l}
    constraintSetup [{targetBone {boneName=clavicle_l}}, {targetBone {boneName=spine_05}}]
    constraintWeights [1,1]
   }
   n38 = "Constraint - dyn_Hair_bk_04_r (PP)"!(ComponentPose=@n37, Alpha=1, ConstraintWeights_0=0.7, ConstraintWeights_1:float=0.3) -> (Pose) {
    alphaBoolBlend {blendOption=Linear}
    boneToModify {boneName=dyn_Hair_bk_04_r}
    constraintSetup [{targetBone {boneName=spine_05}}, {targetBone {boneName=clavicle_r}}]
    constraintWeights [1,1]
   }
   n39 = "Transform (Modify) Bone - Bone: dyn_Hair_bk_04_ct"(ComponentPose=@n38, Alpha=1, Translation="0,0,0", Rotation="-6,0,0", Scale="1,1,1") {
    rotationMode=BMM_Additive
    rotationSpace=BCS_ParentBoneSpace
    translationMode=BMM_Additive
    translationSpace=BCS_ParentBoneSpace
    alphaBoolBlend {blendOption=Linear}
    boneToModify {boneName=dyn_Hair_bk_04_ct}
    rotation {pitch=-6}
   }
   n40 = "Transform (Modify) Bone - Bone: dyn_Hair_bk_05_ct"(ComponentPose=@n39, Alpha=1, Translation="0,0,0", Rotation="5,0,0", Scale="1,1,1") {
    rotationMode=BMM_Additive
    rotationSpace=BCS_ParentBoneSpace
    translationMode=BMM_Additive
    translationSpace=BCS_ParentBoneSpace
    alphaBoolBlend {blendOption=Linear}
    boneToModify {boneName=dyn_Hair_bk_05_ct}
    rotation {pitch=5}
   }
   n41 = "Transform (Modify) Bone - Bone: dyn_Hair_bk_06_ct"(ComponentPose=@n40, Alpha=1, Translation="0,0,0", Rotation="5,0,0", Scale="1,1,1") {
    rotationMode=BMM_Additive
    rotationSpace=BCS_ParentBoneSpace
    translationMode=BMM_Additive
    translationSpace=BCS_ParentBoneSpace
    alphaBoolBlend {blendOption=Linear}
    boneToModify {boneName=dyn_Hair_bk_06_ct}
    rotation {pitch=5}
   }
   n42 = "Transform (Modify) Bone - Bone: dyn_Hair_bk_03_l"(ComponentPose=@n41, Alpha=1, Translation="0,0,0", Rotation="0,0,-15", Scale="1,1,1") {
    rotationMode=BMM_Additive
    rotationSpace=BCS_ParentBoneSpace
    translationMode=BMM_Additive
    translationSpace=BCS_ParentBoneSpace
    alphaBoolBlend {blendOption=Linear}
    boneToModify {boneName=dyn_Hair_bk_03_l}
    rotation {roll=-15}
   }
   n43 = "Transform (Modify) Bone - Bone: dyn_Hair_bk_03_r"(ComponentPose=@n42, Alpha=1, Translation="0,0,0.5", Rotation="0,0,15", Scale="1,1,1") {
    rotationMode=BMM_Additive
    rotationSpace=BCS_ParentBoneSpace
    translationMode=BMM_Additive
    translationSpace=BCS_ParentBoneSpace
    alphaBoolBlend {blendOption=Linear}
    boneToModify {boneName=dyn_Hair_bk_03_r}
    rotation {roll=15}
    translation {z=0.5}
   }
   n44 = "Transform (Modify) Bone - Bone: dyn_Hair_bk_04_l"(ComponentPose=@n43, Alpha=1, Translation="0,0,0.5", Rotation="0,0,-15", Scale="1,1,1") {
    rotationMode=BMM_Additive
    rotationSpace=BCS_ParentBoneSpace
    translationMode=BMM_Additive
    translationSpace=BCS_ParentBoneSpace
    alphaBoolBlend {blendOption=Linear}
    boneToModify {boneName=dyn_Hair_bk_04_l}
    rotation {roll=-15}
    translation {z=0.5}
   }
   n45 = "Transform (Modify) Bone - Bone: dyn_Hair_bk_04_r"(ComponentPose=@n44, Alpha=1, Translation="0,0,0.5", Rotation="0,0,15", Scale="1,1,1") {
    rotationMode=BMM_Additive
    rotationSpace=BCS_ParentBoneSpace
    translationMode=BMM_Additive
    translationSpace=BCS_ParentBoneSpace
    alphaBoolBlend {blendOption=Linear}
    boneToModify {boneName=dyn_Hair_bk_04_r}
    rotation {roll=15}
    translation {z=0.5}
   }
   n46 = Control Rig(Source=Component To Local(ComponentPose=@n45)) {
    alphaBoolBlend {blendOption=Linear}
    controlRigAssetReference {blueprintRigClass=&/F_MED_MarchTreatBay_FaceAcc_CtrlRig_Post.F_MED_MarchTreatBay_FaceAcc_CtrlRig_Post_C}
   }
   n47 = RigidBody(ComponentPose=Local To Component(LocalPose=Use cached pose 'PreSimCache'()), OverrideWorldGravity=Get GravityOverride()) {
    bClampLinearTranslationLimitToRefPose=true
    bOverrideWorldGravity=true
    cachedBoundsScale=1.4
    overridePhysicsAsset=&F_MED_MarchTreatBay_FaceAcc_Physics
    simulationSpace=BaseBoneSpace
    alphaBoolBlend {blendOption=Linear}
    alphaScaleBiasClamp {bInterpResult=true, interpSpeedIncreasing=2}
    baseBoneRef {boneName=head}
    componentAppliedLinearAccClamp=<1000 1000 1000>
    simSpaceSettings {
     maxAngularAcceleration=32
     maxAngularVelocity=4
     maxLinearAcceleration=1150
     maxLinearVelocity=500
     worldAlpha=0.8
     externalLinearDragV=<2 2 1>
    }
   }
   n48 = "Transform (Modify) Bone - Bone: dyn_Hair_bk_01_ct"(ComponentPose=@n47, Alpha=1, Translation="0,0,2", Rotation="5,0,0", Scale="1,1,1") {
    rotationMode=BMM_Additive
    rotationSpace=BCS_ParentBoneSpace
    translationMode=BMM_Additive
    translationSpace=BCS_ParentBoneSpace
    alphaBoolBlend {blendOption=Linear}
    alphaScaleBiasClamp {bInterpResult=true}
    boneToModify {boneName=dyn_Hair_bk_01_ct}
    rotation {pitch=5}
    translation {z=2}
   }
   n49 = "Transform (Modify) Bone - Bone: dyn_Hair_bk_02_ct"(ComponentPose=@n48, Alpha=1, Translation="0,0,2", Rotation="0,0,0", Scale="1,1,1") {
    rotationMode=BMM_Additive
    rotationSpace=BCS_ParentBoneSpace
    translationMode=BMM_Additive
    translationSpace=BCS_ParentBoneSpace
    alphaBoolBlend {blendOption=Linear}
    alphaScaleBiasClamp {bInterpResult=true}
    boneToModify {boneName=dyn_Hair_bk_02_ct}
    translation {z=2}
   }
   n50 = "Transform (Modify) Bone - Bone: dyn_Hair_bk_03_ct"(ComponentPose=@n49, Alpha=1, Translation="0,0,-4", Rotation="-20,0,0", Scale="1,1,1") {
    rotationMode=BMM_Additive
    rotationSpace=BCS_ParentBoneSpace
    translationMode=BMM_Additive
    translationSpace=BCS_ParentBoneSpace
    alphaBoolBlend {blendOption=Linear}
    alphaScaleBiasClamp {bInterpResult=true}
    boneToModify {boneName=dyn_Hair_bk_03_ct}
    rotation {pitch=-20}
    translation {z=-4}
   }
   n51 = "Transform (Modify) Bone - Bone: dyn_Hair_bk_04_ct"(ComponentPose=@n50, Alpha=1, Translation="0,0,0", Rotation="50,0,0", Scale="1,1,1") {
    rotationMode=BMM_Additive
    rotationSpace=BCS_ParentBoneSpace
    translationMode=BMM_Additive
    translationSpace=BCS_ParentBoneSpace
    alphaBoolBlend {blendOption=Linear}
    alphaScaleBiasClamp {bInterpResult=true}
    boneToModify {boneName=dyn_Hair_bk_04_ct}
    rotation {pitch=50}
   }
   n52 = "Transform (Modify) Bone - Bone: dyn_Hair_bk_05_ct"(ComponentPose=@n51, Alpha=1, Translation="0,0,0", Rotation="-10,0,0", Scale="1,1,1") {
    rotationMode=BMM_Additive
    rotationSpace=BCS_ParentBoneSpace
    translationMode=BMM_Additive
    translationSpace=BCS_ParentBoneSpace
    alphaBoolBlend {blendOption=Linear}
    alphaScaleBiasClamp {bInterpResult=true}
    boneToModify {boneName=dyn_Hair_bk_05_ct}
    rotation {pitch=-10}
   }
   n53 = "Transform (Modify) Bone - Bone: dyn_Hair_bk_06_ct"(ComponentPose=@n52, Alpha=1, Translation="0,0,0", Rotation="50,0,0", Scale="1,1,1") {
    rotationMode=BMM_Additive
    rotationSpace=BCS_ParentBoneSpace
    translationMode=BMM_Additive
    translationSpace=BCS_ParentBoneSpace
    alphaBoolBlend {blendOption=Linear}
    alphaScaleBiasClamp {bInterpResult=true}
    boneToModify {boneName=dyn_Hair_bk_06_ct}
    rotation {pitch=50}
   }
   Save cached pose 'RigidBodyCache'(Pose=Blend Poses by bool(BlendPose_0=@n46, BlendPose_1=Component To Local(ComponentPose=@n53), bActiveValue=Get Hair() -> (Hair:bool)) {blendType=Linear}) {cacheName=RigidBodyCache}
   n54 = Layered blend per bone(
    BasePose=Copy Pose From Mesh() {bCopyCurves=true, bUseAttachedParent=true}
    BlendPoses_0=Copy Pose From Mesh(SourceMeshComponent=Get body Mesh() -> (body Mesh:Skeletal Mesh Component Object Reference))
   ) {layerSetup [{branchFilters [{boneName=upperarm_l}, {boneName=upperarm_r}, {boneName=clavicle_l}, {boneName=clavicle_r}]}]}
   n55 = Control Rig(Source=@n54) {
    alphaBoolBlend {blendOption=Linear}
    controlRigAssetReference {blueprintRigClass=&/F_MED_MarchTreatBay_FaceAcc_CtrlRig.F_MED_MarchTreatBay_FaceAcc_CtrlRig_C}
   }
   Save cached pose 'PreSimCache'(Pose=@n55) {cacheName=PreSimCache}
  ]}
 EventGraph {nodes [
   Event BlueprintUpdateAnimation()
   > n1 = Sequence!(execute) -> (then_0, then_1, then_2:Exec, then_3:Exec, then_4:Exec)
   > n2 = "Gravity Override State Function (Parameter Asset) (Nativized)"(
    CharacterPartAnimInstance=Self-Reference()
    GameTimeInSeconds=GetGameTimeInSeconds()
    Parameters=&/F_MED_MarchTreatBay_FaceAcc_Parameters
   )
   > Set GravityOverride(GravityOverride=@n2)
   Set body Mesh(execute=@n1.then_1, body Mesh:Skeletal Mesh Component Object Reference=GetOwnerSkeletalMeshComponent()) -> (Output_Get:Skeletal Mesh Component Object Reference)
   n3 = Get bRigidBodyEnabled()
   Set Hair(execute=@n1.then_3, Hair:bool=OR Boolean(A=Get IsWearingBackpackOrCape() -> (IsWearingBackpackOrCape:bool), B=NOT Boolean(A=@n3))) -> (Output_Get:bool)
   n4 = Branch(execute=@n1.then_4, Condition=Get bIsSkydiveFloating())
   > Set isSkyDivingFloat(isSkyDivingFloat:double=1) -> (Output_Get:double)
   Set isSkyDivingFloat(execute=@n4.else, isSkyDivingFloat=0)
   n5 = NOT Boolean(A=Get bIsPlayingEmote())
   n6 = NOT Boolean(A=Get bIsSkyDiving())
   n7 = Branch(execute=@n1.then_2, Condition=AND Boolean(A=Get bIsInFrontEnd(), B=AND Boolean(A=@n6, B=@n5)))
   > Set SimSpaceSettings(SimSpaceSettings:Anim Phys Sim Space Settings Structure=Get SimSpaceSettingsFE() -> (SimSpaceSettingsFE:Anim Phys Sim Space Settings Structure)) -> (Output_Get:Anim Phys Sim Space Settings Structure)
   Set SimSpaceSettings(execute=@n7.else, SimSpaceSettings=Get SimSpaceSettingsPIE() -> (SimSpaceSettingsPIE:Anim Phys Sim Space Settings Structure))
   Event OnCharacterCustomizationLoaded()
   n8 = CheckForBackbling()
   > Set IsWearingBackpackOrCape(IsWearingBackpackOrCape:bool=OR Boolean!(A=@n8, B=@n8.bIsBackblingACape, C:bool=@n8.bIsBackblingLong) -> (ReturnValue)) -> (Output_Get:bool)
   Event BlueprintInitializeAnimation()
   > Set body Mesh(body Mesh=GetOwnerSkeletalMeshComponent())
  ]}
]
variables [
 body Mesh {type=SkeletalMeshComponent}
 GravityOverride {type=Vector}
 Hair {type=boolean}
 isSkyDivingFloat {type=number}
 IsWearingBackpackOrCape {type=boolean}
 SimSpaceSettings {
  type=AnimPhysSimSpaceSettings
  default {maxAngularAcceleration=100000, maxAngularVelocity=100000, simSpaceAngularAlpha=1, externalAngularVelocity=<100000 100000 100000>}
 }
 SimSpaceSettingsFE {
  type=AnimPhysSimSpaceSettings
  default {maxAngularAcceleration=50, maxAngularVelocity=3, simSpaceAngularAlpha=1, externalAngularVelocity=<0 0 0>}
 }
 SimSpaceSettingsPIE {
  type=AnimPhysSimSpaceSettings
  default {maxAngularAcceleration=20, maxAngularVelocity=10, simSpaceAngularAlpha=1, externalAngularVelocity=<0 0 2>}
 }
]
